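; Screen haze effect actor for ToyFox that feeds its world location and a timestamp to the MPC_ScreenHazeUp material parameter collection.
; Both the construction script and BeginPlay write the actor location to Position and the current game time to GameTime.
; There is no per-tick update, so these values are snapshots from construction or play start.
; The overlap event and the Position and Strength OnRep functions are empty.
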(event EventBeginPlay
  (Rendering|Material|SetVectorParameterValue "/ToyFoxGameplay/Assets/FX/Materials/MPC_ScreenHazeUp.MPC_ScreenHazeUp" "Position" (Math|Conversions|ToLinearColor(Vector) (Transformation|GetActorLocation)) :!class /Script/Engine.KismetMaterialLibrary)
  (Rendering|Material|SetScalarParameterValue "/ToyFoxGameplay/Assets/FX/Materials/MPC_ScreenHazeUp.MPC_ScreenHazeUp" "GameTime" (Utilities|Time|GetGameTimeinSeconds) :!class /Script/Engine.KismetMaterialLibrary))

(event Collision|EventActorBeginOverlap
  (params (OtherActor :type Actor%20Object%20Reference)))

(fn OnRep_Position)

(fn OnRep_Strength)

(fn ConstructionScript
  (Rendering|Material|SetVectorParameterValue "/ToyFoxGameplay/Assets/FX/Materials/MPC_ScreenHazeUp.MPC_ScreenHazeUp" "Position" (Math|Conversions|ToLinearColor(Vector) (Transformation|GetActorLocation)) :!class /Script/Engine.KismetMaterialLibrary)
  (Rendering|Material|SetScalarParameterValue "/ToyFoxGameplay/Assets/FX/Materials/MPC_ScreenHazeUp.MPC_ScreenHazeUp" "GameTime" (Utilities|Time|GetGameTimeinSeconds) :!class /Script/Engine.KismetMaterialLibrary))
