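# Translucent, unlit, two-sided Niagara ribbon material for the dust trail of ToyFox's power sword.
# A sharpened sine across UV1 and a fade at both ends form a mask, multiplied by panning noise and particle alpha, then depth-faded for opacity.
# Panning smoke, tinted by particle color and a mask-driven gradient, feeds the emissive adjustment function.
# The final color blend is fixed at 1, so only the white-to-grey gradient shows, not the brown pair.
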Material /ToyFoxGameplay/Assets/FX/M_ToyFox_PowerSword_DustTrail
graph {
 properties {
  bAutomaticallySetUsageInEditor=false
  bUsedWithNiagaraRibbons=true
  blendMode=BLEND_Translucent
  shadingModel=MSM_Unlit
  twoSided=true
  usedShadingModels=Unlit
 }
 nodes [
  n1 = ParticleColor()
  n2 = Multiply(A=MaterialFunctionCall(Float2=TextureCoordinate() {coordinateIndex=1}) {materialFunction=&BreakOutFloat2Components}.R) {constB=0.5}
  n3 = Power(Base=Saturate(None=Sine(None=@n2)), Exp=ScalarParameter() {defaultValue=10})
  n4 = MaterialFunctionCall(Float2=OneMinus(None=TextureCoordinate() {coordinateIndex=1})) {materialFunction=&BreakOutFloat2Components}
  n5 = Multiply(A=OneMinus(None=@n4.G)) {constB=15}
  n6 = Multiply(A=@n3, B=Multiply(A=Saturate(None=@n5), B=@n4.G)) {constA=1}
  n7 = ScalarParameter() {defaultValue=1, parameterName=U_Tile}
  n8 = ScalarParameter() {defaultValue=1, parameterName=V_Tile}
  n9 = Panner(Coordinate=Multiply(A=TextureCoordinate(), B=Multiply(A=@n7, B=@n8))) {speedX=-0.5}
  n10 = TextureSample(UVs=@n9) {texture=/Game/Effects/Niagara/MasterMaterials/SplineThickenTrail/Texture/T_Pickaxe_SoapPocket_NoiseGlow}
  DepthFade(Opacity=Multiply(A=@n1.A, B=Multiply(A=@n6, B=@n10.R))) {fadeDistanceDefault=300}
  n11 = Power(Base=TextureSample(UVs=@n9) {samplerType=SAMPLERTYPE_LinearColor, texture=/Game/Global/Textures/Default/Tileable_Patterns/T_Tile_Smoke_Small}.R)
  n12 = LinearInterpolate(
   A=LinearInterpolate(
    A=VectorParameter() {parameterName=Param_1, defaultValue=<rgba 0.6997 0.5479 0.447 1>}.RGB
    B=VectorParameter() {defaultValue=<rgba 0.1181 0.0731 0.0425 1>}.RGB
    Alpha=@n3
   )
   B=LinearInterpolate(
    A=VectorParameter() {parameterName=Param_2, defaultValue=<rgba 1 1 1 0>}.RGB
    B=VectorParameter() {parameterName=Param_3, defaultValue=<rgba 0.526 0.526 0.526 0>}.RGB
    Alpha=@n3
   )
  ) {constAlpha=1}
  MaterialFunctionCall(Emissive=Multiply(A=@n11, B=Multiply(A=@n12, B=@n1.RGB))) {materialFunction=&MF_FX_EmissiveAdjustments}
  Multiply(A=Constant() {r=0.5}, B=@n11)
 ]
}
