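# Translucent, unlit, two-sided refractive shockwave material for ToyFox effects, usable on Niagara and mesh particles.
# Panning, self-distorted noise, eroded by dynamic parameters and tinted by particle color, drives glow, optionally mixed with channel-split scene color for chromatic aberration.
# Fresnel, view-angle fade, bottom gradient and depth fade shape the opacity mask.
# Refraction scales with camera FOV and optional near-camera fade, and is disabled on Low quality, forward and mobile.
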Material /ToyFoxGameplay/Assets/FX/M_ToyFox_Soft_Ref_Shockwave
graph {
 properties {
  bUsedWithMeshParticles=true
  bUsedWithNanite=true
  bUsedWithNiagaraMeshParticles=true
  bUsedWithNiagaraRibbons=true
  bUsedWithNiagaraSprites=true
  blendMode=BLEND_Translucent
  pixelDepthOffsetMode=PDOM_Legacy
  refractionMethod=RM_IndexOfRefraction
  shadingModel=MSM_Unlit
  twoSided=true
  usedShadingModels=Unlit
 }
 nodes [
  n1 = ScalarParameter() {defaultValue=1, parameterName=Param_5}
  n2 = Multiply(A=TextureCoordinate(), B=AppendVector(A=ScalarParameter() {defaultValue=5}, B=ScalarParameter() {defaultValue=1, parameterName=Param_1}))
  n3 = Multiply(A=Time(), B=ScalarParameter() {defaultValue=1, parameterName=TexPanScale})
  n4 = TextureSample(UVs=Panner(Coordinate=@n2, Time=@n3) {speedX=-0.3, speedY=-0.3}) {samplerType=SAMPLERTYPE_LinearColor, texture=&T_NoisePack_LinearColor}
  n5 = Panner(Coordinate=Add(A=@n2, B=Divide(A=@n4.R) {constB=5}) {constB=0}, Time=@n3) {speedX=0.5, speedY=0.5}
  n6 = TextureSample(UVs=@n5) {texture=/Game/Effects/Fort_Effects/Textures/Tile/T_Noise_03}
  n7 = TextureSample(UVs=@n5) {samplerType=SAMPLERTYPE_LinearColor, texture=/Game/Characters/Player/Male/Medium/Bodies/M_MED_SRDrift/Textures/T_ElectricNoise}
  n8 = LinearInterpolate(Alpha=Add(A=@n6.B, B=@n7.R)) {constA=100, constB=1000}
  n9 = Saturate(None=OneMinus(None=DepthFade(Opacity=@n1, FadeDistance=@n8)))
  n10 = ParticleColor()
  n11 = MaterialFunctionCall(
   In=OneMinus(None=MaterialFunctionCall() {materialFunction=&LinearGradient}.VGradient)
   Contrast=ScalarParameter() {defaultValue=0.05, parameterName=BottomFalloff}
  ) {materialFunction=&CheapContrast}
  n12 = ScalarParameter() {defaultValue=3, parameterName=FresnelExponent}
  n13 = DynamicParameter() {defaultValue=<rgba 1 0 1 1>, paramNames [RefractionIndex, Erosion, FresnelInner, FresnelOuter]}
  n14 = MaterialFunctionCall("FadeSensitivity (optional)"=Multiply(A=@n13.FresnelOuter, B=ScalarParameter() {defaultValue=1, parameterName=FresnelExponent2})) {materialFunction=&FadeBasedOnViewAngle}
  n15 = Multiply(A=Fresnel(ExponentIn=Multiply(A=@n12, B=@n13.FresnelInner)) {baseReflectFraction=0, exponent=1}, B=@n14.Result)
  n16 = ScalarParameter() {defaultValue=1, parameterName=FresnelBrightness}
  n17 = DepthFade(Opacity=Abs(None=Multiply(A=@n15, B=@n16)), FadeDistance=ScalarParameter() {defaultValue=1000, parameterName=DepthFade_Distance})
  n18 = Constant() {r=1}
  n19 = QualitySwitch(Default=@n17, Low=@n18, High=@n17, Medium=@n17, Epic=@n17)
  n20 = Multiply(A=@n10.A, B=Multiply(A=@n11.Result, B=ShadingPathSwitch(Default=@n19, Deferred=@n19, Forward=@n18, Mobile=@n18)))
  Add(A=Multiply(A=@n9, B=@n20), B=@n20)
  n21 = ScreenPosition()
  n22 = Constant() {r=1}
  n23 = Constant() {r=-1}
  n24 = Panner(Coordinate=Multiply(A=TextureCoordinate(), B=AppendVector(A=@n22, B=@n23)), Speed=AppendVector(A=Constant(), B=Constant() {r=1})) {speedY=0.2}
  n25 = Divide(
   A=ComponentMask(None=TextureSample(UVs=@n24) {samplerType=SAMPLERTYPE_Normal, texture=/Game/Global/Textures/Default/Tileable_Patterns/1Normal}.RGB) {g=true, r=true}
   B=Constant() {r=20}
  ) {constB=6}
  n26 = SceneColor!(UVs=Add(A=@n21.ViewportUV, B=@n25))
  n27 = SceneColor(UVs=Add(A=@n21.ViewportUV, B=Divide(A=@n25)))
  n28 = SceneColor(UVs=Add(A=@n21.ViewportUV, B=Divide(A=@n25) {constB=3}))
  n29 = Multiply(
   A=MaterialFunctionCall(
    R=MaterialFunctionCall(Float3=@n26.RGB) {materialFunction=&BreakOutFloat3Components}.R
    G=MaterialFunctionCall(Float3=@n27.RGB) {materialFunction=&BreakOutFloat3Components}.G
    B=MaterialFunctionCall(Float3=@n28.RGB) {materialFunction=&BreakOutFloat3Components}.B
   ) {materialFunction=&AppendMany}.RGB
   B=ScalarParameter() {defaultValue=20, parameterName=ChromaticIntensity}
  )
  n30 = Constant()
  n31 = ShadingPathSwitch(Default=@n29, Deferred=@n29, Forward=@n30, Mobile=@n30)
  n32 = ScalarParameter() {parameterName=UseSceneColorForChromaticFX}
  n33 = Divide(A=MaterialFunctionCall(In=@n7.RGB, Contrast=ScalarParameter() {defaultValue=0.5, parameterName=Param_7}) {materialFunction=&CheapContrast}.Result)
  n34 = Saturate(None=Subtract(
    A=Add(A=@n6.RGB, B=@n33)
    B=DynamicParameter() {defaultValue=<rgba 1 0 1 1>, paramNames [RefractionIndex, Erosion, FresnelInner, FresnelOuter]}.Erosion
   ))
  n35 = ScalarParameter() {defaultValue=20, parameterName=Param_2}
  n36 = Multiply(A=Multiply(A=LinearInterpolate(A=@n31, Alpha=@n32), B=Multiply(A=@n34, B=@n35)), B=@n10.RGB)
  Add(A=@n36, B=Multiply(A=Multiply(A=@n36, B=@n9), B=ScalarParameter() {defaultValue=50, parameterName=Param_6}))
  MaterialFunctionCall(Emissive=@n36) {materialFunction=&MF_FX_EmissiveAdjustments}
  n37 = Clamp(None=Multiply(A=@n11.Result, B=@n14.Result))
  n38 = ScalarParameter() {defaultValue=4096, group=CameraFade, parameterName=Refraction CameraFade Distance}
  n39 = Multiply(
   A=@n37
   B=OneMinus(None=MaterialFunctionCall(Distance=@n38) {materialFunction=&Near_Camera_Fade}.Result)
  )
  n40 = ScalarParameter() {group=CameraFade, parameterName=Use Refraction CameraFade, sortPriority=0}
  n41 = ComponentMask(None=Multiply(A=Custom!(None) {code="atan(View.ClipToView[0][0])*2", desc=FOV in radians, description=GrabFOV, outputType=CMOT_Float2}) {constB=57.2958, desc=to degrees}) {g=true}
  n42 = ScalarParameter() {defaultValue=30, desc="Camera FOV Min Angle (long lens)", group=CameraFOVFade, parameterName=FOV Degrees Min, sliderMax=30, sortPriority=20}
  n43 = ScalarParameter() {defaultValue=80, desc="Camera FOV Max Angle (wide lens)", group=CameraFOVFade, parameterName=FOV Degrees Max, sliderMax=90, sortPriority=21}
  n44 = ScalarParameter() {defaultValue=0.4, desc="IOR at Min (long lens) Angle", group=CameraFOVFade, parameterName=FOV Refract Min, sliderMax=1, sortPriority=22}
  n45 = ScalarParameter() {defaultValue=1, desc="IOR at Max(wide lens) Angle", group=CameraFOVFade, parameterName=FOV Refract Max, sliderMax=1, sortPriority=23}
  n46 = LinearInterpolate(
   B=DynamicParameter() {defaultValue=<rgba 1 0 1 1>, paramNames [RefractionIndex, Erosion, FresnelInner, FresnelOuter]}.RefractionIndex
   Alpha=Multiply(
    A=LinearInterpolate(A=@n37, B=@n39, Alpha=@n40) {constA=1, constB=0}
    B=MaterialFunctionCall(Input=@n41, Input Low=@n42, Input High=@n43, Target Low=@n44, Target High=@n45) {materialFunction=&RemapValueRange}.Result
   ) {desc="0=NoRefract"}
  ) {constA=1}
  n47 = Constant() {r=1}
  n48 = QualitySwitch(Default=@n46, Low=@n47, High=@n46, Medium=@n46, Epic=@n46)
  ShadingPathSwitch(Default=@n48, Deferred=@n48, Forward=@n47, Mobile=@n47)
 ]
}
