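# Unlit, two-sided masked material for Niagara ribbons that renders a glowing, animated plasma tube trail for ToyFox effects.
# Noise-distorted, panning plasma coil samples are lighten-blended, modulated by scrolling noise, then tinted and added to a particle-colored gradient driven by Fresnel and a curve atlas.
# A UV1 length gradient boosts emissive toward one end.
# Particle alpha, end fades and near-camera fade feed dithered temporal AA for soft masking.
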Material /ToyFoxGameplay/Assets/FX/M_ToyFox_TubeTrail
graph {
 properties {
  bAutomaticallySetUsageInEditor=false
  bUsedWithNiagaraRibbons=true
  blendMode=BLEND_Masked
  shadingModel=MSM_Unlit
  twoSided=true
  usedShadingModels=Unlit
 }
 nodes [
  n1 = TextureCoordinate()
  n2 = ScalarParameter() {defaultValue=1, parameterName=Tile}
  n3 = AppendVector(A=ScalarParameter() {defaultValue=1, parameterName=Tile_U}, B=Divide(A=@n2))
  n4 = Multiply(A=TextureCoordinate(), B=@n2)
  n5 = ComponentMask(None=TextureSample(UVs=Panner(Coordinate=@n4) {speedX=0.5, speedY=0.5}) {texture=/Game/Effects/Fort_Effects/Emotes/UltraLight/Texture/T_SoftNoise_BC7}.RGB) {g=true, r=true}
  n6 = ScalarParameter() {defaultValue=5, parameterName=UV_Distortion}
  n7 = Panner(Coordinate=Add(A=Multiply(A=@n1, B=@n3), B=Divide(A=@n5, B=@n6))) {speedX=-1, speedY=-0.5}
  n8 = TextureSample(UVs=@n7) {samplerType=SAMPLERTYPE_LinearGrayscale, texture=/Game/Effects/Fort_Effects/Emotes/UltraLight/Texture/T_PlasmaCoils}
  n9 = TextureSample(UVs=Multiply(A=@n7) {constB=2}) {samplerType=SAMPLERTYPE_LinearGrayscale, texture=&T_PlasmaCoils}
  n10 = Multiply(A=TextureCoordinate(), B=@n3)
  n11 = Panner(Coordinate=Multiply(A=TextureCoordinate(), B=@n2)) {speedX=-0.5, speedY=0.5}
  n12 = Divide(
   A=ComponentMask(None=TextureSample(UVs=@n11) {samplerType=SAMPLERTYPE_LinearGrayscale, texture=/Game/Environments/Hexmap/HexClouds/T_HexmapFog_01_Height02}.RGB) {g=true, r=true}
   B=@n6
  )
  n13 = TextureSample(UVs=Panner(Coordinate=Add(A=@n10, B=@n12)) {speedX=1, speedY=1}) {texture=/Game/Effects/Fort_Effects/Textures/Tile/T_Noise_02}
  n14 = Constant() {r=5}
  n15 = Multiply(
   A=Add(A=MaterialFunctionCall!(Blend=@n8.R, Base=@n9.R) {materialFunction=/Engine/Functions/Engine_MaterialFunctions03/Blends/Blend_Lighten}.Result) {constB=0}
   B=Saturate(None=Multiply(A=@n13.RGB, B=@n14))
  )
  n16 = ScalarParameter() {defaultValue=1.5, parameterName=Param_4}
  n17 = VectorParameter() {parameterName=ColorA, defaultValue=<rgba 1 0.1931 0.0557 1>}
  n18 = ScalarParameter() {defaultValue=5}
  n19 = ParticleColor()
  n20 = VectorParameter() {parameterName=ColorB, defaultValue=<rgba 0.0461 0 0.05 1>}
  n21 = ScalarParameter() {defaultValue=10, parameterName=Param_3}
  n22 = CurveAtlasRowParameter(CurveTime=Power(Base=Fresnel(ExponentIn=@n21) {baseReflectFraction=0, exponent=1}, Exp=ScalarParameter() {defaultValue=1, parameterName=Param_1})) {atlas=&/CA_ToyFox_VFX, curve=&/Curve_ToyFox_Color01, parameterName=Param_2}
  n23 = Add(
   A=Multiply(A=Power(Base=@n15, Exp=@n16), B=Multiply(A=@n17.RGB, B=@n18))
   B=Multiply(A=@n19.RGB, B=LinearInterpolate(A=@n20.RGB, B=@n17.RGB, Alpha=@n22.A))
  )
  n24 = MaterialFunctionCall(UV Channel=TextureCoordinate() {coordinateIndex=1}) {materialFunction=&LinearGradient}
  n25 = OneMinus(None=@n24.UGradient)
  n26 = Multiply(A=@n17.RGB, B=Power(Base=@n25, Exp=ScalarParameter() {defaultValue=10, parameterName=Exp}) {constExponent=4})
  n27 = Multiply(A=Multiply(A=@n23, B=@n26), B=ScalarParameter() {defaultValue=100, parameterName=B}) {constB=20}
  n28 = Multiply(A=@n26) {constB=2}
  MaterialFunctionCall(Emissive=Add(A=@n23, B=Add(A=@n27, B=@n28))) {materialFunction=&MF_FX_EmissiveAdjustments}
  n29 = ScalarParameter() {defaultValue=25, parameterName=Param_5}
  n30 = Multiply(A=@n25, B=LinearInterpolate(B=Multiply(A=@n24.UGradient), Alpha=ScalarParameter() {parameterName=MaskFadeOutHeadEdge}) {constA=1})
  n31 = ScalarParameter() {defaultValue=2, parameterName=MaskEdgeHardness}
  n32 = Multiply(A=Saturate(None=Multiply(A=@n24.UGradient, B=@n29)), B=Saturate(None=Multiply(A=@n30, B=@n31) {constB=2}))
  n33 = Multiply(
   A=MaterialFunctionCall(Distance=ScalarParameter() {defaultValue=5000, parameterName=NearCameraFade}) {materialFunction=&Near_Camera_Fade}.Result
   B=ScalarParameter() {defaultValue=4, parameterName=NearCameraFadeHardness}
  )
  MaterialFunctionCall(Alpha Threshold=Multiply(A=Multiply(A=@n19.A, B=@n32), B=Saturate(None=@n33))) {materialFunction=&DitherTemporalAA_Fortnite}
  n34 = Add(A=ScreenPosition().ViewportUV, B=Divide(A=Time()) {constB=100})
  Multiply(A=TextureSample(UVs=Multiply(A=@n34) {constB=7}).RGB) {constB=2}
 ]
}
