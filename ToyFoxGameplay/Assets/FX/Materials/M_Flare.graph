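# Additive, unlit, two-sided lens-flare-style material for Niagara particles, built on a diffraction bloom texture.
# A radial mask and particle-driven UVMove warp the texture's UVs, and a dynamic HDRPower controls its LDR-to-HDR expansion.
# Time and noise shift the hue, then a color dot-product mask blends this with an unshifted copy, both saturation-boosted.
# Particle color tints emissive, while particle alpha and two radial masks form an opacity term with a suppressed center.
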Material /ToyFoxGameplay/Assets/FX/Materials/M_Flare
graph {
 properties {
  bAutomaticallySetUsageInEditor=false
  bUsedWithNiagaraMeshParticles=true
  bUsedWithNiagaraRibbons=true
  bUsedWithNiagaraSprites=true
  blendMode=BLEND_Additive
  shadingModel=MSM_Unlit
  twoSided=true
  usedShadingModels=Unlit
 }
 nodes [
  n1 = Multiply(A=Time(), B=ScalarParameter() {parameterName=TM})
  n2 = TextureObjectParameter() {
   group=Noise
   parameterName=Noise_Texture
   samplerType=SAMPLERTYPE_Grayscale
   sortPriority=0
   texture=/Game/Weapons/FORT_Melee/Pickaxe_Cube_Ninja_Male/FX/T_CellNoise
  }
  n3 = VectorParameter() {group=Noise, parameterName=Noise_UVControl, channelNames {a=VPan, b=UPan, g=VScale, r=UScale}, defaultValue=<rgba 1 1 0 0>}
  n4 = MaterialFunctionCall(Texture=@n2, UVControl_RGB=@n3, UVControl_A=@n3) {materialFunction=&FXCommon_TextureSample}
  n5 = TextureCoordinate()
  n6 = Subtract(A=@n5, B=Constant2Vector() {g=0.5, r=0.5})
  n7 = TextureCoordinate()
  n8 = Constant2Vector() {g=0.5, r=0.5}
  n9 = DynamicParameter() {defaultValue=<rgba 0 0.2 1 1>, paramNames [UVMove,HDRPower,Param3,Param4]}
  n10 = Multiply(
   A=Power(Base=SphereMask(A=@n7, B=@n8) {attenuationRadius=0.45, hardnessPercent=0}, Exp=ScalarParameter() {defaultValue=0.7, parameterName=T})
   B=@n9.UVMove
  )
  n11 = TextureSample(UVs=Add(A=Multiply(A=@n6, B=@n10), B=@n5)) {texture=&/T_DiffractionBloom}
  n12 = MaterialFunctionCall(
   LDR Input=@n11.RGB
   HDR Dynamic Range=ScalarParameter() {defaultValue=4, parameterName="HDR Dynamic Range (S)"}
   HDR Intensity Multiplier=ScalarParameter() {defaultValue=1, parameterName="HDR Intensity Multiplier (S)"}
   Derive HDR Power=@n9.HDRPower
  ) {materialFunction=&DeriveHDRfromLDR}
  n13 = Desaturation(
   None=MaterialFunctionCall(Hue Shift Percentage=Add(A=@n1, B=@n4.R), Texture=@n12.RemappedRange) {materialFunction=&HueShift}.Result
   Fraction=ScalarParameter() {defaultValue=-4, parameterName=Des}
  )
  n14 = Desaturation(None=@n12.RemappedRange, Fraction=ScalarParameter() {defaultValue=-1, parameterName=Des2})
  n15 = Power(
   Base=DotProduct(A=Normalize(VectorInput=@n11.RGB), B=VectorParameter() {parameterName=DotR, defaultValue=<rgba 0 0.2044 1 1>}.RGB)
   Exp=ScalarParameter() {defaultValue=5, parameterName=RDD}
  )
  n16 = ParticleColor()
  MaterialFunctionCall(Emissive=Multiply(A=LinearInterpolate(A=@n13, B=@n14, Alpha=@n15), B=@n16.RGB)) {materialFunction=&MF_FX_EmissiveAdjustments}
  n17 = SphereMask(
   A=TextureCoordinate()
   B=Constant2Vector() {g=0.5, r=0.5}
   Radius=ScalarParameter() {defaultValue=0.5, parameterName=R}
   Hardness=ScalarParameter() {parameterName=H}
  )
  n18 = VectorParameter() {parameterName=OpacityRe_Refine, defaultValue=<rgba 3.8225 1.5836 1.105 0.5>}
  n19 = SphereMask(
   A=TextureCoordinate()
   B=Constant2Vector() {g=0.5, r=0.5}
   Radius=ScalarParameter() {defaultValue=0.1672, parameterName=R3}
   Hardness=ScalarParameter() {defaultValue=0.3136, parameterName=H3}
  )
  Multiply(
   A=Multiply(
    A=@n16.A
    B=MaterialFunctionCall(Texture=@n17, RGB=@n18.RGB, A=@n18.RGB) {materialFunction=&MF_FX_Refine}.Result
   )
   B=Power(Base=OneMinus(None=@n19))
  )
 ]
}
