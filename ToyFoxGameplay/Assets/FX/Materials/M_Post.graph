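# Additive, unlit Niagara material that builds a full-screen, blue-tinted edge effect in screen UVs with depth testing off.
# It sums edge-masked scrolling radial wave lines, pulsing side glows chosen by camera direction against an Axis vector, and a noise-driven hue-shifted base.
# A platform switch limits the base's distorted, camera-offset scene-normal Fresnel vignette to high-end targets.
# The sum feeds emissive adjustment and a debug node, with no material output connection shown.
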Material /ToyFoxGameplay/Assets/FX/Materials/M_Post
graph {
 properties {
  bAutomaticallySetUsageInEditor=false
  bDisableDepthTest=true
  bUsedWithNiagaraMeshParticles=true
  bUsedWithNiagaraRibbons=true
  bUsedWithNiagaraSprites=true
  blendMode=BLEND_Additive
  shadingModel=MSM_Unlit
  usedShadingModels=Unlit
 }
 nodes [
  n1 = MaterialFunctionCall(Float2=ScreenPosition().ViewportUV) {materialFunction=&BreakOutFloat2Components}
  Add(A=@n1.G) {constB=0.5}
  MaterialFunctionCall(Emissive=Constant() {r=1}) {materialFunction=&MF_FX_EmissiveAdjustments}
  n2 = ScreenPosition()
  n3 = VectorParameter() {group=UMask, parameterName=UMask_EdgeMask_InvertControl, channelNames {a="", b=Power, g=Multiply, r=Offset}, defaultValue=<rgba 0 2.5 1.5 1>}
  n4 = MaterialFunctionCall!(
   UorV=MaterialFunctionCall(Float2=@n2.ViewportUV) {materialFunction=&BreakOutFloat2Components}.R
   Invert_EdgeControl=@n3
   NoInvert_EdgeControl=@n3
  ) {materialFunction=/Game/Effects/Niagara/MasterMaterials/Functions/Common/FXCommon_EdgeMask_Ex}
  n5 = Constant2Vector() {g=0.5, r=0.5}
  n6 = ScalarParameter() {defaultValue=0.8, parameterName=Rad5}
  n7 = Power(
   Base=OneMinus(None=SphereMask(A=@n2.ViewportUV, B=@n5, Radius=@n6) {attenuationRadius=1, hardnessPercent=0})
   Exp=ScalarParameter() {defaultValue=2.9458, parameterName=Pow5}
  ) {constExponent=3}
  n8 = Multiply(
   A=Add(A=ScreenPosition().ViewportUV) {constB=-0.5}
   B=ComponentMask(None=VectorParameter() {parameterName=et4, defaultValue=<rgba 2 1 0 1>}.RGB) {g=true, r=true}
  )
  n9 = Add(
   A=MaterialFunctionCall(Float2=MaterialFunctionCall(Vector or UVs=@n8) {materialFunction=&VectorToRadialValue}.Radial Coordinates) {materialFunction=&BreakOutFloat2Components}.G
   B=Multiply(A=Time(), B=ScalarParameter() {defaultValue=-0.01, parameterName=WaveLineOffset})
  )
  n10 = ScalarParameter() {defaultValue=30, parameterName=WaveLineScale}
  n11 = Abs(None=Sine(None=Multiply(A=@n9, B=@n10)))
  n12 = ScalarParameter() {group=WaveLine, parameterName=WaveLine_GrayAdjust_Add}
  n13 = Saturate(None=Multiply(A=Add(A=@n11, B=@n12), B=ScalarParameter() {defaultValue=1, group=WaveLine, parameterName=WaveLine_GrayAdjust_Mul}))
  n14 = ScalarParameter() {defaultValue=300.3369, group=WaveLine, parameterName=WaveLine_GrayAdjust_Power}
  n15 = Multiply(A=Saturate(None=Power(Base=@n13, Exp=@n14)), B=VectorParameter() {parameterName=WaveLine, defaultValue=<rgba 0 0.309 5 1>}.RGB)
  n16 = Multiply(A=@n1.G) {constB=0.5}
  n17 = VectorParameter() {parameterName=curve, defaultValue=<rgba -50.2557 0.7123 0 1>}
  n18 = Multiply(
   A=MaterialFunctionCall(Input=Sine(None=@n16), Input Low=@n17.R, Input High=@n17.R) {materialFunction=&RemapValue_0-1}.Result
   B=ScalarParameter() {defaultValue=-0.8502, group=EdgeBright, parameterName=EdgeBright_GrayAdjust_Add}
  )
  n19 = ScalarParameter() {defaultValue=6.5385, group=EdgeBright, parameterName=EdgeBright_GrayAdjust_Mul}
  n20 = Multiply(A=Add(A=OneMinus(None=@n1.R), B=@n18), B=@n19)
  n21 = ScalarParameter() {defaultValue=2, group=EdgeBright, parameterName=EdgeBright_GrayAdjust_Power}
  n22 = Saturate(None=Power(Base=Saturate(None=@n20), Exp=@n21))
  n23 = VectorParameter() {parameterName=EdgeBright_Refine, defaultValue=<rgba 15 1 0.05 0.5>}
  n24 = Subtract(
   A=VectorParameter() {parameterName=Axis, defaultValue=<rgba 1 0 0 1>}.RGB
   B=MaterialFunctionCall() {materialFunction=&CameraDirectionVector}.Result
  )
  n25 = ComponentMask(None=Transform(None=@n24) {transformSourceType=TRANSFORMSOURCE_World, transformType=TRANSFORM_Camera}) {g=true, r=true}
  n26 = Length(None=@n24) {is=class MaterialExpressionLength}
  n27 = ScalarParameter() {defaultValue=0.5, parameterName=EdgeBrightOffsetC}
  n28 = Subtract(A=@n26, B=@n27)
  n29 = Multiply(
   A=Sign(None=MaterialFunctionCall(Float2=@n25) {materialFunction=&BreakOutFloat2Components}.R)
   B=Divide(A=Saturate(None=@n28), B=OneMinus(None=@n27))
  )
  n30 = ScalarParameter() {defaultValue=1, parameterName=Sca}
  n31 = Multiply(A=Clamp(None=@n29), B=@n30)
  n32 = Multiply(A=Add(A=@n1.R, B=@n18), B=@n19)
  n33 = Saturate(None=Power(Base=Saturate(None=@n32), Exp=@n21))
  n34 = Multiply(A=Clamp(None=Multiply(A=@n29) {constB=-1}), B=@n30)
  n35 = Add(
   A=Multiply(
    A=MaterialFunctionCall(Texture=@n22, RGB=@n23.RGB, A=@n23.RGB) {materialFunction=&MF_FX_Refine}.Result
    B=Saturate(None=@n31)
   )
   B=Multiply(
    A=MaterialFunctionCall(Texture=@n33, RGB=@n23.RGB, A=@n23.RGB) {materialFunction=&MF_FX_Refine}.Result
    B=Saturate(None=@n34)
   )
  )
  n36 = Sine(None=Multiply(A=Time(), B=ScalarParameter() {defaultValue=1, parameterName=EgeBTimer}))
  n37 = Multiply(
   A=VectorParameter() {parameterName=EdgeBrightColor, defaultValue=<rgba 4.087 9.5174 50 1>}.RGB
   B=MaterialFunctionCall(Input=Abs(None=@n36), Input Low=Constant() {r=-0.5}, Input High=Constant() {r=1}) {materialFunction=&RemapValue_0-1}.Result
  )
  n38 = TextureObjectParameter() {
   group=Noise
   parameterName=Noise_Texture
   samplerType=SAMPLERTYPE_Normal
   sortPriority=0
   texture=/Game/Characters/Player/Female/Medium/Bodies/F_MED_Bistro_Astronaut/FX/T_GaussianNoise_Normal
  }
  n39 = VectorParameter() {group=Noise, parameterName=Noise_UVControl, channelNames {a=VPan, b=UPan, g=VScale, r=UScale}, defaultValue=<rgba 1 0.7 0 0.2>}
  n40 = ScreenPosition()
  n41 = MaterialFunctionCall(Texture=@n38, UVControl_RGB=@n39, UVControl_A=@n39, "[Override]UVInput"=@n40.ViewportUV) {materialFunction=&FXCommon_TextureSample}
  n42 = Multiply(
   A=MaterialFunctionCall(Float2=@n41.RG) {materialFunction=&BreakOutFloat2Components}.R
   B=ScalarParameter() {defaultValue=0.02, parameterName=S}
  )
  n43 = VectorParameter() {
   group=Noise
   parameterName=Noise_NoiseControl
   sortPriority=60
   channelNames {a=VStrength, b=UStrength, g=VOffset, r=UOffset}
   defaultValue=<rgba 0 0 0.1 0.1>
  }
  n44 = ScreenPosition()
  n45 = VectorParameter() {parameterName=MainMS, defaultValue=<rgba 1 1.5 0 1>}
  n46 = ComponentMask(None=AppendVector(A=@n45.R, B=@n45.R)) {g=true, r=true}
  n47 = ScalarParameter() {defaultValue=-0.2, parameterName=MoveStrength}
  n48 = ScalarParameter() {defaultValue=0.2, parameterName=MaxMove}
  n49 = Multiply(A=Normalize(VectorInput=@n25), B=Clamp(None=Multiply(A=@n26, B=@n47), Min=Multiply(A=@n48) {constB=-1}, Max=@n48)) {constB=0.5}
  n50 = Add(
   A=Add(
    A=MaterialFunctionCall!(NoiseValue=@n41.RG, Control_RGB=@n43, Control_A=@n43) {materialFunction=/Game/Effects/Niagara/MasterMaterials/Functions/Common/FXCommon_NoiseControl}.NoiseValue
    B=MaterialFunctionCall(UVs=@n44.ViewportUV, Texture Scale=@n46) {materialFunction=&ScaleUVsByCenter}.UVs
   )
   B=Multiply(A=ComponentMask(None=@n49) {g=true, r=true}, B=Constant2Vector() {r=1})
  )
  n51 = SphereMask(A=@n50, B=Constant2Vector() {g=0.5, r=0.5}, Radius=ScalarParameter() {defaultValue=0.9168, parameterName=Rad2}) {attenuationRadius=1, hardnessPercent=0}
  n52 = ScalarParameter() {defaultValue=24.4789, parameterName=ExponentIn}
  n53 = ScalarParameter() {parameterName=BaseReflectFractionIn}
  n54 = SceneTexture() {sceneTextureId=PPI_WorldNormal}
  n55 = Multiply(
   A=Power(Base=OneMinus(None=@n51), Exp=ScalarParameter() {defaultValue=4.3789, parameterName=Pow2}) {constExponent=3}
   B=Multiply(
    A=VectorParameter() {parameterName=F, defaultValue=<rgba 0 0.7613 4 1>}.RGB
    B=Fresnel(ExponentIn=@n52, BaseReflectFractionIn=@n53, Normal=@n54.Color) {baseReflectFraction=0.616}
   )
  )
  n56 = Constant()
  n57 = Add(
   A=VectorParameter() {defaultValue=<rgba 0 0.2018 1 1>}.RGB
   B=MaterialFunctionCall!("Xb1,PS4"=@n55, "PC-High&&Epic,XSX,PS5"=@n55, "PC-Perf,Android"=@n56, Switch=@n56, "PC-Low(NotPerf)&&Medium"=@n56) {materialFunction=/Game/Effects/Niagara/MasterMaterials/Functions/Common/PlatformSwitch_3D}.Result
  )
  n58 = Add(
   A=Multiply(A=OneMinus(None=@n4.Mask), B=Multiply(A=@n7, B=@n15))
   B=Add(
    A=Multiply(A=@n35, B=@n37)
    B=MaterialFunctionCall(Hue Shift Percentage=@n42, Texture=@n57) {materialFunction=&HueShift}.Result
   )
  )
  MaterialFunctionCall(
   Number=MaterialFunctionCall(Emissive=Divide(A=@n58) {constB=2.271}) {materialFunction=&MF_FX_EmissiveAdjustments}.Result
   UVs=ScreenPosition().ViewportUV
  ) {materialFunction=&DebugScalarValues}
  MaterialFunctionCall(Vector or UVs=Subtract(A=@n40.ViewportUV) {constB=0.5}) {materialFunction=&VectorToRadialValue}
  n59 = SphereMask(A=@n50, B=Constant2Vector() {g=0.5, r=0.5}, Radius=ScalarParameter() {defaultValue=0.8, parameterName=Rad}) {attenuationRadius=1, hardnessPercent=0}
  Multiply(
   A=Power(Base=OneMinus(None=@n59), Exp=ScalarParameter() {defaultValue=2.5, parameterName=Pow}) {constExponent=3}
   B=ScalarParameter() {defaultValue=1, parameterName=PostStrength}
  )
  Multiply(A=@n44.ViewportUV, B=@n46)
  VectorParameter() {parameterName=CameraDirection, defaultValue=<rgba 0 0 0 1>}
 ]
}
