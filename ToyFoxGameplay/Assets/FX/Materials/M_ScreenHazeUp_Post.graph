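# Post-process material that tints the upper screen edge with a pulsing blue haze and bright rim, apparently signaling something overhead.
# UV gradients and center distance shape the haze, with animated radial wave lines zeroed on lower-end platforms via a platform switch.
# MPC_ScreenHazeUp's Position slides a horizontal mask and fades the effect when the camera faces high above it.
# GameTime and camera distance gate the final blend over scene color; noise nodes are unused.
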Material /ToyFoxGameplay/Assets/FX/Materials/M_ScreenHazeUp_Post
graph {
 properties {bAutomaticallySetUsageInEditor=false, materialDomain=MD_PostProcess, usedShadingModels=Default Lit}
 nodes [
  n1 = Sine(None=Multiply(A=Time(), B=ScalarParameter() {defaultValue=1.5, parameterName=Scale}))
  Abs(None=@n1)
  n2 = ComponentMask(None=SceneTexture() {sceneTextureId=PPI_PostProcessInput0}.Color) {b=true, g=true, r=true}
  n3 = Constant()
  n4 = TextureCoordinate()
  n5 = MaterialFunctionCall(Float2=@n4) {materialFunction=&BreakOutFloat2Components}
  n6 = OneMinus(None=@n5.G)
  n7 = ScalarParameter() {group=waveEdge, parameterName=waveEdge_GrayAdjust_Add}
  n8 = Saturate(None=Multiply(A=Add(A=@n6, B=@n7), B=ScalarParameter() {defaultValue=1, group=waveEdge, parameterName=waveEdge_GrayAdjust_Mul}))
  n9 = ScalarParameter() {defaultValue=4, group=waveEdge, parameterName=waveEdge_GrayAdjust_Power}
  n10 = ScreenPosition()
  n11 = VectorParameter() {parameterName=et4, defaultValue=<rgba 1 1 0 1>}
  n12 = MaterialFunctionCall(Vector or UVs=Multiply(A=Add(A=@n10.ViewportUV) {constB=-0.5}, B=ComponentMask(None=@n11.RGB) {g=true, r=true})) {materialFunction=&VectorToRadialValue}
  n13 = Time()
  n14 = ScalarParameter() {defaultValue=-0.01, parameterName=WaveLineOffset}
  n15 = Multiply(
   A=Add(
    A=MaterialFunctionCall(Float2=@n12.Radial Coordinates) {materialFunction=&BreakOutFloat2Components}.G
    B=Multiply(A=@n13, B=@n14)
   )
   B=ScalarParameter() {defaultValue=30, parameterName=WaveLineScale}
  )
  n16 = Add(A=Abs(None=Sine(None=@n15)), B=ScalarParameter() {group=WaveLine, parameterName=WaveLine_GrayAdjust_Add})
  n17 = ScalarParameter() {defaultValue=1, group=WaveLine, parameterName=WaveLine_GrayAdjust_Mul}
  n18 = Power(
   Base=Saturate(None=Multiply(A=@n16, B=@n17))
   Exp=ScalarParameter() {defaultValue=300.3369, group=WaveLine, parameterName=WaveLine_GrayAdjust_Power}
  )
  n19 = Multiply(
   A=Saturate(None=Power(Base=@n8, Exp=@n9))
   B=Multiply(A=Saturate(None=@n18), B=VectorParameter() {parameterName=WaveLine, defaultValue=<rgba 0 0.4326 7 1>}.RGB)
  )
  n20 = VectorParameter() {defaultValue=<rgba 0.5 1.25 5 1>}
  n21 = Constant() {r=1}
  n22 = MaterialFunctionCall(
   Input=@n1
   Input Low=Constant() {r=-1}
   Input High=@n21
   Target Low=ScalarParameter() {defaultValue=0.5, parameterName=Low}
   Target High=@n21
  ) {materialFunction=&RemapValueRange}
  n23 = Add(A=@n6, B=ScalarParameter() {group=UpEdge, parameterName=UpEdge_GrayAdjust_Add})
  n24 = ScalarParameter() {defaultValue=1, group=UpEdge, parameterName=UpEdge_GrayAdjust_Mul}
  n25 = Power(Base=Saturate(None=Multiply(A=@n23, B=@n24)), Exp=ScalarParameter() {defaultValue=2.5, group=UpEdge, parameterName=UpEdge_GrayAdjust_Power})
  n26 = Distance(A=@n4, B=Constant2Vector() {g=0.5, r=0.5})
  n27 = Subtract(A=@n26, B=ScalarParameter() {parameterName=Upbend_Offset})
  n28 = Multiply(
   A=Add(A=Saturate(None=@n27), B=ScalarParameter() {group=UpBend, parameterName=UpBend_GrayAdjust_Add})
   B=ScalarParameter() {defaultValue=1, group=UpBend, parameterName=UpBend_GrayAdjust_Mul}
  )
  n29 = Multiply(
   A=Power(Base=Saturate(None=@n28), Exp=ScalarParameter() {defaultValue=2.5, group=UpBend, parameterName=UpBend_GrayAdjust_Power})
   B=ScalarParameter() {defaultValue=2, parameterName=ag43}
  )
  n30 = LinearInterpolate(
   A=@n2
   B=Add(
    A=MaterialFunctionCall!("Xb1,PS4"=@n3, "PC-High&&Epic,XSX,PS5"=@n19, "PC-Perf,Android"=@n3, Switch=@n3, "PC-Low(NotPerf)&&Medium"=@n19) {materialFunction=/Game/Effects/Niagara/MasterMaterials/Functions/Common/PlatformSwitch_3D}.Result
    B=Multiply(A=@n20.RGB, B=@n22.Result)
   )
   Alpha=Multiply(A=Saturate(None=@n25), B=Saturate(None=@n29))
  )
  n31 = Multiply(A=VectorParameter() {parameterName=EdgeBrightColor, defaultValue=<rgba 1.5 1.625 3 1>}.RGB, B=@n22.Result)
  n32 = OneMinus(None=@n5.G)
  n33 = ScalarParameter() {group=UpEdge, parameterName=UpEdge2_GrayAdjust_Add}
  n34 = Saturate(None=Multiply(A=Add(A=@n32, B=@n33), B=ScalarParameter() {defaultValue=1, group=UpEdge, parameterName=UpEdge2_GrayAdjust_Mul}))
  n35 = ScalarParameter() {defaultValue=38.5, group=UpEdge, parameterName=UpEdge2_GrayAdjust_Power}
  n36 = ScalarParameter() {parameterName=Upbend2_Offset}
  n37 = Add(A=Saturate(None=Subtract(A=@n26, B=@n36)), B=ScalarParameter() {group=UpBend, parameterName=UpBend_2GrayAdjust_Add})
  n38 = ScalarParameter() {defaultValue=1, group=UpBend, parameterName=UpBend2_GrayAdjust_Mul}
  n39 = Power(Base=Saturate(None=Multiply(A=@n37, B=@n38)), Exp=ScalarParameter() {defaultValue=4, group=UpBend, parameterName=UpBend2_GrayAdjust_Power})
  n40 = ScalarParameter() {defaultValue=7, parameterName=ag432}
  n41 = Multiply(A=Saturate(None=Power(Base=@n34, Exp=@n35)), B=Saturate(None=Multiply(A=@n39, B=@n40)))
  n42 = Add(
   A=ComponentMask(None=CollectionParameter() {collection=&/MPC_ScreenHazeUp, parameterName=Position}) {b=true, g=true, r=true}
   B=VectorParameter() {parameterName=gh, defaultValue=<rgba 0 0 100000 1>}.RGB
  )
  n43 = CameraPositionWS()
  n44 = DotProduct(
   A=MaterialFunctionCall() {materialFunction=&CameraDirectionVector}.Result
   B=Normalize(VectorInput=Subtract(A=@n42, B=@n43))
  )
  n45 = ScalarParameter() {defaultValue=-0.3, group=Mask, parameterName=Mask_GrayAdjust_Add}
  n46 = Saturate(None=Multiply(A=Add(A=@n44, B=@n45), B=ScalarParameter() {defaultValue=2, group=Mask, parameterName=Mask_GrayAdjust_Mul}))
  n47 = OneMinus(None=Power(Base=OneMinus(None=@n46), Exp=ScalarParameter() {defaultValue=2, group=Mask, parameterName=Mask_GrayAdjust_Power}))
  n48 = TextureCoordinate()
  n49 = Add(
   A=ComponentMask(None=CollectionParameter() {collection=&MPC_ScreenHazeUp, parameterName=Position}) {b=true, g=true, r=true}
   B=Constant3Vector() {constant=<rgba 0 0 19800 0>}
  )
  n50 = Divide(A=Multiply(A=ViewProperty().Property) {constB=0.5}) {constB=6.2832}
  n51 = Divide(
   A=Normalize(VectorInput=TransformPosition(None=@n49) {transformSourceType=TRANSFORMPOSSOURCE_World, transformType=TRANSFORMPOSSOURCE_View})
   B=ComponentMask(None=Sine(None=@n50)) {r=true}
  )
  n52 = MaterialFunctionCall(Float2=Divide(A=Clamp(None=@n51) {minDefault=-1})) {materialFunction=&BreakOutFloat2Components}
  n53 = ScalarParameter() {defaultValue=0.5, parameterName=LimitetheRange}
  n54 = VectorParameter() {group=TheMark, parameterName=TheMark_EdgeMask_Control, channelNames {a="", b=Power, g=Multiply, r=Offset}, defaultValue=<rgba 0 2 4 2>}
  n55 = MaterialFunctionCall!(
   UorV=Subtract(
    A=MaterialFunctionCall(Float2=@n48) {materialFunction=&BreakOutFloat2Components}.R
    B=Multiply(A=@n52.R, B=@n53)
   )
   EdgeControl=@n54
  ) {materialFunction=/Game/Effects/Niagara/MasterMaterials/Functions/Common/FXCommon_EdgeMask_Simple}
  n56 = OneMinus(None=Power(Base=OneMinus(None=@n55.Mask), Exp=@n54.A))
  n57 = Constant()
  n58 = Constant() {r=1}
  n59 = ScalarParameter() {defaultValue=0.15, parameterName=sFFW}
  n60 = Constant() {r=1}
  n61 = Time()
  n62 = CollectionParameter() {collection=&MPC_ScreenHazeUp, parameterName=GameTime}
  n63 = Subtract(A=@n61, B=@n62)
  n64 = Subtract(A=Add(A=@n62, B=ScalarParameter() {defaultValue=29.5, parameterName=TimeLen}) {constB=5}, B=@n61)
  n65 = Multiply(A=Saturate(None=Divide(A=@n63) {constB=3}), B=Saturate(None=Divide(A=@n64) {constB=1}))
  n66 = VectorParameter() {parameterName=CameraDistance, defaultValue=<rgba 40000 500 0 1>}
  n67 = Subtract(A=Distance(A=CameraPositionWS(), B=@n49), B=@n66.R)
  n68 = OneMinus(None=Saturate(None=Divide(A=@n67, B=@n66.G)))
  LinearInterpolate(
   A=@n2
   B=LinearInterpolate(A=LinearInterpolate(A=@n30, B=@n31, Alpha=@n41), B=@n2, Alpha=Saturate(None=@n47))
   Alpha=Multiply(
    A=MaterialFunctionCall(Input=@n56, Input Low=@n57, Input High=@n58, Target Low=@n59, Target High=@n60) {materialFunction=&RemapValueRange}.Result
    B=Multiply(A=@n65, B=@n68) {constA=1}
   )
  )
  n69 = TextureObjectParameter() {
   group=Noise
   parameterName=Noise_Texture
   samplerType=SAMPLERTYPE_Normal
   sortPriority=0
   texture=/Game/Characters/Player/Female/Medium/Bodies/F_MED_Bistro_Astronaut/FX/T_GaussianNoise_Normal
  }
  n70 = VectorParameter() {group=Noise, parameterName=Noise_UVControl, channelNames {a=VPan, b=UPan, g=VScale, r=UScale}, defaultValue=<rgba 1 0.7 0 0.2>}
  n71 = ScreenPosition()
  n72 = VectorParameter() {
   group=Noise
   parameterName=Noise_NoiseControl
   sortPriority=60
   channelNames {a=VStrength, b=UStrength, g=VOffset, r=UOffset}
   defaultValue=<rgba 0 0 0.1 0.1>
  }
  MaterialFunctionCall!(
   NoiseValue=MaterialFunctionCall(Texture=@n69, UVControl_RGB=@n70, UVControl_A=@n70, "[Override]UVInput"=@n71.ViewportUV) {materialFunction=&FXCommon_TextureSample}.RG
   Control_RGB=@n72
   Control_A=@n72
  ) {materialFunction=/Game/Effects/Niagara/MasterMaterials/Functions/Common/FXCommon_NoiseControl}
  MaterialFunctionCall(Vector or UVs=Subtract(A=@n71.ViewportUV) {constB=0.5}) {materialFunction=&VectorToRadialValue}
  Multiply()
  Multiply(B=ScalarParameter() {defaultValue=2, parameterName=Sca})
 ]
}
