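# Additive unlit Niagara material for sprites, ribbons and mesh particles that draws a tinted, glowing sheen band with a Fresnel rim.
# A dynamic Offset parameter shifts V before scaling, and two edge masks on it combine by Max into the movable band.
# A Fresnel term, scaled by dynamic Fres and masked along unshifted V, is added, then multiplied by particle alpha.
# Color, Emissive and particle color feed a shared emissive-adjustment function.
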Material /ToyFoxGameplay/Assets/FX/Materials/M_Sheen
graph {
 properties {
  bAutomaticallySetUsageInEditor=false
  bUsedWithNiagaraMeshParticles=true
  bUsedWithNiagaraRibbons=true
  bUsedWithNiagaraSprites=true
  blendMode=BLEND_Additive
  shadingModel=MSM_Unlit
  usedShadingModels=Unlit
 }
 nodes [
  n1 = VectorParameter() {parameterName=Color, defaultValue=<rgba 1 0.6208 0.3425 1>}
  n2 = ScalarParameter() {defaultValue=600, parameterName=Emissive}
  n3 = ParticleColor()
  MaterialFunctionCall(Emissive=Multiply(A=Multiply(A=@n1.RGB, B=@n2), B=@n3.RGB)) {materialFunction=&MF_FX_EmissiveAdjustments}
  n4 = TextureCoordinate()
  n5 = Constant() {r=1}
  n6 = DynamicParameter() {defaultValue=<rgba 0 1 1 1>, paramNames [Offset,Fres,Param3,Param4]}
  n7 = MaterialFunctionCall(X=@n5, Y=@n6.Offset) {materialFunction=&MakeFloat2}
  n8 = ScalarParameter() {defaultValue=1.8415, parameterName=ScaleV}
  n9 = MaterialFunctionCall(Float2=Multiply(
    A=Add(A=@n4, B=@n7.Result)
    B=MaterialFunctionCall(X=@n5, Y=@n8) {materialFunction=&MakeFloat2}.Result
   )) {materialFunction=&BreakOutFloat2Components}
  n10 = VectorParameter() {group=EDge, parameterName=EDge2_EdgeMask_Control, channelNames {a="", b=Power, g=Multiply, r=Offset}, defaultValue=<rgba 0 1 4.4098 0.3>}
  n11 = Multiply(
   A=MaterialFunctionCall!(UorV=@n9.G, EdgeControl=@n10) {materialFunction=/Game/Effects/Niagara/MasterMaterials/Functions/Common/FXCommon_EdgeMask_Simple}.Mask
   B=@n10.A
  )
  n12 = VectorParameter() {group=EDge, parameterName=EDge_EdgeMask_Control, channelNames {a="", b=Power, g=Multiply, r=Offset}, defaultValue=<rgba 0 2.0663 17.5385 1>}
  n13 = MaterialFunctionCall(UorV=@n9.G, EdgeControl=@n12) {materialFunction=&FXCommon_EdgeMask_Simple}
  n14 = Multiply(
   A=@n6.Fres
   B=Fresnel(
    ExponentIn=ScalarParameter() {defaultValue=2.4289, parameterName=EX}
    BaseReflectFractionIn=ScalarParameter() {defaultValue=0.02, parameterName=BR}
   ) {baseReflectFraction=0.001}
  )
  n15 = VectorParameter() {
   group=FresEdge
   parameterName=FresEdge_EdgeMask_InvertControl
   channelNames {a="", b=Power, g=Multiply, r=Offset}
   defaultValue=<rgba 0 2.7855 5.4087 1>
  }
  n16 = VectorParameter() {
   group=FresEdge
   parameterName=FresEdge_EdgeMask_Control
   channelNames {a="", b=Power, g=Multiply, r=Offset}
   defaultValue=<rgba 0 7049.5815 1 1>
  }
  n17 = MaterialFunctionCall!(
   UorV=MaterialFunctionCall(Float2=@n4) {materialFunction=&BreakOutFloat2Components}.G
   Invert_EdgeControl=@n15
   NoInvert_EdgeControl=@n16
  ) {materialFunction=/Game/Effects/Niagara/MasterMaterials/Functions/Common/FXCommon_EdgeMask_Ex}
  Multiply(A=Add(A=Max(A=@n11, B=@n13.Mask), B=Multiply(A=@n14, B=@n17.Mask)), B=@n3.A)
 ]
}
