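# M_ToyFox_Dark is an unlit translucent material for Niagara sprite particles, drawing a round, soft-edged dot.
# A SphereMask centered in UV space produces the circle procedurally, with no texture sample.
# The mask multiplies particle color RGB and alpha separately, so each particle's color and opacity come from the particle system.
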Material /ToyFoxGameplay/Assets/FX/Materials/M_ToyFox_Dark
graph {
 properties {
  bAutomaticallySetUsageInEditor=false
  bUsedWithNiagaraSprites=true
  blendMode=BLEND_Translucent
  shadingModel=MSM_Unlit
  usedShadingModels=Unlit
 }
 nodes [
  n1 = ParticleColor()
  n2 = SphereMask(A=TextureCoordinate(), B=Constant2Vector() {g=0.5, r=0.5}) {attenuationRadius=0.5, hardnessPercent=80}
  Multiply(A=@n1.RGB, B=@n2)
  Multiply(A=@n1.A, B=@n2)
 ]
}
