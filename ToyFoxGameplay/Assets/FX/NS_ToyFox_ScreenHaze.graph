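# NS_ToyFox_ScreenHaze is a Niagara system for the ToyFox gameplay feature, named as a screen haze effect and drawn with a mesh renderer.
# Its single emitter burst-spawns static, uniformly enlarged mesh particles with a fixed lifetime and a per-particle material random value.
# The emitter copies User.Axis into Emitter.Axis; user parameters for local-player state and post strength are also exposed, though no listed module reads them.
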NiagaraSystem /ToyFoxGameplay/Assets/FX/NS_ToyFox_ScreenHaze
dependencies {usedDataInterfaces [], usedDynamicInputs [], usedModules [&SystemState]}
emitters {ScreenHaze {
  emitterUpdateScript [
   &EmitterState()
   &SpawnBurst_Instantaneous(Use Spawn Probability=false, UseLoopCountLimit=false)
   /SetVariables(Emitter.Axis:Vector3f=$User.Axis) {isSetParameters=true}
  ]
  particleSpawnScript [&InitializeParticle(
    Color Mode=::NewEnumerator0
    Lifetime=5
    Mesh Scale Mode=::NewEnumerator3
    Mesh Uniform Scale=6.1005
    UsePositionOffset=false
    Write Material Random=true
   )]
  particleUpdateScript [&ParticleState()]
  rendererClasses [&NiagaraMeshRendererProperties]
  renderers [{rendererClass=&NiagaraMeshRendererProperties, rendererIndex=0}]
 }}
summary {
 systemName=NS_ToyFox_ScreenHaze
 emitters [{emitterName=ScreenHaze, rendererClasses [&NiagaraMeshRendererProperties]}]
 userVariables [
  User.Axis {type=Vector3f, description="", defaultValue {struct=Vector3f, value=<1 0 0>}}
  User.bIsLocalPlayer {type=NiagaraBool, defaultValue=true}
  User.PostStrength {type=NiagaraFloat, defaultValue=1, description=""}
 ]
}
