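# Niagara system for the ToyFox up-screen haze effect, built from one minimal emitter that draws through a component renderer rather than sprites or meshes.
# An instantaneous burst spawns long-lived particles, each initialized with a material random value.
# Each update sets the particle's ComponentsEnabled flag by comparing normalized age to a threshold near end of life, gating the rendered component by particle age.
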NiagaraSystem /ToyFoxGameplay/Assets/FX/NS_ToyFox_UpScreenHaze
dependencies {usedDataInterfaces [], usedDynamicInputs [], usedModules [&SystemState]}
emitters {Minimal {
  emitterUpdateScript [
   &EmitterState()
   &SpawnBurst_Instantaneous(Use Spawn Probability=false, UseLoopCountLimit=false)
  ]
  particleSpawnScript [&InitializeParticle(Color Mode=::NewEnumerator0, Lifetime=90, UsePositionOffset=false, Write Material Random=true)]
  particleUpdateScript [
   &ParticleState()
   /SetVariables(Particles.ComponentsEnabled:NiagaraBool=&SetBoolByFloatComparison(A=$Particles.NormalizedAge, B=0.99, Comparison Type=::NewEnumerator6)) {isSetParameters=true}
  ]
  rendererClasses [&NiagaraComponentRendererProperties]
  renderers [{rendererClass=&NiagaraComponentRendererProperties, rendererIndex=0}]
 }}
summary {
 systemName=NS_ToyFox_UpScreenHaze
 emitters [{emitterName=Minimal, rendererClasses [&NiagaraComponentRendererProperties]}]
 userVariables []
}
