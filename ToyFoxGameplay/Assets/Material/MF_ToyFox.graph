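# Material function for the ToyFox character that layers surface detail, eye glow, cannon heat and visor masking onto incoming material attributes.
# Dirt, dents, color variation and metal-flake normals apply only above Medium quality, which keeps the input.
# Pre-skinned box and sphere masks place eye and cannon effects, driven by Blueprint-timed stage parameters, curve atlases and a sequence-control collection.
# It outputs final attributes, MetalNoise and VisorMask, with emissive clamped on Mobile and Medium.
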MaterialFunction /ToyFoxGameplay/Assets/Material/MF_ToyFox
graph {nodes [
  n1 = FunctionInput() {inputName=IN MA, inputType=FunctionInput_MaterialAttributes}
  n2 = GetMaterialAttributes(None=@n1) {attributeGetTypes [
    #1
    #2
    #3
    #4
    #5
    #6
   ]}
  n3 = Max(A=NamedRerouteUsage(), B=NamedRerouteUsage())
  n4 = SetMaterialAttributes!(
   MaterialAttributes=@n2.MaterialAttributes
   BaseColor=NamedRerouteUsage()
   ClearCoatBottomNormal=NamedRerouteUsage()
   Normal=NamedRerouteUsage()
   Roughness=NamedRerouteUsage()
   Specular=NamedRerouteUsage()
   CustomData0=Multiply(A=@n2.CustomData0, B=OneMinus(None=@n3))
  ) {attributeSetTypes [
    #1
    #7
    #2
    #4
    #5
    #6
   ]}
  n5 = BlendMaterialAttributes(A=@n4, B=@n1, Alpha=NamedRerouteUsage())
  n6 = GetMaterialAttributes(None=MaterialFunctionCall!(Low=@n5, High=@n5, Epic=@n5, --------, Mobile=@n1, Medium=@n1) {materialFunction=/FortniteBaseMaterials/Functions/MF_QualitySwitch_MaterialAttributes}.Result) {attributeGetTypes [#8]}
  n7 = Multiply(A=@n6.EmissiveColor, B=NamedRerouteUsage())
  n8 = NamedRerouteUsage()
  n9 = CollectionParameter() {collection=&/MPC_Toyfox_SequenceControl, desc="Default = 0", group=Toyfox, parameterName=EyeGlowIntensity}
  n10 = Multiply(
   A=Min(A=@n9, B=ScalarParameter() {defaultValue=4, group=Toyfox, parameterName=MobileEmssiveMaxClamp})
   B=ScalarParameter() {defaultValue=0.5, group=Toyfox, parameterName=Mobile Emissive Multiply}
  )
  n11 = Add(
   A=ScalarParameter() {defaultValue=1, group=Toyfox, parameterName=EyeIntensity, sliderMax=1}
   B=MaterialFunctionCall!(Low=@n9, High=@n9, Epic=@n9, --------, Mobile=@n10, Medium=@n10) {materialFunction=/FortniteBaseMaterials/Functions/MF_QualitySwitch_Scalar}.Result
  )
  n12 = Constant() {r=100}
  n13 = Multiply(
   A=Min(A=@n12, B=ScalarParameter() {defaultValue=4, group=Toyfox, parameterName=MobileEmssiveMaxClamp})
   B=ScalarParameter() {defaultValue=0.5, group=Toyfox, parameterName=Mobile Emissive Multiply}
  )
  n14 = NamedRerouteUsage()
  n15 = CurveAtlasRowParameter(CurveTime=Multiply(A=NamedRerouteUsage(), B=OneMinus(None=@n14))) {
   atlas=/ToyFoxGameplay/Assets/Texture/Curves/CA_ToyFox_Sword_Finisher
   curve=/ToyFoxGameplay/Assets/Texture/Curves/C_ToyFox_Sword_Lightning
   defaultValue=1
   group=Toyfox
   parameterName=AnimCurves
  }
  n16 = LinearInterpolate(
   B=MaterialFunctionCall!(Low=@n12, High=@n12, Epic=@n12, --------, Mobile=@n13, Medium=@n13) {materialFunction=/FortniteBaseMaterials/Functions/MF_QualitySwitch_Vector}.Result
   Alpha=@n15.B
  ) {constB=800}
  n17 = LinearInterpolate(A=@n6.EmissiveColor, B=Multiply(A=Multiply(A=@n7, B=@n8), B=Add(A=@n11, B=@n16)), Alpha=NamedRerouteUsage())
  n18 = Constant() {r=1}
  n19 = OneMinus(None=NamedRerouteUsage())
  n20 = Power(Base=@n19) {constExponent=8}
  n21 = SphereMask(A=NamedRerouteUsage(), B=Constant3Vector() {constant=<rgba 0 3 155 0>}) {attenuationRadius=6, hardnessPercent=0}
  n22 = VertexInterpolator(VS=@n21)
  n23 = Add(A=@n20, B=Step(Y=ShaderStageSwitch(PixelShader=@n21, VertexShader=@n22.PS)) {constX=0.1, constY=0.1})
  n24 = CollectionParameter() {collection=&MPC_Toyfox_SequenceControl, group=Toyfox, parameterName=VisorAlpha}
  n25 = StaticSwitchParameter(
   True=LinearInterpolate(
    A=MaterialFunctionCall(Low=@n20, High=@n20, Epic=@n20, Mobile=@n23, Medium=@n23) {materialFunction=&MF_QualitySwitch_Vector}.Result
    Alpha=LinearInterpolate(Alpha=@n24) {constA=-0.1}
   )
   False=Constant() {r=1}
  ) {group=Toyfox, parameterName=IsHead}
  n26 = StaticBoolParameter() {group=ToyFox, parameterName=Is Outro, sortPriority=0}
  n27 = GetMaterialAttributes(None=SetMaterialAttributes!(
    MaterialAttributes=@n6.MaterialAttributes
    EmissiveColor=StaticSwitchParameter(True=@n17, False=@n6.EmissiveColor) {group=Toyfox, parameterName=IsHead}
    OpacityMask=StaticSwitch(True=@n18, False=@n25, Value=@n26)
   ) {attributeSetTypes [#8, #9]}) {attributeGetTypes [
    #1
    #4
    #5
    #8
   ]}
  n28 = NamedRerouteUsage()
  n29 = Multiply(A=NamedRerouteUsage(), B=Multiply(A=Power(Base=@n27.Roughness) {constExponent=0.25}) {constB=4})
  n30 = SetMaterialAttributes!(
   MaterialAttributes=@n27.MaterialAttributes
   BaseColor=Multiply(A=@n27.BaseColor, B=OneMinus(None=@n28))
   Roughness=LinearInterpolate(A=@n27.Roughness, B=Saturate(None=@n29), Alpha=@n28)
   Specular=Multiply(A=@n27.Specular, B=OneMinus(None=@n28))
   EmissiveColor=Add(A=@n27.EmissiveColor, B=NamedRerouteUsage())
  ) {attributeSetTypes [
    #1
    #4
    #5
    #8
   ]}
  FunctionOutput(None=@n30) {outputName=OUT MA}
  FunctionOutput(None=NamedRerouteUsage()) {outputName=MetalNoise, sortPriority=1}
  FunctionOutput(None=@n25) {outputName=VisorMask, sortPriority=2}
  Multiply(A=@n19) {constB=10}
  n31 = NamedRerouteUsage()
  n32 = Constant3Vector() {constant=<rgba 23.5 10 161.4 0>}
  n33 = Constant3Vector() {constant=<rgba -1 1 1 0>}
  n34 = Multiply(A=@n32, B=@n33) {constB=-1}
  n35 = NamedRerouteUsage()
  n36 = NamedRerouteDeclaration(None=Step(Y=ComponentMask(None=@n35) {r=true}) {constX=0}) {name=Mirror_X_Axis_Msk, nodeColor=<rgba 0 0.1562 1 1>}
  NamedRerouteDeclaration(None=MaterialFunctionCall!(A=@n31, B=LinearInterpolate(A=@n32, B=@n34, Alpha=@n36), Bounds=Constant3Vector() {constant=<rgba 7 27 6 0>}, Edge Falloff) {materialFunction=/Engine/Functions/Engine_MaterialFunctions01/Gradient/BoxMask-3D}.Result) {name=CannonMask, nodeColor=<rgba 1 0.6563 0 1>}
  n37 = NamedRerouteUsage()
  n38 = Multiply(
   A=Multiply(A=ScalarParameter() {defaultValue=0.6, group=Toyfox, parameterName=Cannon_HeatAmount}, B=NamedRerouteUsage())
   B=OneMinus(None=NamedRerouteUsage())
  )
  n39 = Constant() {r=0.25}
  n40 = Saturate(None=MaterialFunctionCall!(Mask In=ComponentMask(None=@n37) {g=true}, Amount=OneMinus(None=@n38), Range=@n39, Start Offset, Sharpness=@n39) {materialFunction=/Game/Global/MaterialFunctions/MF_MaskShifting}.Result)
  n41 = Multiply(A=NamedRerouteUsage()) {constB=0.5}
  n42 = Multiply(A=Add(A=@n40, B=Multiply(A=@n40, B=@n41)), B=NamedRerouteUsage())
  n43 = TextureSample(UVs=Clamp(None=@n42) {minDefault=0.01}) {
   samplerSource=SSM_Clamp_WorldGroupSettings
   samplerType=SAMPLERTYPE_LinearColor
   texture=/ToyFoxGameplay/Assets/Texture/Curves/CA_ToyFox_CannonHeat
  }
  n44 = Multiply(A=ScalarParameter() {defaultValue=2, group=Toyfox, parameterName=Cannon_HeatIntensity}, B=NamedRerouteUsage())
  n45 = OneMinus(None=NamedRerouteUsage())
  NamedRerouteDeclaration(None=Multiply(A=@n43.RGB, B=Multiply(A=@n44, B=@n45))) {name=CannonHeatEmissive, nodeColor=<rgba 1 0.1711 0.0104 1>}
  n46 = FunctionInput() {inputName=IN TAMsk, sortPriority=1}
  NamedRerouteDeclaration(None=ComponentMask(None=@n46) {b=true}) {name=VisorMask, nodeColor=<rgba 0 1 0.7891 1>}
  n47 = CameraPositionWS()
  n48 = WorldPosition()
  n49 = Divide(
   A=Length(None=Subtract(A=@n47, B=@n48.XYZ)) {is=class MaterialExpressionLength}
   B=Multiply(A=ScalarParameter() {defaultValue=24, group=Toyfox, parameterName=DistanceStartFade}) {constB=100}
  )
  NamedRerouteDeclaration(None=Saturate(None=Power(Base=@n49) {constExponent=4})) {name=DistToFade, nodeColor=<rgba 0.4531 1 0 1>}
  NamedRerouteDeclaration(None=@n46) {name=TAMsk, nodeColor=<rgba 0 1 0.8125 1>}
  n50 = OneMinus(None=ComponentMask(None=@n37) {b=true})
  n51 = Constant4Vector() {constant=<rgba 0.5 0.1 0 0.1>}
  NamedRerouteDeclaration(None=MaterialFunctionCall(Mask In=@n50, Amount=@n51.R, Range=@n51.R, Start Offset=@n51.R, Sharpness=@n51.R) {materialFunction=&MF_MaskShifting}.Result) {name=Bounding_Leg_Msk, nodeColor=<rgba 0 1 0.6016 1>}
  NamedRerouteDeclaration(None=@n2.Normal) {name=Original_N, nodeColor=<rgba 1 0.9375 0 1>}
  n52 = ScalarParameter() {group=Toyfox, parameterName=Cannon_CurrentStage, sortPriority=3}
  n53 = LinearInterpolate(A=Subtract(A=@n52), B=@n52, Alpha=NamedRerouteUsage()) {desc=Native lerp toward}
  n54 = Constant4Vector() {desc=Stage 1 to 2, constant=<rgba 1 2 0 1>}
  NamedRerouteDeclaration(None=Saturate(None=MaterialFunctionCall(Input=@n53, Input Low=@n54.R, Input High=@n54.R, Target Low=@n54.R, Target High=@n54.R) {materialFunction=&RemapValueRange}.Result)) {name=Cannon_Stage2_Lerp, nodeColor=<rgba 0.1893 1 0 1>}
  n55 = Add(A=NamedRerouteUsage(), B=Constant3Vector() {constant=<rgba 0 0 1 0>})
  n56 = TextureCoordinate()
  n57 = ScalarParameter() {defaultValue=40, group=DetailEnhance, parameterName=DetailNormalTiling}
  n58 = MaterialFunctionCall(
   Normal=TextureSample(UVs=Multiply(A=@n56, B=@n57) {constB=20}, Tex=NamedRerouteUsage()) {samplerSource=SSM_Wrap_WorldGroupSettings, samplerType=SAMPLERTYPE_LinearColor}.RGB
   Flatness=Constant() {r=0.85}
  ) {materialFunction=&FlattenNormal}
  n59 = Multiply(A=@n58.Result, B=Constant3Vector() {constant=<rgba -1 -1 1 0>})
  n60 = Multiply(A=@n55, B=DotProduct(A=@n55, B=@n59))
  n61 = ComponentMask(None=@n55) {b=true}
  NamedRerouteDeclaration(None=Subtract(A=Divide(A=@n60, B=@n61), B=@n59)) {name=Combined_DN, nodeColor=<rgba 0.25 0.25 1 1>}
  NamedRerouteDeclaration(None=@n2.Metallic) {name=Original_Metal, nodeColor=<rgba 0 1 0.2969 1>}
  NamedRerouteDeclaration(None=FunctionInput() {bUsePreviewValueAsDefault=true, inputName=IN Bounding, sortPriority=1}) {name=IN Bounding, nodeColor=<rgba 1 0.75 0 1>}
  n62 = NamedRerouteUsage()
  n63 = NamedRerouteUsage()
  NamedRerouteDeclaration(None=TextureSample(UVs=Multiply(A=@n62) {constB=20}, Tex=@n63) {samplerSource=SSM_Wrap_WorldGroupSettings}.A) {name=MetalFlakNoise, nodeColor=<rgba 0 1 0.1562 1>}
  NamedRerouteDeclaration(None=@n42) {name=Cannon_HeatMsk, nodeColor=<rgba 1 0.3092 0 1>}
  NamedRerouteDeclaration(None=@n2.Roughness) {name=Original_Roughness, nodeColor=<rgba 0.7578 1 0 1>}
  n64 = MaterialFunctionCall(Float3=RgbToHsv!(None=NamedRerouteUsage())) {materialFunction=&BreakOutFloat3Components}
  n65 = NamedRerouteUsage()
  n66 = Multiply(
   A=@n64.G
   B=LinearInterpolate(A=ScalarParameter() {defaultValue=0.95, group=DetailEnhance, parameterName=BaseColor_Saturation_Var}, Alpha=@n65) {constA=1}
  ) {constB=0.85}
  n67 = Multiply(
   A=@n64.B
   B=LinearInterpolate(A=ScalarParameter() {defaultValue=0.8, group=DetailEnhance, parameterName=BaseColor_Value_Var}, Alpha=@n65) {constA=1}
  ) {constB=0}
  NamedRerouteDeclaration(None=HsvToRgb!(None=MaterialFunctionCall(X=@n64.R, Y=@n66, Z=@n67) {materialFunction=&MakeFloat3}.Result)) {name=BC_Variation, nodeColor=<rgba 0.3125 1 0 1>}
  NamedRerouteDeclaration(None=@n2.BaseColor) {name=Original_BC, nodeColor=<rgba 1 0 0.6328 1>}
  NamedRerouteDeclaration(None=@n58.Result) {name=ClearCoatBottomNormal, nodeColor=<rgba 0 0.9531 1 1>}
  n68 = NamedRerouteUsage()
  n69 = Multiply(A=Multiply(A=ComponentMask(None=@n68) {b=true}, B=NamedRerouteUsage()), B=Saturate(None=NamedRerouteUsage()))
  NamedRerouteDeclaration(None=Multiply(
    A=SmoothStep(Value=@n69) {constMax=0.25}
    B=ScalarParameter() {defaultValue=0.68, group=DetailEnhance, parameterName=DirtReveal, sliderMax=1}
   )) {name=DirtMask, nodeColor=<rgba 0.5469 1 0 1>}
  n70 = NamedRerouteUsage()
  n71 = NamedRerouteUsage()
  n72 = Saturate(None=OneMinus(None=NamedRerouteUsage()))
  n73 = VectorParameter() {group=DetailEnhance, parameterName=DirtColor, defaultValue=<rgba 0.354 0.292 0.173 1>}
  n74 = NamedRerouteUsage()
  NamedRerouteDeclaration(None=LinearInterpolate(A=LinearInterpolate(A=@n70, B=@n71, Alpha=@n72), B=Multiply(A=@n73.RGB, B=@n74), Alpha=@n74)) {name=DetialEnhanced_BC, nodeColor=<rgba 1 0 0.8672 1>}
  n75 = NamedRerouteUsage()
  n76 = Add(A=@n75, B=NamedRerouteUsage())
  n77 = NamedRerouteUsage()
  n78 = ConstantBiasScale(None=Power(Base=Add(A=@n76, B=@n77)) {constExponent=0.28}) {bias=-0.5, scale=4}
  n79 = NamedRerouteUsage()
  n80 = Multiply(A=Multiply(A=NamedRerouteUsage()) {constB=0.5}, B=OneMinus(None=NamedRerouteUsage()))
  n81 = Saturate(None=Max(A=NamedRerouteUsage(), B=Max(A=@n79, B=@n80)))
  NamedRerouteDeclaration(None=LinearInterpolate(A=@n75, B=Saturate(None=@n78), Alpha=@n81)) {name=DetialEnhanced_Roughness, nodeColor=<rgba 1 0 0.8672 1>}
  NamedRerouteDeclaration(None=LinearInterpolate(A=NamedRerouteUsage(), B=NamedRerouteUsage(), Alpha=NamedRerouteUsage())) {name=DetialEnhanced_Normal, nodeColor=<rgba 0.25 0.25 1 1>}
  n82 = TextureSampleParameter2D() {parameterName=SRM, samplerSource=SSM_Wrap_WorldGroupSettings, texture=/Game/Global/Textures/Default/Blanks/T_MidgreysRGB}
  n83 = Multiply(A=SmoothStep(Value=@n82.G) {constMin=0.5}, B=OneMinus(None=@n82.B))
  n84 = NamedRerouteUsage()
  n85 = OneMinus(None=SmoothStep(Value=@n82.R) {constMax=0.3})
  n86 = NamedRerouteUsage()
  n87 = ComponentMask(None=NamedRerouteUsage()) {r=true}
  n88 = LinearInterpolate(A=Max(A=Subtract(A=@n85, B=@n86), B=Multiply(A=@n82.B, B=@n87)), Alpha=Multiply(A=NamedRerouteUsage()) {constB=10}) {constB=0}
  NamedRerouteDeclaration(None=Max(A=Multiply(A=@n83, B=@n84), B=Saturate(None=@n88))) {name=DentMsk, nodeColor=<rgba 0.9453 1 0 1>}
  NamedRerouteDeclaration(None=@n2.Specular) {name=Original_Specular, nodeColor=<rgba 1 0.5391 0 1>}
  n89 = LocalPosition() {localOrigin=InstancePreSkinning}
  NamedRerouteDeclaration(None=FunctionInput(Preview=VertexInterpolator(VS=@n89.XYZ).PS) {inputName=IN PreSkinnedLoc, sortPriority=3}) {name=PreskinnedLoc, nodeColor=<rgba 0.5158 0 1 1>}
  n90 = NamedRerouteUsage()
  n91 = Multiply(A=NamedRerouteUsage()) {constB=0.5}
  NamedRerouteDeclaration(None=LinearInterpolate(A=@n90, B=Multiply(A=@n90, B=@n91), Alpha=@n81) {constB=0.2}) {name=DetialEnhanced_Specular, nodeColor=<rgba 0.7812 1 0 1>}
  NamedRerouteDeclaration(None=MaterialFunctionCall(A=@n31, B=Constant3Vector() {constant=<rgba 0 1.5 165.5 1>}, Bounds=Constant3Vector() {constant=<rgba 0.65 6.5 8.5 1>}) {materialFunction=&BoxMask-3D}.Result) {name=HornMask, nodeColor=<rgba 0.9922 0 1 1>}
  n92 = Constant3Vector() {constant=<rgba 2.8 4.486 156.5 0>}
  n93 = SphereMask(A=@n31, B=LinearInterpolate(A=@n92, B=Multiply(A=@n92, B=@n33) {constB=-1}, Alpha=@n36)) {attenuationRadius=3, hardnessPercent=1}
  n94 = ShaderStageSwitch(PixelShader=@n93, VertexShader=VertexInterpolator(VS=@n93).PS)
  NamedRerouteDeclaration(None=Multiply(A=Power(Base=@n94)) {constB=2, desc=EyeLightMask}) {name=EyeLIghtSpotMask, nodeColor=<rgba 1 0.6328 0 1>}
  NamedRerouteDeclaration(None=TextureObject() {texture=/ToyFoxGameplay/Assets/Texture/T_ToyFox_TA_Mask}) {name=TA_Noise_Mask, nodeColor=<rgba 1 0.9844 0 1>}
  NamedRerouteDeclaration(None=TextureObject() {desc=DN, samplerType=SAMPLERTYPE_Normal, texture=/ToyFoxGameplay/Assets/Texture/T_ToyFox_MetallicFlake_N}) {name=DetialNormal, nodeColor=<rgba 0.6641 0 1 1>}
  NamedRerouteDeclaration(None=Saturate(None=@n53)) {name=Cannon_Stage1_Lerp, nodeColor=<rgba 0.1893 1 0 1>}
  NamedRerouteDeclaration(None=@n50) {name=Bounding_Y, nodeColor=<rgba 1 0.0234 0 1>}
  n95 = Subtract(A=Time(), B=ScalarParameter() {defaultValue=-99999, group=Toyfox, parameterName=Cannon_GameTimeFromBP})
  n96 = Add(A=ScalarParameter() {defaultValue=1, group=Toyfox, parameterName=Cannon_AnimDuration}) {constB=0}
  NamedRerouteDeclaration(None=Saturate(None=Divide(A=@n95, B=@n96))) {name=Cannon_TimeNative, nodeColor=<rgba 0.2891 1 0 1>}
  n97 = Subtract(A=Time(), B=ScalarParameter() {defaultValue=-99999, group=Toyfox, parameterName=Eye_GameTimeFromBP})
  n98 = Add(A=ScalarParameter() {defaultValue=1, group=Toyfox, parameterName=Eye_AnimDuration}) {constB=0}
  NamedRerouteDeclaration(None=Saturate(None=Divide(A=@n97, B=@n98))) {name=Eye_TimeNative, nodeColor=<rgba 0.2891 1 0 1>}
  n99 = ScalarParameter() {group=Toyfox, parameterName=Eye_CurrentStage, sortPriority=3}
  n100 = LinearInterpolate(A=Subtract(A=@n99), B=@n99, Alpha=NamedRerouteUsage()) {desc=Native lerp toward}
  NamedRerouteDeclaration(None=Saturate(None=@n100)) {name=Eye_Stage1_Lerp, nodeColor=<rgba 0.1893 1 0 1>}
  NamedRerouteDeclaration(None=FunctionInput() {inputName=IN TriplannerUV, inputType=FunctionInput_Vector2, sortPriority=1}) {name=HardTriplannarUV, nodeColor=<rgba 0.0312 0 1 1>}
  n101 = Constant4Vector() {desc=Stage 1 to 2, constant=<rgba 1 2 0 1>}
  NamedRerouteDeclaration(None=Saturate(None=MaterialFunctionCall(Input=@n100, Input Low=@n101.R, Input High=@n101.R, Target Low=@n101.R, Target High=@n101.R) {materialFunction=&RemapValueRange}.Result)) {name=Eye_Stage2_Lerp, nodeColor=<rgba 0.1893 1 0 1>}
  NamedRerouteDeclaration(None=TextureSample(UVs=Multiply(A=@n62) {constB=4}, Tex=@n63) {samplerSource=SSM_Wrap_WorldGroupSettings}.B) {name=MetalNoise, nodeColor=<rgba 0 0.5781 1 1>}
  n102 = PreSkinnedNormal()
  NamedRerouteDeclaration(None=FunctionInput(Preview=VertexInterpolator(VS=@n102).PS) {inputName=IN PreSkinnedNormal, sortPriority=4}) {name=PreSkinnedNormal, nodeColor=<rgba 0 0.9297 1 1>}
 ]}
