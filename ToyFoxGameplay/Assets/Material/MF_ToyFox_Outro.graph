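# Material function for the ToyFox outro, which erodes the character away with a glowing edge when Is Outro is on.
# It blends weakpoint into original attributes, then adds edge emissive and dithered, visor-masked opacity.
# Bounding gradients and animated sine noise feed two MF_MaskShifting calls that form the dissolve and edge band.
# Stage and game-time parameters supply progress; skylight reflection, with a Mobile/Medium fallback, and panning curve-atlas beams build a color term.
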MaterialFunction /ToyFoxGameplay/Assets/Material/MF_ToyFox_Outro
graph {nodes [
  n1 = BlendMaterialAttributes(
   A=FunctionInput() {inputName=IN WeakpointVisual MA, inputType=FunctionInput_MaterialAttributes}
   B=FunctionInput() {inputName=IN OriginalVisual MA, inputType=FunctionInput_MaterialAttributes}
   Alpha=NamedRerouteUsage() {desc=Clear Weakpoint and LowLife Apperence}
  )
  n2 = GetMaterialAttributes(None=@n1) {attributeGetTypes [#1]}
  n3 = FunctionInput() {bUsePreviewValueAsDefault=true, inputName=IN Bounding, sortPriority=1}
  n4 = OneMinus(None=ComponentMask(None=@n3) {b=true})
  n5 = ComponentMask(None=@n3) {r=true}
  n6 = Min(A=OneMinus(None=@n5), B=@n5)
  n7 = Saturate(None=Multiply(A=Power(Base=@n6) {constExponent=4}) {constB=20})
  n8 = Constant3Vector() {desc=Random, constant=<rgba -0.3864 0.6 0.4464 0>}
  n9 = Multiply(A=Time(), B=ScalarParameter() {defaultValue=2, group=ToyFoxOutro, parameterName=ErosionNoise_Frequence, sortPriority=1})
  n10 = MaterialFunctionCall(Float3=Multiply(A=LocalPosition().XYZ, B=VectorParameter() {group=ToyFoxOutro, parameterName=ErosionNoise_Size, defaultValue=<rgba 1 1 1 1>}.RGB)) {materialFunction=&BreakOutFloat3Components}
  n11 = Sine(None=Multiply(A=@n8.R, B=Add(A=@n9, B=@n10.R))) {period=6.283}
  n12 = Sine(None=Multiply(A=@n8.G, B=Add(A=@n9, B=@n10.G))) {period=6.283}
  n13 = Sine(None=Multiply(A=@n8.B, B=Add(A=@n9, B=@n10.B))) {period=6.283}
  n14 = MaterialFunctionCall(Float3=Multiply(
    A=MaterialFunctionCall(X=@n11, Y=@n12, Z=@n13) {materialFunction=&MakeFloat3}.Result
    B=ScalarParameter() {defaultValue=0.1, group=ToyFoxOutro, parameterName=ErosionNoise_Amplitude, sortPriority=4}
   ) {desc=wind factor}) {materialFunction=&BreakOutFloat3Components}
  n15 = Multiply(
   A=Add(A=Add(A=@n14.R, B=@n14.R), B=@n14.B)
   B=ScalarParameter() {defaultValue=1, group=ToyFoxOutro, parameterName=ErosionNoise_Strength, sortPriority=3}
  )
  n16 = VertexInterpolator(VS=@n15)
  n17 = Subtract(A=@n4, B=Add(A=@n7, B=ShaderStageSwitch(PixelShader=@n16.PS, VertexShader=@n15)))
  n18 = VectorParameter() {group=ToyFoxOutro, parameterName=OutroMask, defaultValue=<rgba 0.5 0.5 0 0.1>}
  n19 = MaterialFunctionCall!(
   Mask In=Add(A=NamedRerouteUsage(), B=Add(A=@n17, B=@n4))
   Amount=LinearInterpolate(Alpha=NamedRerouteUsage()) {constA=-2.5, constB=2.5}
   Range=@n18.G
   Start Offset=@n18.G
   Sharpness=@n18.G
  ) {materialFunction=/Game/Global/MaterialFunctions/MF_MaskShifting}
  n20 = VectorParameter() {group=ToyFoxOutro, parameterName=OutroEdgeMask, defaultValue=<rgba 0.05 0.1 0 1>}
  n21 = MaterialFunctionCall(Mask In=@n19.Result, Amount=@n20.R, Range=@n20.R, Start Offset=@n20.R, Sharpness=@n20.R) {materialFunction=&MF_MaskShifting}
  n22 = Saturate(None=Add(A=Subtract(A=@n19.Result, B=@n21.Result), B=NamedRerouteUsage()))
  n23 = Add(A=@n22, B=Power(Base=@n22) {constExponent=4})
  n24 = NamedRerouteUsage()
  n25 = Add(
   A=@n2.EmissiveColor
   B=Multiply(A=Multiply(A=@n23, B=@n24), B=ScalarParameter() {defaultValue=1, group=ToyFoxOutro, parameterName=ErosionEdge_Intensity})
  )
  n26 = MaterialFunctionCall(Alpha Threshold=Multiply(A=FunctionInput() {inputName=IN VisorMask, inputType=FunctionInput_Scalar, sortPriority=1}, B=Saturate(None=@n19.Result))) {materialFunction=&DitherTemporalAA_Fortnite}
  FunctionOutput(None=StaticSwitch(
    True=SetMaterialAttributes!(MaterialAttributes=@n2.MaterialAttributes, EmissiveColor=@n25, OpacityMask=@n26.Result) {attributeSetTypes [#1, #2]}
    False=@n1
    Value=StaticBoolParameter() {group=ToyFox, parameterName=Is Outro, sortPriority=0}
   ))
  NamedRerouteDeclaration(None=FunctionInput() {inputName=IN TriplannarUV, inputType=FunctionInput_Scalar, sortPriority=1}) {name=HardTriplannar, nodeColor=<rgba 0.7578 0 1 1>}
  n27 = DotProduct(A=VertexNormalWS(), B=CameraVectorWS())
  NamedRerouteDeclaration(None=Power(Base=OneMinus(None=@n27))) {name=ToyFoxFresnel, nodeColor=<rgba 0 0.1797 1 1>}
  NamedRerouteDeclaration(None=ScalarParameter() {defaultValue=1, group=ToyFoxOutro, parameterName=Outro_Animation_Duration, sortPriority=101}) {name=Duration, nodeColor=<rgba 0.6641 1 0 1>}
  NamedRerouteDeclaration(None=@n4) {name=IN Bounding Y, nodeColor=<rgba 0.2188 0 1 1>}
  n28 = CameraVectorWS()
  n29 = VertexNormalWS()
  n30 = Constant() {r=0.5}
  n31 = SkyLightEnvMapSample(
   Direction=Add(A=Multiply(A=@n28) {constB=-1}, B=Multiply(A=@n29, B=@n30))
   Roughness=ScalarParameter() {defaultValue=0.2, group=ToyFoxOutro, parameterName=SkylightEnvMapRoughness}
  )
  n32 = Constant3Vector() {constant=<rgba 1 0.8717 0.505 1>}
  n33 = Multiply(A=Multiply(A=@n7) {constB=3}, B=@n32)
  n34 = NamedRerouteUsage()
  n35 = Add(
   A=MaterialFunctionCall!(Low=@n31, High=@n31, Epic=@n31, --------, Mobile=@n33, Medium=@n33) {materialFunction=/FortniteBaseMaterials/Functions/MF_QualitySwitch_Vector}.Result
   B=@n34
  )
  n36 = NamedRerouteUsage()
  n37 = Multiply(A=@n36, B=Constant2Vector() {g=0.1, r=1})
  n38 = CurveAtlasRowParameter(CurveTime=TextureSample(UVs=Panner(Coordinate=@n37) {speedX=0.25, speedY=0.5}) {samplerSource=SSM_Wrap_WorldGroupSettings, texture=/ToyFoxGameplay/Assets/Texture/T_ToyFox_TA_Mask}.R) {
   atlas=/ToyFoxGameplay/Assets/FX/CA_ToyFox_VFX
   curve=/ToyFoxGameplay/Assets/FX/Curve_ToyFox_Color03
   defaultValue=2
   group=ToyFoxOutro
   parameterName=OutroBeamCurve01
  }
  n39 = OneMinus(None=NamedRerouteUsage())
  n40 = Constant2Vector() {g=0.01, r=0.4}
  n41 = TextureSample(UVs=Panner(Coordinate=Multiply(A=@n36, B=@n40)) {speedX=0.05, speedY=0.1}) {samplerSource=SSM_Wrap_WorldGroupSettings, texture=&T_ToyFox_TA_Mask}
  n42 = Max(A=@n41.G, B=@n41.G)
  n43 = Multiply(A=@n42, B=OneMinus(None=@n42))
  n44 = Divide(A=@n43, B=LinearInterpolate(Alpha=NamedRerouteUsage()) {constA=4, constB=10})
  n45 = Multiply(
   A=Multiply(A=Multiply(A=@n38, B=@n32), B=ScalarParameter() {defaultValue=4, group=ToyFoxOutro, parameterName=OutroBeam01_Intensity})
   B=Multiply(A=Add(A=@n39, B=@n44), B=ScalarParameter() {defaultValue=2, group=ToyFoxOutro, parameterName=OutroBeam02_Intensity})
  )
  n46 = VectorParameter() {group=ToyFoxOutro, parameterName=ErosionEdge_Color_Light, defaultValue=<rgba 0.1609 0.4845 0.776 1>}
  n47 = Power(Base=@n34, Exp=LinearInterpolate(Alpha=NamedRerouteUsage()) {constA=4, constB=0.5}) {constExponent=0.5}
  n48 = Multiply(A=Multiply(A=NamedRerouteUsage(), B=LinearInterpolate(A=@n45, B=@n46.RGB, Alpha=@n47))) {constB=0.5}
  NamedRerouteDeclaration(None=Saturate(None=Multiply(A=@n35, B=@n48))) {name=Cubemap, nodeColor=<rgba 1 0 0.7266 1>}
  NamedRerouteDeclaration(None=MaterialFunctionCall(X=Add(A=@n14.G, B=@n14.G), Y=Add(A=@n14.R, B=@n14.R), Z=Add(A=@n14.R, B=@n14.R)) {materialFunction=&MakeFloat3}.Result) {name=WaveNoise, nodeColor=<rgba 0 1 0.3203 1>}
  n49 = ScalarParameter() {group=ToyFoxOutro, parameterName=Outro_Stage}
  n50 = LinearInterpolate(A=Subtract(A=@n49), B=@n49, Alpha=NamedRerouteUsage()) {constB=20}
  n51 = Constant4Vector() {constant=<rgba 1 2 0 1>}
  NamedRerouteDeclaration(None=Saturate(None=MaterialFunctionCall(Input=@n50, Input Low=@n51.R, Input High=@n51.R, Target Low=@n51.R, Target High=@n51.R) {materialFunction=&RemapValueRange}.Result)) {name=Stages02, nodeColor=<rgba 0.3812 1 0.0069 1>}
  NamedRerouteDeclaration(None=FunctionInput() {inputName=IN PreSkinnedLoc, sortPriority=3}) {name=PreskinnedLoc, nodeColor=<rgba 0.5158 0 1 1>}
  NamedRerouteDeclaration(None=@n43) {name=StripNoise, nodeColor=<rgba 1 0.2812 0 1>}
  n52 = Subtract(A=Time(), B=ScalarParameter() {defaultValue=-99999, group=ToyFoxOutro, parameterName=Outro_GameTimeFromBP})
  n53 = Add(A=NamedRerouteUsage()) {constB=0}
  NamedRerouteDeclaration(None=Saturate(None=Divide(A=@n52, B=@n53))) {name=TimeNative, nodeColor=<rgba 0.2891 1 0 1>}
  NamedRerouteDeclaration(None=@n22) {name=BeamFillMask, nodeColor=<rgba 1 0.1172 0 1>}
  NamedRerouteDeclaration(None=Saturate(None=@n50)) {name=Stages01, nodeColor=<rgba 0.3812 1 0.0069 1>}
  NamedRerouteDeclaration(None=FunctionInput() {inputName=IN PreSkinnedNormal, sortPriority=4}) {name=PreSkinnedNormal, nodeColor=<rgba 0 0.9297 1 1>}
  ScalarParameter() {parameterName=FadeoutTest, sliderMax=1}
 ]}
