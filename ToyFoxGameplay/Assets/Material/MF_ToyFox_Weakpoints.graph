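# Material function that overlays eight animated weakpoints on the ToyFox character's incoming material attributes, adding corrupted, tempered-metal surface shading and emissive glow.
# Sphere masks at hardcoded pre-skinned body positions combine into one mask, broken up by triplanar noise.
# Per-point stage and start-time parameters drive spawn transitions and hit blinks through curve atlases, with similar timers for LF and PowerRegen states.
# Weakspot position custom-primitive-data parameters exist but are unconnected.
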MaterialFunction /ToyFoxGameplay/Assets/Material/MF_ToyFox_Weakpoints
graph {nodes [
  n1 = NamedRerouteUsage()
  n2 = ShaderStageSwitch(PixelShader=VertexInterpolator(VS=@n1).PS, VertexShader=@n1)
  n3 = NamedRerouteUsage()
  n4 = ShaderStageSwitch(PixelShader=VertexInterpolator(VS=@n3).PS, VertexShader=@n3)
  n5 = NamedRerouteUsage()
  n6 = ShaderStageSwitch(PixelShader=VertexInterpolator(VS=@n5).PS, VertexShader=@n5)
  n7 = NamedRerouteUsage()
  n8 = ShaderStageSwitch(PixelShader=VertexInterpolator(VS=@n7).PS, VertexShader=@n7)
  n9 = Max(A=Max(A=@n2, B=@n4), B=Max(A=@n6, B=@n8))
  n10 = NamedRerouteUsage()
  n11 = ShaderStageSwitch(PixelShader=VertexInterpolator(VS=@n10).PS, VertexShader=@n10)
  n12 = NamedRerouteUsage()
  n13 = ShaderStageSwitch(PixelShader=VertexInterpolator(VS=@n12).PS, VertexShader=@n12)
  n14 = NamedRerouteUsage()
  n15 = ShaderStageSwitch(PixelShader=VertexInterpolator(VS=@n14).PS, VertexShader=@n14)
  n16 = NamedRerouteUsage()
  n17 = ShaderStageSwitch(PixelShader=VertexInterpolator(VS=@n16).PS, VertexShader=@n16)
  n18 = Max(A=Max(A=@n11, B=@n13), B=Max(A=@n15, B=@n17))
  n19 = Constant4Vector() {constant=<rgba 0.2 0.1 0 0.5>}
  n20 = MaterialFunctionCall!(
   Mask In=NamedRerouteDeclaration(None=Max(A=@n9, B=@n18)) {name=Initial_Weakpooint_Mask, nodeColor=<rgba 1 0.4688 0 1>}
   Amount=@n19.R
   Range=@n19.R
   Start Offset=@n19.R
   Sharpness=@n19.R
  ) {materialFunction=/Game/Global/MaterialFunctions/MF_MaskShifting}
  Add(A=@n20.Result, B=Multiply(A=NamedRerouteUsage(), B=NamedRerouteUsage()))
  Constant() {r=4}
  n21 = Constant() {r=2}
  Fmod(A=Divide(A=Time(), B=@n21), B=@n21)
  Frac(None=Time())
  n22 = GetMaterialAttributes(None=FunctionInput() {inputType=FunctionInput_MaterialAttributes}) {attributeGetTypes [
    #1
    #2
    #3
    #4
    #5
    #6
    #7
   ]}
  n23 = NamedRerouteUsage()
  n24 = Desaturation(None=@n23)
  n25 = VectorParameter() {group=Weakpoints, parameterName=CorrupedArea_BC, defaultValue=<rgba 0.3739 0.1204 0.3854 1>}
  n26 = Multiply(A=@n24, B=Constant3Vector() {constant=<rgba 0.26 0.18 0.148 1>}) {constB=0.25}
  n27 = OneMinus(None=NamedRerouteUsage())
  n28 = NamedRerouteUsage()
  n29 = Multiply(A=Multiply(A=NamedRerouteUsage()) {constB=0.5}, B=@n28)
  n30 = Multiply(A=@n28) {constB=0.5}
  n31 = LinearInterpolate(
   A=@n23
   B=LinearInterpolate(A=Multiply(A=@n24, B=@n25.RGB), B=Multiply(A=@n26, B=@n27), Alpha=@n27)
   Alpha=Saturate(None=Add(A=@n29, B=@n30))
  ) {constB=0.2}
  n32 = NamedRerouteUsage()
  n33 = SmoothStep(Value=@n32) {constMax=0.25}
  n34 = OneMinus(None=@n33)
  n35 = Multiply(A=NamedRerouteUsage(), B=OneMinus(None=NamedRerouteUsage()))
  n36 = Multiply(A=Add(A=NamedRerouteUsage(), B=Multiply(A=@n34, B=@n35)), B=@n34)
  n37 = Constant4Vector() {constant=<rgba 0 1.4 1 0.9>}
  n38 = Multiply(
   A=@n33
   B=MaterialFunctionCall(Mask In=@n32, Amount=@n37.R, Range=@n37.R, Start Offset=@n37.R, Sharpness=@n37.R) {materialFunction=&MF_MaskShifting}.Result
  )
  n39 = NamedRerouteUsage()
  n40 = Saturate(None=Max(A=Multiply(A=@n36) {constB=0.54}, B=Subtract(A=@n38, B=@n39)))
  n41 = Power(Base=@n40) {constExponent=0.8}
  n42 = TextureSample(UVs=Clamp(None=ComponentMask(None=@n41) {r=true}) {minDefault=0.01}) {
   samplerSource=SSM_Clamp_WorldGroupSettings
   samplerType=SAMPLERTYPE_LinearColor
   texture=/ToyFoxGameplay/Assets/Texture/Curves/CA_ToyFox_TemperedMetal
  }
  n43 = MaterialFunctionCall(Float3=RgbToHsv!(None=@n42.RGB)) {materialFunction=&BreakOutFloat3Components}
  n44 = Multiply(A=@n43.G, B=ScalarParameter() {defaultValue=0.5, group=Weakpoints, parameterName=MetalColor_Sat_Percentage, sliderMax=1}) {constB=0.5}
  n45 = MaterialFunctionCall(
   Base=@n31
   Blend=HsvToRgb!(None=MaterialFunctionCall(X=@n43.R, Y=@n44, Z=@n43.R) {materialFunction=&MakeFloat3}.Result)
  ) {materialFunction=&Blend_Overlay}
  n46 = LinearInterpolate(Alpha=@n33) {constA=0.9, constB=0.45}
  n47 = NamedRerouteUsage()
  n48 = NamedRerouteUsage()
  n49 = NamedRerouteUsage()
  n50 = NamedRerouteUsage()
  n51 = Add(A=NamedRerouteUsage(), B=Multiply(A=NamedRerouteUsage()) {constB=2}) {constB=2}
  n52 = SetMaterialAttributes!(
   MaterialAttributes=@n22.MaterialAttributes
   BaseColor=LinearInterpolate(A=@n31, B=Multiply(A=@n45.Result, B=@n46) {constB=0.9}, Alpha=Multiply(A=@n40, B=@n47))
   Metallic=NamedRerouteUsage()
   Specular=LinearInterpolate(A=NamedRerouteUsage(), B=Multiply(A=@n48) {constB=0.65}, Alpha=@n49) {constB=0.65}
   Roughness=LinearInterpolate(
    A=@n50
    B=MaterialFunctionCall(Base=@n50, Blend=@n51) {materialFunction=&Blend_Overlay}.Result
    Alpha=@n49
   ) {constB=0.8}
   EmissiveColor=NamedRerouteUsage()
   Normal=LinearInterpolate(A=NamedRerouteUsage(), B=NamedRerouteUsage(), Alpha=@n49)
   CustomData0=Multiply(A=@n22.CustomData0, B=OneMinus(None=@n49))
  ) {attributeSetTypes [
    #1
    #2
    #3
    #4
    #6
    #5
    #7
   ]}
  FunctionOutput(None=@n52)
  n53 = ScalarParameter() {group=Weakpoints, parameterName=PowerRegen_Stage}
  n54 = LinearInterpolate(A=Subtract(A=@n53), B=@n53, Alpha=NamedRerouteUsage()) {constB=20}
  NamedRerouteDeclaration(None=Saturate(None=@n54)) {name=PowerRegen_Stages_01, nodeColor=<rgba 0.3812 1 0.0069 1>}
  n55 = Constant4Vector() {desc=WeakpointComp_LowerLeg_L, constant=<rgba 18.2012 -2.0518 39.0069 22>}
  n56 = Saturate(None=NamedRerouteUsage())
  NamedRerouteDeclaration(None=SphereMask(
    A=LocalPosition() {localOrigin=InstancePreSkinning}.XYZ
    B=ComponentMask(None=@n55.RGBA) {b=true, g=true, r=true}
    Radius=Multiply(A=@n55.A, B=@n56)
   ) {attenuationRadius=20, hardnessPercent=0}) {name=Weakpoint03_Msk, nodeColor=<rgba 1 0 0.75 1>}
  n57 = Constant4Vector() {desc=WeakpointComp_Shoulder_R, constant=<rgba -30 -4 140 18>}
  n58 = Saturate(None=NamedRerouteUsage())
  NamedRerouteDeclaration(None=SphereMask(
    A=LocalPosition() {localOrigin=InstancePreSkinning}.XYZ
    B=ComponentMask(None=@n57.RGBA) {b=true, g=true, r=true}
    Radius=Multiply(A=@n57.A, B=@n58)
   ) {attenuationRadius=20, hardnessPercent=0}) {name=Weakpoint04_Msk, nodeColor=<rgba 1 0 0.75 1>}
  n59 = Constant4Vector() {desc=WeakpointComp_Shoulder_L, constant=<rgba 30 -4 140 18>}
  n60 = Saturate(None=NamedRerouteUsage())
  NamedRerouteDeclaration(None=SphereMask(
    A=LocalPosition() {localOrigin=InstancePreSkinning}.XYZ
    B=ComponentMask(None=@n59.RGBA) {b=true, g=true, r=true}
    Radius=Multiply(A=@n59.A, B=@n60)
   ) {attenuationRadius=20, hardnessPercent=0}) {name=Weakpoint05_Msk, nodeColor=<rgba 1 0 0.75 1>}
  n61 = Constant4Vector() {desc=WeakpointComp_Arm_L, constant=<rgba 57 0 110 20>}
  n62 = Saturate(None=NamedRerouteUsage())
  NamedRerouteDeclaration(None=SphereMask(
    A=LocalPosition() {localOrigin=InstancePreSkinning}.XYZ
    B=ComponentMask(None=@n61.RGBA) {b=true, g=true, r=true}
    Radius=Multiply(A=@n61.A, B=@n62)
   ) {attenuationRadius=20, hardnessPercent=0}) {name=Weakpoint06_Msk, nodeColor=<rgba 1 0 0.75 1>}
  n63 = Constant4Vector() {desc=WeakpointComp_Arm_R, constant=<rgba -57 0 110 20>}
  n64 = Saturate(None=NamedRerouteUsage())
  NamedRerouteDeclaration(None=SphereMask(
    A=LocalPosition() {localOrigin=InstancePreSkinning}.XYZ
    B=ComponentMask(None=@n63.RGBA) {b=true, g=true, r=true}
    Radius=Multiply(A=@n63.A, B=@n64)
   ) {attenuationRadius=20, hardnessPercent=0}) {name=Weakpoint07_Msk, nodeColor=<rgba 1 0 0.75 1>}
  n65 = Constant4Vector() {desc=WeakpointComp_Back, constant=<rgba 0 -30 135.4447 22>}
  n66 = Saturate(None=NamedRerouteUsage())
  NamedRerouteDeclaration(None=SphereMask(
    A=LocalPosition() {localOrigin=InstancePreSkinning}.XYZ
    B=ComponentMask(None=@n65.RGBA) {b=true, g=true, r=true}
    Radius=Multiply(A=@n65.A, B=@n66)
   ) {attenuationRadius=20, hardnessPercent=0}) {name=Weakpoint08_Msk, nodeColor=<rgba 1 0 0.75 1>}
  n67 = Sine(None=Divide(A=Time()))
  NamedRerouteDeclaration(None=Multiply(A=Subtract(A=@n67)) {constB=0.02}) {name=WeakpointImpulse, nodeColor=<rgba 1 0.8438 0 1>}
  n68 = NamedRerouteUsage()
  n69 = Divide(A=Time() {period=1}) {constB=4}
  n70 = Constant4Vector() {constant=<rgba 0 0.05 0 0.1>}
  n71 = Sine(None=MaterialFunctionCall(Mask In=@n68, Amount=OneMinus(None=@n69), Range=@n70.G, Start Offset=@n70.G, Sharpness=@n70.G) {materialFunction=&MF_MaskShifting}.Result)
  n72 = SmoothStep(Value=Multiply(A=ConstantBiasScale(None=@n71), B=@n68)) {constMax=0.75}
  n73 = NamedRerouteUsage()
  NamedRerouteDeclaration(None=Saturate(None=Multiply(A=@n72, B=@n73))) {name=Weakpoint_Impuse, nodeColor=<rgba 1 0 0.3516 1>}
  n74 = NamedRerouteUsage()
  n75 = Max(
   A=Multiply(A=NamedRerouteUsage(), B=Step(X=@n2) {constX=0.1, constY=0.1})
   B=Multiply(A=NamedRerouteUsage(), B=Step(X=@n4) {constX=0.1, constY=0.1})
  )
  n76 = Max(
   A=Multiply(A=NamedRerouteUsage(), B=Step(X=@n6) {constX=0.1, constY=0.1})
   B=Multiply(A=NamedRerouteUsage(), B=Step(X=@n8) {constX=0.1, constY=0.1})
  )
  n77 = Max(
   A=Multiply(A=NamedRerouteUsage(), B=Step(X=@n11) {constX=0.1, constY=0.1})
   B=Multiply(A=NamedRerouteUsage(), B=Step(X=@n13) {constX=0.1, constY=0.1})
  )
  n78 = Max(
   A=Multiply(A=NamedRerouteUsage(), B=Step(X=@n15) {constX=0.1, constY=0.1})
   B=Multiply(A=NamedRerouteUsage(), B=Step(X=@n17) {constX=0.1, constY=0.1})
  )
  n79 = Constant4Vector() {constant=<rgba 1 2 0 1>}
  n80 = MaterialFunctionCall(Input=Max(A=Max(A=@n75, B=@n76), B=Max(A=@n77, B=@n78)), Input Low=@n79.R, Input High=@n79.R, Target Low=@n79.R, Target High=@n79.R) {materialFunction=&RemapValueRange}
  n81 = CurveAtlasRowParameter(CurveTime=@n80.Result) {
   atlas=/ToyFoxGameplay/Assets/Texture/Curves/CA_ToyFox_Weakpoint
   curve=/ToyFoxGameplay/Assets/Texture/Curves/C_Toyfox_Weakpoint_Transition
   group=Weakpoints
   parameterName=WeakpointCurveAnimation
  }
  n82 = LinearInterpolate(A=@n74, B=Multiply(A=Power(Base=@n74)) {constB=2}, Alpha=@n81.G) {constA=1, constB=10}
  n83 = OneMinus(None=@n81.R)
  NamedRerouteDeclaration(None=Multiply(A=Multiply(A=@n82, B=@n83), B=LinearInterpolate(Alpha=@n81.G) {constA=1, constB=10})) {name=Weakpoints_EmissiveAnimated, nodeColor=<rgba 1 0 0.3984 1>}
  n84 = ScalarParameter() {group=Weakpoints, parameterName=Weakpoint01Stage}
  NamedRerouteDeclaration(None=LinearInterpolate(A=Subtract(A=@n84), B=@n84, Alpha=NamedRerouteUsage()) {constB=20}) {name=WP01_Stages, nodeColor=<rgba 0.3812 1 0.0069 1>}
  n85 = ScalarParameter() {group=Weakpoints, parameterName=Weakpoint02Stage}
  NamedRerouteDeclaration(None=LinearInterpolate(A=Subtract(A=@n85), B=@n85, Alpha=NamedRerouteUsage()) {constB=20}) {name=WP02_Stages, nodeColor=<rgba 0.3812 1 0.0069 1>}
  n86 = NamedRerouteUsage()
  n87 = LinearInterpolate(A=Multiply(A=@n86) {constA=0.5, constB=0.5}, B=Multiply(A=@n86) {constB=2}, Alpha=NamedRerouteUsage())
  n88 = Multiply(
   A=VectorParameter() {group=Weakpoints, parameterName=WeakpointEdgeColor, defaultValue=<rgba 0.8228 0.0999 0.7682 1>}.RGB
   B=ScalarParameter() {defaultValue=1, group=Weakpoints, parameterName=EdgeIntensity}
  ) {constB=40}
  n89 = NamedRerouteUsage()
  n90 = VectorParameter() {group=Weakpoints, parameterName=WeakpointCoreColor, defaultValue=<rgba 0.4508 0 0.7682 1>}
  n91 = NamedRerouteUsage()
  n92 = Multiply(A=@n91) {constB=0.25}
  n93 = Multiply(A=@n91, B=NamedRerouteUsage())
  n94 = Multiply(A=Power(Base=Add(A=@n92, B=@n93)) {constExponent=3}) {constB=20}
  NamedRerouteDeclaration(None=Add(A=LinearInterpolate(A=@n87, B=@n88, Alpha=@n89), B=Multiply(A=@n90.RGB, B=@n94))) {name=Weakpoints_Emissive, nodeColor=<rgba 0 1 0.3438 1>}
  n95 = ScalarParameter() {group=Weakpoints, parameterName=Weakpoint03Stage}
  NamedRerouteDeclaration(None=LinearInterpolate(A=Subtract(A=@n95), B=@n95, Alpha=NamedRerouteUsage()) {constB=20}) {name=WP03_Stages, nodeColor=<rgba 0.3812 1 0.0069 1>}
  n96 = ScalarParameter() {group=Weakpoints, parameterName=Weakpoint04Stage}
  NamedRerouteDeclaration(None=LinearInterpolate(A=Subtract(A=@n96), B=@n96, Alpha=NamedRerouteUsage()) {constB=20}) {name=WP04_Stages, nodeColor=<rgba 0.3812 1 0.0069 1>}
  n97 = ScalarParameter() {group=Weakpoints, parameterName=Weakpoint05Stage}
  NamedRerouteDeclaration(None=LinearInterpolate(A=Subtract(A=@n97), B=@n97, Alpha=NamedRerouteUsage()) {constB=20}) {name=WP05_Stages, nodeColor=<rgba 0.3812 1 0.0069 1>}
  n98 = ScalarParameter() {group=Weakpoints, parameterName=Weakpoint06Stage}
  NamedRerouteDeclaration(None=LinearInterpolate(A=Subtract(A=@n98), B=@n98, Alpha=NamedRerouteUsage()) {constB=20}) {name=WP06_Stages, nodeColor=<rgba 0.3812 1 0.0069 1>}
  n99 = ScalarParameter() {group=Weakpoints, parameterName=Weakpoint07Stage}
  NamedRerouteDeclaration(None=LinearInterpolate(A=Subtract(A=@n99), B=@n99, Alpha=NamedRerouteUsage()) {constB=20}) {name=WP07_Stages, nodeColor=<rgba 0.3812 1 0.0069 1>}
  n100 = ScalarParameter() {group=Weakpoints, parameterName=Weakpoint08Stage}
  NamedRerouteDeclaration(None=LinearInterpolate(A=Subtract(A=@n100), B=@n100, Alpha=NamedRerouteUsage()) {constB=20}) {name=WP08_Stages, nodeColor=<rgba 0.3812 1 0.0069 1>}
  n101 = OneMinus(None=@n80.Result)
  NamedRerouteDeclaration(None=Saturate(None=Multiply(A=@n80.Result, B=@n101))) {name=Weakpoints_IsolationMsk, nodeColor=<rgba 1 0.2344 0 1>}
  n102 = NamedRerouteUsage()
  n103 = Multiply(A=@n2, B=NamedRerouteUsage())
  n104 = Multiply(A=@n4, B=NamedRerouteUsage())
  n105 = Multiply(A=@n6, B=NamedRerouteUsage())
  n106 = Multiply(A=@n8, B=NamedRerouteUsage())
  n107 = Multiply(A=@n11, B=NamedRerouteUsage())
  n108 = Multiply(A=@n13, B=NamedRerouteUsage())
  n109 = Multiply(A=@n15, B=NamedRerouteUsage())
  n110 = Multiply(A=@n17, B=NamedRerouteUsage())
  n111 = Max(A=Max(A=Max(A=@n103, B=@n104), B=Max(A=@n105, B=@n106)), B=Max(A=Max(A=@n107, B=@n108), B=Max(A=@n109, B=@n110)))
  n112 = Multiply(A=@n111) {constB=3}
  n113 = Constant4Vector() {constant=<rgba 0 0.5 0 1>}
  n114 = MaterialFunctionCall(Mask In=@n102, Amount=OneMinus(None=@n112), Range=@n113.G, Start Offset=@n112, Sharpness=@n113.G) {materialFunction=&MF_MaskShifting}
  n115 = OneMinus(None=@n114.Result)
  n116 = CurveAtlasRowParameter(CurveTime=@n111) {
   atlas=&CA_ToyFox_Weakpoint
   curve=&C_Toyfox_Weakpoint_Transition
   group=Weakpoints
   parameterName=WeakpointCurveAnimation
  }
  n117 = Multiply(A=Saturate(None=Multiply(A=@n115, B=@n114.Result)), B=@n116.B) {constA=1}
  n118 = Multiply(A=NamedRerouteUsage(), B=@n102)
  NamedRerouteDeclaration(None=Multiply(A=Multiply(A=@n117, B=@n118)) {constB=2}) {name=Hit_Blink, nodeColor=<rgba 1 0.2578 0 1>}
  n119 = Time()
  n120 = Subtract(A=@n119, B=ScalarParameter() {defaultValue=-99999, group=Weakpoints, parameterName=Hit01CurveTime})
  n121 = Add(A=NamedRerouteUsage()) {constB=0}
  NamedRerouteDeclaration(None=Saturate(None=Divide(A=@n120, B=@n121))) {name=Hit01_Curve, nodeColor=<rgba 1 0 0.375 1>}
  n122 = Time()
  n123 = Subtract(A=@n122, B=ScalarParameter() {defaultValue=-99999, group=Weakpoints, parameterName=Hit02CurveTime})
  n124 = Add(A=NamedRerouteUsage()) {constB=0}
  NamedRerouteDeclaration(None=Saturate(None=Divide(A=@n123, B=@n124))) {name=Hit02_Curve, nodeColor=<rgba 1 0 0.375 1>}
  n125 = Time()
  n126 = Subtract(A=@n125, B=ScalarParameter() {defaultValue=-99999, group=Weakpoints, parameterName=Hit03CurveTime})
  n127 = Add(A=NamedRerouteUsage()) {constB=0}
  NamedRerouteDeclaration(None=Saturate(None=Divide(A=@n126, B=@n127))) {name=Hit03_Curve, nodeColor=<rgba 1 0 0.375 1>}
  n128 = Time()
  n129 = Subtract(A=@n128, B=ScalarParameter() {defaultValue=-99999, group=Weakpoints, parameterName=Hit04CurveTime})
  n130 = Add(A=NamedRerouteUsage()) {constB=0}
  NamedRerouteDeclaration(None=Saturate(None=Divide(A=@n129, B=@n130))) {name=Hit04_Curve, nodeColor=<rgba 1 0 0.375 1>}
  n131 = Time()
  n132 = Subtract(A=@n131, B=ScalarParameter() {defaultValue=-99999, group=Weakpoints, parameterName=Hit05CurveTime})
  n133 = Add(A=NamedRerouteUsage()) {constB=0}
  NamedRerouteDeclaration(None=Saturate(None=Divide(A=@n132, B=@n133))) {name=Hit05_Curve, nodeColor=<rgba 1 0 0.375 1>}
  n134 = Time()
  n135 = Subtract(A=@n134, B=ScalarParameter() {defaultValue=-99999, group=Weakpoints, parameterName=Hit06CurveTime})
  n136 = Add(A=NamedRerouteUsage()) {constB=0}
  NamedRerouteDeclaration(None=Saturate(None=Divide(A=@n135, B=@n136))) {name=Hit06_Stage, nodeColor=<rgba 1 0 0.375 1>}
  n137 = Time()
  n138 = Subtract(A=@n137, B=ScalarParameter() {defaultValue=-99999, group=Weakpoints, parameterName=Hit07CurveTime})
  n139 = Add(A=NamedRerouteUsage()) {constB=0}
  NamedRerouteDeclaration(None=Saturate(None=Divide(A=@n138, B=@n139))) {name=Hit07_Curve, nodeColor=<rgba 1 0 0.375 1>}
  n140 = Time()
  n141 = Subtract(A=@n140, B=ScalarParameter() {defaultValue=-99999, group=Weakpoints, parameterName=Hit08CurveTime})
  n142 = Add(A=NamedRerouteUsage()) {constB=0}
  NamedRerouteDeclaration(None=Saturate(None=Divide(A=@n141, B=@n142))) {name=Hit08_Curve, nodeColor=<rgba 1 0 0.375 1>}
  NamedRerouteDeclaration(None=FunctionInput() {inputName=IN MetalNoise, inputType=FunctionInput_Scalar, sortPriority=1}) {name=MetalNoise, nodeColor=<rgba 0 0.5781 1 1>}
  n143 = NamedRerouteUsage()
  n144 = Constant4Vector() {constant=<rgba 0.36 0.1 0 1>}
  n145 = Subtract(
   A=Multiply(A=NamedRerouteUsage(), B=@n143)
   B=MaterialFunctionCall(Mask In=@n143, Amount=@n144.R, Range=@n144.R, Start Offset=@n144.R, Sharpness=@n144.R) {materialFunction=&MF_MaskShifting}.Result
  )
  n146 = Multiply(A=Power(Base=Saturate(None=@n145)) {constExponent=4}, B=NamedRerouteUsage())
  NamedRerouteDeclaration(None=Multiply(A=NamedRerouteUsage(), B=Multiply(A=@n146) {constB=10})) {name=Weakpoint_RingMsk, nodeColor=<rgba 1 0 0.8203 1>}
  n147 = NamedRerouteUsage()
  n148 = VectorParameter() {group=Weakpoints, parameterName=CorruptNoise_Scale, defaultValue=<rgba 20 20 30 1>}
  n149 = Add(A=NamedRerouteUsage(), B=Multiply(A=NamedRerouteUsage()) {constB=0.5})
  n150 = NamedRerouteUsage()
  n151 = Constant()
  n152 = Constant2Vector() {g=-0.01}
  n153 = NamedRerouteUsage()
  n154 = OneMinus(None=Multiply(A=@n153, B=NamedRerouteUsage()))
  n155 = LinearInterpolate(A=Multiply(A=@n152, B=Multiply(A=@n154) {constB=0.01}), B=@n152, Alpha=Step(Y=@n153) {constX=0})
  n156 = Multiply(A=@n148.RGB) {constB=2}
  n157 = Multiply(A=@n155) {constB=1.5}
  n158 = ComponentMask(None=Add(
    A=MaterialFunctionCall!(
     Export Float 4
     TextureObject=@n147
     TextureSize=@n148.RGB
     WorldPosition=@n149
     World Space Normal=@n150
     ProjectionTransitionContrast=@n151
     Pan_YZ=@n155
     Pan_XZ=@n155
     Pan_XY
    ) {materialFunction=/BRCosmetics/Gadgets/Assets/VinderTech_GliderChute/Glider_MissusMind/Materials/WorldAlignedTexture_ScalePanXYZ}.XYZ Texture
    B=MaterialFunctionCall(
     TextureObject=@n147
     TextureSize=@n156
     WorldPosition=@n149
     World Space Normal=@n150
     ProjectionTransitionContrast=@n151
     Pan_YZ=@n157
     Pan_XZ=@n157
    ) {materialFunction=&WorldAlignedTexture_ScalePanXYZ}.XYZ Texture
   )) {g=true}
  NamedRerouteDeclaration(None=@n158) {name=TriplannarNoise_Static, nodeColor=<rgba 1 0.394 0.9658 1>}
  NamedRerouteDeclaration(None=FunctionInput() {inputName=IN TriplannerUV, inputType=FunctionInput_Vector2, sortPriority=1}) {name=HardTriplannarUV, nodeColor=<rgba 0.0312 0 1 1>}
  NamedRerouteDeclaration(None=TextureObject() {texture=/ToyFoxGameplay/Assets/Texture/T_ToyFox_TA_Mask}) {name=Noise_TEX, nodeColor=<rgba 0 0.4232 1 1>}
  n159 = OneMinus(None=DotProduct(A=VertexNormalWS(), B=CameraVectorWS()))
  NamedRerouteDeclaration(None=Power(Base=Saturate(None=@n159)) {constExponent=4}) {name=SwordFresnel, nodeColor=<rgba 0.6172 1 0 1>}
  n160 = ComponentMask(None=NamedRerouteUsage()) {g=true, r=true}
  n161 = ScalarParameter() {group=Weakpoints, parameterName=IsPowerRegen}
  n162 = Panner(
   Coordinate=Add(A=NamedRerouteUsage(), B=Multiply(A=@n160) {constB=0.1}) {desc=Add normal dirstort}
   Speed=LinearInterpolate(A=Constant2Vector() {r=0.1}, B=Constant2Vector() {g=0.2}, Alpha=@n161) {constAlpha=1}
  ) {speedX=0.1}
  n163 = NamedRerouteUsage()
  n164 = TextureSample(UVs=Multiply(A=@n162, B=Constant() {r=0.5}) {constB=4}, Tex=@n163) {samplerSource=SSM_Wrap_WorldGroupSettings}
  n165 = TextureSample(UVs=Multiply(A=@n162) {constB=-1}, Tex=@n163) {samplerSource=SSM_Wrap_WorldGroupSettings}
  NamedRerouteDeclaration(None=Multiply(A=Subtract(A=@n164.R, B=@n165.R), B=ScalarParameter() {defaultValue=0.35, group=FinisherAnimation, parameterName=LightningDistort_Intensity})) {name=DistortNoise, nodeColor=<rgba 0.2891 1 0 1>}
  n166 = NamedRerouteUsage()
  n167 = NamedRerouteUsage()
  n168 = Add(A=@n166, B=Multiply(A=ComponentMask(None=@n167) {g=true, r=true}) {constB=0.25}) {desc=Add normal dirstort}
  n169 = MaterialFunctionCall(UVs=Panner(Coordinate=@n168) {speedX=0.05, speedY=-0.05}, "Rotation Angle (0-1)"=Constant() {r=0.1}) {materialFunction=&CustomRotator}
  NamedRerouteDeclaration(None=@n169.Rotated Values) {name=RotatedTriUV, nodeColor=<rgba 1 0 0.3281 1>}
  NamedRerouteDeclaration(None=@n169.Rotated Values) {name=LightningUV_01, nodeColor=<rgba 0 1 0.7422 1>}
  n170 = Add(A=@n168, B=Constant2Vector() {r=-0.2})
  NamedRerouteDeclaration(None=MaterialFunctionCall(UVs=Panner(Coordinate=@n170) {speedX=-0.05, speedY=0.05}, "Rotation Angle (0-1)"=Constant() {r=0.02}) {materialFunction=&CustomRotator}.Rotated Values) {name=LightningUV_02, nodeColor=<rgba 1 0.4219 0 1>}
  n171 = Constant() {r=0.2}
  n172 = Frac(None=Multiply(A=MaterialFunctionCall(UVs=@n166, "Rotation Angle (0-1)"=@n171) {materialFunction=&CustomRotator}.Rotated Values) {constB=2})
  n173 = ComponentMask(None=Subtract(A=Multiply(A=@n172) {constB=2})) {r=true}
  NamedRerouteDeclaration(None=SmoothStep(Value=Abs(None=@n173)) {constMax=2}) {name=DistortPattern, nodeColor=<rgba 0.1016 0 1 1>}
  n174 = Time()
  n175 = LinearInterpolate(Alpha=ConstantBiasScale(None=MaterialFunctionCall(Float=@n174) {materialFunction=&MF_Hash}.Float)) {constA=0.007, constAlpha=0.2, constB=0.01}
  n176 = Add(A=NamedRerouteUsage(), B=NamedRerouteUsage())
  n177 = Multiply(A=NamedRerouteUsage(), B=NamedRerouteUsage()) {constB=2}
  n178 = Frac(None=Multiply(A=Add(A=@n176, B=@n177)) {constB=5, desc=Count})
  n179 = ComponentMask(None=Subtract(A=Multiply(A=@n178) {constB=2})) {r=true}
  n180 = Saturate(None=Divide(A=@n175, B=Abs(None=@n179)) {constA=0.008})
  n181 = Add(A=NamedRerouteUsage(), B=NamedRerouteUsage())
  n182 = Multiply(A=NamedRerouteUsage(), B=NamedRerouteUsage()) {constB=0.2}
  n183 = Frac(None=Multiply(A=Add(A=@n181, B=@n182)) {constB=3, desc=Count})
  n184 = ComponentMask(None=Subtract(A=Multiply(A=@n183) {constB=2})) {r=true}
  n185 = Saturate(None=Divide(A=@n175, B=Abs(None=@n184)) {constA=0.01})
  n186 = Power(Base=Saturate(None=Min(A=@n180, B=@n185)))
  n187 = ScalarParameter() {defaultValue=20, parameterName=Lightning_Intensity}
  n188 = VectorParameter() {group=FinisherAnimation, parameterName=Lightning_Color, defaultValue=<rgba 0.2 0.3 0.8 1>}
  n189 = Panner(Coordinate=Multiply(A=NamedRerouteUsage(), B=Constant2Vector() {g=0.1, r=1})) {speedX=0.05, speedY=0.25}
  n190 = NamedRerouteUsage()
  n191 = Multiply(A=Power(Base=TextureSample(UVs=@n189, Tex=@n190) {samplerSource=SSM_Wrap_WorldGroupSettings}.RGB) {constExponent=1}) {constB=2}
  NamedRerouteDeclaration(None=Multiply(A=Multiply(A=@n186, B=@n187) {constB=40}, B=LinearInterpolate(A=@n188.RGB, B=@n191, Alpha=@n161))) {name=LightningWrap, nodeColor=<rgba 0 0.6016 1 1>}
  n192 = Subtract(A=Time(), B=ScalarParameter() {defaultValue=-99999, group=Weakpoints, parameterName=LF_GameTimeFromBP})
  n193 = Add(A=ScalarParameter() {defaultValue=1, group=Weakpoints, parameterName=LF_Animation_Duration, sortPriority=101}) {constB=0}
  NamedRerouteDeclaration(None=Saturate(None=Divide(A=@n192, B=@n193))) {name=TimeNative_LF, nodeColor=<rgba 0.2891 1 0 1>}
  n194 = ScalarParameter() {group=Weakpoints, parameterName=LF_Stage}
  n195 = LinearInterpolate(A=Subtract(A=@n194), B=@n194, Alpha=NamedRerouteUsage()) {constB=20}
  NamedRerouteDeclaration(None=Saturate(None=@n195)) {name=LF_Stages_01, nodeColor=<rgba 0.3812 1 0.0069 1>}
  n196 = Constant4Vector() {constant=<rgba 1 2 0 1>}
  NamedRerouteDeclaration(None=Saturate(None=MaterialFunctionCall(Input=@n195, Input Low=@n196.R, Input High=@n196.R, Target Low=@n196.R, Target High=@n196.R) {materialFunction=&RemapValueRange}.Result)) {name=LF_Stages_02, nodeColor=<rgba 0.3812 1 0.0069 1>}
  n197 = Constant4Vector() {constant=<rgba 1 2 0 1>}
  NamedRerouteDeclaration(None=Saturate(None=MaterialFunctionCall(Input=@n54, Input Low=@n197.R, Input High=@n197.R, Target Low=@n197.R, Target High=@n197.R) {materialFunction=&RemapValueRange}.Result)) {name=PowerRegen_Stages_02, nodeColor=<rgba 0.3812 1 0.0069 1>}
  n198 = Subtract(A=Time(), B=ScalarParameter() {defaultValue=-99999, group=Weakpoints, parameterName=PowerRegen_GameTimeFromBP})
  n199 = Add(A=ScalarParameter() {defaultValue=1, group=Weakpoints, parameterName=PowerRegen_Animation_Duration, sortPriority=101}) {constB=0}
  NamedRerouteDeclaration(None=Saturate(None=Divide(A=@n198, B=@n199))) {name=TimeNative_PowerRegen, nodeColor=<rgba 0.2891 1 0 1>}
  n200 = NamedRerouteUsage()
  n201 = Desaturation(None=@n200)
  n202 = NamedRerouteUsage()
  n203 = LinearInterpolate(
   A=Multiply(A=@n200, B=ScalarParameter() {defaultValue=0.5, group=Weakpoints, parameterName=Weakpoints_Intensity_Main}) {constA=1, desc=emissive}
   B=Multiply(
    A=LinearInterpolate(A=@n200, B=@n201, Alpha=@n202) {constB=2}
    B=ScalarParameter() {defaultValue=10, group=Weakpoints, parameterName=Hit_IntensityBoost}
   ) {constB=6}
   Alpha=@n202
  ) {constB=8}
  n204 = NamedRerouteUsage()
  n205 = NamedRerouteUsage()
  n206 = Multiply(A=NamedRerouteUsage(), B=OneMinus(None=NamedRerouteUsage()))
  NamedRerouteDeclaration(None=Add(A=Add(A=@n203, B=@n204), B=Multiply(A=@n205, B=@n206))) {name=Out_Emissive, nodeColor=<rgba 0.2422 0 1 1>}
  NamedRerouteDeclaration(None=Saturate(None=@n20.Result)) {name=Weakpoints_Base, nodeColor=<rgba 1 0 0.3984 1>}
  n207 = NamedRerouteUsage()
  n208 = NamedRerouteUsage()
  NamedRerouteDeclaration(None=Saturate(None=Subtract(A=@n207, B=@n208))) {name=TransitionGradient, nodeColor=<rgba 1 0.1875 0 1>}
  NamedRerouteDeclaration(None=@n22.Normal) {name=IN_Normal, nodeColor=<rgba 0 1 0.6719 1>}
  NamedRerouteDeclaration(None=@n22.BaseColor) {name=IN_BC, nodeColor=<rgba 0.2656 0 1 1>}
  NamedRerouteDeclaration(None=@n22.Metallic) {name=IN_Metallic, nodeColor=<rgba 0.0547 1 0 1>}
  NamedRerouteDeclaration(None=@n22.Specular) {name=IN_Specular, nodeColor=<rgba 1 0 0.375 1>}
  NamedRerouteDeclaration(None=@n22.Roughness) {name=IN_Roughness, nodeColor=<rgba 0 0.6953 1 1>}
  n209 = Subtract(A=@n119, B=ScalarParameter() {defaultValue=-99999, group=Weakpoints, parameterName=GameTimeFromBP01})
  NamedRerouteDeclaration(None=Saturate(None=Divide(A=@n209, B=@n121))) {name=TimeNative01, nodeColor=<rgba 0.2891 1 0 1>}
  n210 = LocalPosition() {localOrigin=InstancePreSkinning}
  NamedRerouteDeclaration(None=FunctionInput(Preview=VertexInterpolator(VS=@n210.XYZ).PS) {inputName=IN PreSkinnedLoc, sortPriority=2}) {name=PreskinnedPos, nodeColor=<rgba 1 0 1 1>}
  NamedRerouteDeclaration(None=@n22.EmissiveColor) {name=IN_Emissive, nodeColor=<rgba 0 0.8125 1 1>}
  n211 = Add(A=NamedRerouteUsage()) {constB=0.1}
  n212 = Divide(A=Constant() {r=1}, B=Ceil(None=Constant() {r=1}))
  n213 = Multiply(A=@n212) {constB=0.5}
  n214 = Multiply(A=@n212, B=Ceil(None=Constant()))
  n215 = TextureSample(UVs=AppendVector(A=ComponentMask(None=@n211) {r=true}, B=Add(A=@n213, B=@n214))) {
   samplerSource=SSM_Wrap_WorldGroupSettings
   samplerType=SAMPLERTYPE_LinearColor
   texture=/ToyFoxGameplay/Assets/Texture/Curves/CA_Toyfox_Weakpoint_Gradient
  }
  n216 = NamedRerouteUsage()
  NamedRerouteDeclaration(None=Multiply(A=Multiply(A=@n215.RGB, B=@n216), B=Step(X=@n216) {constY=0.1})) {name=WeakpointEmissive, nodeColor=<rgba 1 0.0536 0 1>}
  n217 = PreSkinnedNormal()
  NamedRerouteDeclaration(None=FunctionInput(Preview=VertexInterpolator(VS=@n217).PS) {inputName=IN PreSkinnedNormal, sortPriority=3}) {name=PreSkinnedNormal, nodeColor=<rgba 0.25 0.25 1 1>}
  n218 = NamedRerouteUsage()
  n219 = VectorParameter() {group=Weakpoints, parameterName=Weakpoint_AreaTweak, channelNames {a=Sharp, b=Offset, g=Range, r=Amount}, defaultValue=<rgba 0 0.05 0 0.5>}
  n220 = MaterialFunctionCall(
   Mask In=Multiply(A=NamedRerouteUsage(), B=NamedRerouteUsage())
   Amount=Multiply(A=OneMinus(None=@n218), B=@n219.Amount)
   Range=@n219.Range
   Start Offset=@n219.Range
   Sharpness=@n219.Range
  ) {materialFunction=&MF_MaskShifting}
  n221 = Constant4Vector() {constant=<rgba 0 0.5 0 1>}
  n222 = MaterialFunctionCall(Mask In=NamedRerouteUsage(), Amount=@n221.R, Range=@n221.R, Start Offset=@n221.R, Sharpness=@n221.R) {materialFunction=&MF_MaskShifting}
  n223 = Saturate(None=Multiply(A=OneMinus(None=@n222.Result)) {constB=0.6})
  n224 = Multiply(A=NamedRerouteUsage(), B=OneMinus(None=NamedRerouteUsage()))
  NamedRerouteDeclaration(None=Add(A=Saturate(None=@n220.Result), B=Multiply(A=@n223, B=@n224))) {name=Weakpoint_AreaMask, nodeColor=<rgba 0 1 0.6953 1>}
  n225 = Multiply(A=NamedRerouteUsage(), B=NamedRerouteUsage())
  n226 = Multiply(A=LinearInterpolate(Alpha=NamedRerouteUsage()) {constA=-1}) {constB=0.01}
  NamedRerouteDeclaration(None=Add(A=@n158, B=Add(A=@n225, B=@n226))) {name=TriplannarNoise_Daynamic, nodeColor=<rgba 1 0.394 0.9658 1>}
  n227 = Multiply(A=NamedRerouteUsage()) {constB=4}
  n228 = TextureObject() {samplerType=SAMPLERTYPE_Normal, texture=/Game/Global/Textures/Default/Noise/T_Noise_Normal_01}
  NamedRerouteDeclaration(None=MaterialFunctionCall(Normal=TextureSample(UVs=@n227, Tex=@n228) {samplerSource=SSM_Wrap_WorldGroupSettings}.RGB, Flatness=Constant() {r=0.5}) {materialFunction=&FlattenNormal}.Result) {name=FlakeNormal, nodeColor=<rgba 0.0547 0 1 1>}
  NamedRerouteDeclaration(None=ScalarParameter() {defaultValue=1, group=Weakpoints, parameterName=Animation_Duration, sortPriority=101}) {name=Weapoint_SpawnDuration, nodeColor=<rgba 0 0.8359 1 1>}
  n229 = Subtract(A=@n122, B=ScalarParameter() {defaultValue=-99999, group=Weakpoints, parameterName=GameTimeFromBP02})
  NamedRerouteDeclaration(None=Saturate(None=Divide(A=@n229, B=@n124))) {name=TimeNative02, nodeColor=<rgba 0.2891 1 0 1>}
  n230 = Subtract(A=@n125, B=ScalarParameter() {defaultValue=-99999, group=Weakpoints, parameterName=GameTimeFromBP03})
  NamedRerouteDeclaration(None=Saturate(None=Divide(A=@n230, B=@n127))) {name=TimeNative03, nodeColor=<rgba 0.2891 1 0 1>}
  n231 = Subtract(A=@n128, B=ScalarParameter() {defaultValue=-99999, group=Weakpoints, parameterName=GameTimeFromBP04})
  NamedRerouteDeclaration(None=Saturate(None=Divide(A=@n231, B=@n130))) {name=TimeNative04, nodeColor=<rgba 0.2891 1 0 1>}
  n232 = Subtract(A=@n131, B=ScalarParameter() {defaultValue=-99999, group=Weakpoints, parameterName=GameTimeFromBP05})
  NamedRerouteDeclaration(None=Saturate(None=Divide(A=@n232, B=@n133))) {name=TimeNative05, nodeColor=<rgba 0.2891 1 0 1>}
  n233 = Subtract(A=@n134, B=ScalarParameter() {defaultValue=-99999, group=Weakpoints, parameterName=GameTimeFromBP06})
  NamedRerouteDeclaration(None=Saturate(None=Divide(A=@n233, B=@n136))) {name=TimeNative06, nodeColor=<rgba 0.2891 1 0 1>}
  n234 = Subtract(A=@n137, B=ScalarParameter() {defaultValue=-99999, group=Weakpoints, parameterName=GameTimeFromBP07})
  NamedRerouteDeclaration(None=Saturate(None=Divide(A=@n234, B=@n139))) {name=TimeNative07, nodeColor=<rgba 0.2891 1 0 1>}
  n235 = Subtract(A=@n140, B=ScalarParameter() {defaultValue=-99999, group=Weakpoints, parameterName=GameTimeFromBP08})
  NamedRerouteDeclaration(None=Saturate(None=Divide(A=@n235, B=@n142))) {name=TimeNative08, nodeColor=<rgba 0.2891 1 0 1>}
  n236 = Constant4Vector() {desc=WeakpointComp_Chest, constant=<rgba 0 20 120 18>}
  n237 = Saturate(None=NamedRerouteUsage())
  NamedRerouteDeclaration(None=SphereMask(
    A=LocalPosition() {localOrigin=InstancePreSkinning}.XYZ
    B=ComponentMask(None=@n236.RGBA) {b=true, g=true, r=true}
    Radius=Multiply(A=@n236.A, B=@n237)
   ) {attenuationRadius=20, hardnessPercent=0}) {name=Weakpoint01_Msk, nodeColor=<rgba 1 0 0.75 1>}
  n238 = Constant4Vector() {desc=WeakpointComp_LowerLeg_R, constant=<rgba -18.2012 -2.0518 39.0069 22>}
  n239 = Saturate(None=NamedRerouteUsage())
  NamedRerouteDeclaration(None=SphereMask(
    A=LocalPosition() {localOrigin=InstancePreSkinning}.XYZ
    B=ComponentMask(None=@n238.RGBA) {b=true, g=true, r=true}
    Radius=Multiply(A=@n238.A, B=@n239)
   ) {attenuationRadius=20, hardnessPercent=0}) {name=Weakpoint02_Msk, nodeColor=<rgba 1 0 0.75 1>}
  VectorParameter() {bUseCustomPrimitiveData=true, parameterName=Weakspot_02_Pos, primitiveDataIndex=4}
  VectorParameter() {bUseCustomPrimitiveData=true, parameterName=Weakspot_03_Pos, primitiveDataIndex=8}
  VectorParameter() {bUseCustomPrimitiveData=true, parameterName=Weakspot_04_Pos, primitiveDataIndex=12}
  VectorParameter() {bUseCustomPrimitiveData=true, parameterName=Weakspot_05_Pos, primitiveDataIndex=16}
  VectorParameter() {bUseCustomPrimitiveData=true, parameterName=Weakspot_06_Pos, primitiveDataIndex=20}
  VectorParameter() {bUseCustomPrimitiveData=true, parameterName=Weakspot_07_Pos, primitiveDataIndex=24}
  VectorParameter() {bUseCustomPrimitiveData=true, parameterName=Weakspot_08_Pos, primitiveDataIndex=28}
  VectorParameter() {bUseCustomPrimitiveData=true, parameterName=Weakspot_01_Pos}
 ]}
