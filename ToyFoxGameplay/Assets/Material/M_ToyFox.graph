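# M_ToyFox is the ToyFox character material, using material attributes with per-expression shading models for skinned meshes.
# A shared character parent function feeds MF_ToyFox, whose plain and weakpoint versions both enter an outro function, followed by tech-art overrides.
# Named reroutes share pre-skinned position, normal, bounds-relative coordinates and hard triplanar UVs.
# Those UVs pick one projection by dominant normal axis, using custom HLSL sign flips instead of blending.
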Material /ToyFoxGameplay/Assets/Material/M_ToyFox
graph {
 properties {
  bUseMaterialAttributes=true
  bUsedWithClothing=true
  bUsedWithMeshDeformer=true
  bUsedWithMorphTargets=true
  bUsedWithNanite=true
  bUsedWithNiagaraMeshParticles=true
  bUsedWithSkeletalMesh=true
  shadingModel=MSM_FromMaterialExpression
  usedShadingModels=Default Lit | Subsurface | Clear Coat | Subsurface Profile | Cloth
 }
 nodes [
  n1 = MaterialFunctionCall() {materialFunction=&MF_CharacterParent_2023_PreFX}
  n2 = MaterialFunctionCall!(
   IN MA=@n1.Result
   IN TriplannerUV=NamedRerouteUsage()
   IN Bounding=NamedRerouteUsage()
   IN TAMsk=@n1.Result
   IN PreSkinnedLoc=NamedRerouteUsage()
   IN PreSkinnedNormal=NamedRerouteUsage()
  ) {materialFunction=&/MF_ToyFox}
  n3 = NamedRerouteUsage()
  n4 = NamedRerouteUsage()
  n5 = NamedRerouteUsage()
  MaterialFunctionCall(In=MaterialFunctionCall!(
    IN OriginalVisual MA=@n2.OUT MA
    IN WeakpointVisual MA=MaterialFunctionCall!(In=@n2.OUT MA, IN TriplannerUV=@n3, IN MetalNoise=@n2.OUT MA, IN PreSkinnedLoc=@n4, IN PreSkinnedNormal=@n5) {materialFunction=&/MF_ToyFox_Weakpoints}.Result
    IN Bounding=NamedRerouteUsage()
    IN VisorMask=@n2.OUT MA
    IN TriplannarUV=NamedRerouteUsage()
    IN PreSkinnedLoc=NamedRerouteUsage()
    IN PreSkinnedNormal=NamedRerouteUsage()
   ) {materialFunction=&/MF_ToyFox_Outro}.Result) {materialFunction=&MF_TechArtOverrides}
  NamedRerouteDeclaration(None=@n1.Pre skinned Local Position) {name=PreskinnedLocPos, nodeColor=<rgba 0 1 0.0391 1>}
  n6 = Divide(A=Multiply(A=NamedRerouteUsage()) {constB=-1}, B=Abs(None=Constant() {r=32})) {constB=1024}
  n7 = VectorParameter() {group=Toyfox, parameterName=Triplannar3DTilling, defaultValue=<rgba 1 1 1 1>}
  n8 = MaterialFunctionCall(Float3=Add(A=Multiply(A=@n6, B=@n7.RGB), B=VectorParameter() {group=Toyfox, parameterName=Triplannar3DOffset, defaultValue=<rgba 0 0 0 1>}.RGB)) {materialFunction=&BreakOutFloat3Components}
  n9 = NamedRerouteUsage()
  MaterialExpressionComponentMask_8 = ComponentMask(None=@n9) {r=true}
  n10 = Custom!(x=@MaterialExpressionComponentMask_8) {
   code="return float(x < 0) ? -1 : 1;"
   description=StepReturn
   inputs [{inputName=x, input {expression=&self:MaterialExpressionComponentMask_8}}]
  }
  MaterialExpressionComponentMask_7 = ComponentMask(None=@n9) {g=true}
  n11 = Custom(x=@MaterialExpressionComponentMask_7) {
   code="return float(x < 0) ? 1 : -1;"
   description=StepReturn
   inputs [{inputName=x, input {expression=&self:MaterialExpressionComponentMask_7}}]
  }
  MaterialExpressionComponentMask_6 = ComponentMask(None=@n9) {b=true}
  n12 = Custom(x=@MaterialExpressionComponentMask_6) {
   code="return float(x < 0) ? -1 : 1;"
   description=StepReturn
   inputs [{inputName=x, input {expression=&self:MaterialExpressionComponentMask_6}}]
  }
  n13 = MaterialFunctionCall(Float3=MaterialFunctionCall(X=@n10, Y=@n11, Z=@n12) {materialFunction=&MakeFloat3}.Result) {materialFunction=&BreakOutFloat3Components}
  n14 = Abs(None=@n9)
  n15 = MaterialFunctionCall(Float3=@n14) {materialFunction=&BreakOutFloat3Components}
  n16 = Max(A=@n15.R, B=@n15.R)
  n17 = MaterialFunctionCall(Float3=Step(Y=Max(A=@n16, B=@n15.B), X=@n14)) {materialFunction=&BreakOutFloat3Components}
  n18 = LinearInterpolate(
   A=MaterialFunctionCall(X=Multiply(A=@n8.G, B=@n13.R), Y=@n8.B) {desc=Side, materialFunction=&MakeFloat2}.Result
   B=MaterialFunctionCall(X=Multiply(A=@n8.R, B=@n13.G), Y=@n8.B) {desc=Front, materialFunction=&MakeFloat2}.Result
   Alpha=@n17.G
  )
  n19 = MaterialFunctionCall(X=Multiply(A=@n8.R, B=@n13.B), Y=@n8.G) {desc=Top, materialFunction=&MakeFloat2}
  NamedRerouteDeclaration(None=ConstantBiasScale(None=LinearInterpolate(A=@n18, B=@n19.Result, Alpha=@n17.B))) {name=HardTriplannerUV, nodeColor=<rgba 0.0078 0 1 1>}
  NamedRerouteDeclaration(None=VertexInterpolator(VS=PreSkinnedNormal()).PS) {name=PreskinnedLocN, nodeColor=<rgba 0 0.6016 1 1>}
  n20 = TransformPosition(None=NamedRerouteUsage()) {transformType=TRANSFORMPOSSOURCE_Instance}
  n21 = PreSkinnedLocalBounds()
  NamedRerouteDeclaration(None=Divide(A=Subtract(A=@n20, B=@n21.Min), B=@n21.Extents)) {name=Bounding, nodeColor=<rgba 0 0.5312 1 1>}
 ]
}
