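# Unlit animated intro background material for ToyFox, summing lightning, core glow, flares, clouds and colored warp streaks into a final output.
# Viewport UVs are aspect-corrected and converted to polar coordinates, then distorted by noise panned on stepped time to form lightning beams, with random flash and scan masks.
# A custom HLSL node, RangerWarp, draws six-color radial streaks.
# Offset RGB samples fake chromatic aberration, collapsing to one sample on Mobile and Medium quality.
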Material /ToyFoxGameplay/Assets/Material/M_ToyFox_IntroBG
graph {
 properties {bAutomaticallySetUsageInEditor=false, shadingModel=MSM_Unlit, usedShadingModels=Unlit}
 nodes [
  n1 = MaterialFunctionCall(Float=Time()) {materialFunction=&MF_Hash}
  n2 = Add(A=Multiply(A=ConstantBiasScale(None=@n1.Float)) {constB=0.1}) {constB=0.2}
  n3 = NamedRerouteUsage()
  n4 = NamedRerouteUsage()
  n5 = NamedRerouteUsage()
  n6 = OneMinus(None=Add(A=@n4, B=@n5) {constB=0.05})
  n7 = Frac(None=Multiply(A=Add(A=@n3, B=@n6) {desc=Distort}) {constB=4})
  n8 = ComponentMask(None=Subtract(A=Multiply(A=@n7) {constB=2})) {g=true}
  n9 = Power(Base=Divide(A=LinearInterpolate(Alpha=@n2) {constAlpha=0.2, constB=0.07}, B=Abs(None=@n8))) {constExponent=1}
  n10 = LinearInterpolate() {constAlpha=0.2, constB=0.07}
  n11 = Add(A=@n3, B=@n6) {desc=Distort}
  n12 = Multiply(A=Frac(None=Multiply(A=@n11) {constB=4})) {constB=2}
  n13 = Abs(None=ComponentMask(None=Subtract(A=@n12)) {g=true})
  Add(A=@n9, B=Power(Base=Divide(A=@n10, B=@n13)) {constExponent=1})
  Constant() {r=1}
  n14 = ScalarParameter() {defaultValue=4, parameterName=Lines}
  Multiply(A=@n14) {constB=2}
  n15 = ScalarParameter() {defaultValue=4, parameterName=Lines}
  Multiply(A=@n15) {constB=2}
  Multiply(A=@n4, B=@n5) {constB=1.2}
  n16 = ScalarParameter() {defaultValue=4, parameterName=Lines}
  Multiply(A=@n16) {constB=2}
  n17 = Fmod(A=Time(), B=Constant() {r=10})
  n18 = Divide(A=Floor(None=Multiply(A=@n17) {constB=12})) {constB=10}
  n19 = VectorParameter() {parameterName=LightningSpeed, defaultValue=<rgba -0.4 0.2 0 1>}
  n20 = MaterialFunctionCall(X=@n19.R, Y=@n19.R) {materialFunction=&MakeFloat2}
  n21 = Panner(Coordinate=NamedRerouteUsage(), Time=@n18, Speed=@n20.Result) {speedX=0.5, speedY=0.1}
  n22 = NamedRerouteUsage()
  n23 = NamedRerouteUsage()
  n24 = NamedRerouteUsage()
  NamedRerouteDeclaration(None=Multiply(A=TextureSample!(UVs=@n21, Tex=@n22, "DDX(UVs)"=@n23, "DDY(UVs)"=@n24, Apply View MipBias) {mipValueMode=TMVM_Derivative, samplerSource=SSM_Wrap_WorldGroupSettings, samplerType=SAMPLERTYPE_LinearColor}.R) {constB=0.04}) {name=Distort, nodeColor=<rgba 0 1 0.6484 1>}
  n25 = ViewSize()
  n26 = Divide(A=ComponentMask(None=@n25) {r=true}, B=ComponentMask(None=@n25) {g=true})
  n27 = Constant2Vector() {g=1, r=1}
  n28 = AppendVector(
   A=Divide(B=If(A=@n26, B=@n27.R, "A > B"=@n27.R, "A == B"=@n27.R, "A < B"=@n26))
   B=Multiply(B=If(A=@n26, B=@n27.R, "A > B"=@n26, "A == B"=@n26, "A < B"=@n27.R)) {constA=1}
  )
  n29 = AppendVector(A=@n27.R, B=@n27.R)
  NamedRerouteDeclaration(None=MaterialFunctionCall(UVs=ScreenPosition().ViewportUV, Texture Scale=Multiply(A=@n28, B=@n29)) {materialFunction=&ScaleUVsByCenter}.UVs) {name=InitalUV, nodeColor=<rgba 1 0.8438 0 1>}
  n30 = Saturate(None=MaterialFunctionCall(Float=Time()) {materialFunction=&MF_Hash}.Float)
  n31 = NamedRerouteUsage()
  n32 = LinearInterpolate(Alpha=Length(None=Subtract(A=@n31) {constB=0.5}) {is=class MaterialExpressionLength}) {constB=0.004}
  n33 = Constant2Vector() {g=2, r=2}
  n34 = MaterialFunctionCall!(UV Coord=Add(A=@n31, B=@n32), LenghtScale=@n33.R, RadialScale=@n33.R) {materialFunction=/Game/Global/MaterialFunctions/TechArt/MF_PolarCoordinates}
  n35 = NamedRerouteUsage()
  n36 = MaterialFunctionCall(UV Coord=@n31, LenghtScale=@n33.R, RadialScale=@n33.R) {materialFunction=&MF_PolarCoordinates}
  n37 = ComponentMask(None=Add(A=@n36.Result, B=@n35)) {g=true}
  n38 = MaterialFunctionCall(UV Coord=Subtract(A=@n31, B=@n32) {constB=0.5}, LenghtScale=@n33.R, RadialScale=@n33.R) {materialFunction=&MF_PolarCoordinates}
  n39 = MaterialFunctionCall(X=ComponentMask(None=Add(A=@n34.Result, B=@n35)) {g=true}, Y=@n37, Z=ComponentMask(None=Add(A=@n38.Result, B=@n35)) {g=true}) {materialFunction=&MakeFloat3}
  n40 = NamedRerouteUsage()
  n41 = Frac(None=Multiply(
    A=MaterialFunctionCall!(Low=@n39.Result, High=@n39.Result, Epic=@n39.Result, --------, Mobile=@n37, Medium=@n37) {materialFunction=/FortniteBaseMaterials/Functions/MF_QualitySwitch_Vector}.Result
    B=@n40
   ))
  n42 = Abs(None=Subtract(A=Multiply(A=@n41) {constB=2}))
  n43 = Power(Base=Divide(A=LinearInterpolate(Alpha=@n30) {constA=0.007, constAlpha=0.1, constB=0.01}, B=@n42))
  n44 = ScalarParameter() {defaultValue=0.227, group=Lightning, parameterName=BackLightningExp}
  n45 = LinearInterpolate(Alpha=@n30) {constA=0.007, constAlpha=0.1, constB=0.01}
  n46 = Sign(None=MaterialFunctionCall(Float=Time()) {materialFunction=&MF_Hash}.Float)
  n47 = OneMinus(None=Add(A=Multiply(A=@n35, B=@n46) {constB=2}, B=NamedRerouteUsage()) {constB=0.05})
  n48 = ComponentMask(None=Add(A=@n36.Result, B=@n47)) {g=true}
  n49 = MaterialFunctionCall(X=ComponentMask(None=Add(A=@n34.Result, B=@n47)) {g=true}, Y=@n48, Z=ComponentMask(None=Add(A=@n38.Result, B=@n47)) {g=true}) {materialFunction=&MakeFloat3}
  n50 = Frac(None=Multiply(
    A=MaterialFunctionCall(Low=@n49.Result, High=@n49.Result, Epic=@n49.Result, Mobile=@n48, Medium=@n48) {materialFunction=&MF_QualitySwitch_Vector}.Result
    B=@n40
   ) {constB=8})
  n51 = Abs(None=Subtract(A=Multiply(A=@n50) {constB=2}))
  n52 = Power(Base=Power(Base=Divide(A=@n45, B=@n51)) {constExponent=1}, Exp=@n44)
  n53 = NamedRerouteUsage()
  NamedRerouteDeclaration(None=Add(A=Power(Base=@n43, Exp=@n44), B=Multiply(A=@n52, B=@n53))) {name=LightningCloudNoise, nodeColor=<rgba 0 0 1 1>}
  n54 = NamedRerouteUsage()
  n55 = Add(A=@n54, B=Multiply(A=NamedRerouteUsage(), B=NamedRerouteUsage()) {constB=20})
  n56 = NamedRerouteUsage()
  n57 = Multiply(A=NamedRerouteUsage(), B=@n56) {constB=20}
  n58 = Multiply(A=@n56, B=@n54) {constB=2}
  n59 = NamedRerouteUsage()
  n60 = Subtract(A=NamedRerouteUsage()) {constB=0.5}
  n61 = Multiply(A=Add(A=Add(A=@n55, B=@n57), B=Multiply(A=@n58, B=@n59) {constB=4}), B=SmoothStep(Value=Length(None=@n60)) {constMax=0.2, constMin=0.05})
  n62 = VectorParameter() {parameterName=CoreLightColorA, defaultValue=<rgba 0.2 0.3 0.8 1>}
  NamedRerouteDeclaration(None=Multiply(A=Multiply(A=@n61, B=@n62.RGB), B=ScalarParameter() {defaultValue=1, parameterName=Lightning_Intensity})) {name=Main_LightningStrike, nodeColor=<rgba 1 0.8906 0 1>}
  NamedRerouteDeclaration(None=Saturate(None=@n9)) {name=LightningCloudNoise_1, nodeColor=<rgba 0 0 1 1>}
  MaterialExpressionMultiply_106 = Multiply(A=Time()) {constB=0.2}
  MaterialExpressionNamedRerouteUsage_77 = NamedRerouteUsage()
  n63 = Custom!(Time=@MaterialExpressionMultiply_106, canvs=@MaterialExpressionNamedRerouteUsage_77) {
   code="// The below expression will get compiled\r\n// into the output of this node\r\nstruct MyFunction\r\n{\r\nfloat vDrop(float2 uv, float t){\r\nuv.x = uv.x*16;\t\t\t\t\t\t//Count\r\nfloat dx = frac(uv.x);\r\nuv.x = floor(uv.x);\r\nuv.y *= 0.05;\t\t\t\t\t\t\t// stretch\r\nfloat o=sin(uv.x*215.4);\t\t\t\t// offset\r\nfloat s=cos(uv.x*33.1)*.3 +.7;\t\t\t// speed\r\nfloat trail = lerp(100.0,50.0,s);\t\t\t// trail length\r\nfloat yv = frac(uv.y + t*s + o) * trail;\r\nyv = 1.0/yv;\r\nyv = smoothstep(0.0,1.0,yv*yv);\r\nyv = sin(yv*3.14)*(s*5.0);\r\nfloat d2 = sin(dx*3.14);\r\nreturn yv*(d2*d2);\r\n}\r\n\r\n};\r\n\r\nMyFunction MF;\r\nfloat2 p = canvs.xy - 0.5;\r\nfloat d = length(p)+0.1;\r\np = float2(atan2(p.x, p.y) / 3.14, 0.1 / d);\r\np.y *= 4;\r\nfloat t = Time;\r\nfloat3 col = float3(1.0,0.0,0.0) * MF.vDrop(p,t);\t// red\r\ncol += float3(0.0,0.0,1.0) * MF.vDrop(p,t+0.16667);\t// blue\r\ncol += float3(1.0,1.0,0.0) * MF.vDrop(p,t+0.16667*2);\t// yellow\r\ncol += float3(0.0,1.0,0.0) * MF.vDrop(p,t+0.16667*3);\t// green\r\ncol += float3(1.0,0.1,0.5) * MF.vDrop(p,t+0.16667*4);\t// pink\r\ncol += float3(0.22,0.0,0.230) * MF.vDrop(p,t+0.16667*5);\t// purple\r\ncol *= d*d;\r\nreturn col;\r\n"
   description=RangerWarp
   showCode=true
   inputs [
    {inputName=Time, input {expression=&self:MaterialExpressionMultiply_106}}
    {inputName=canvs, input {expression=&self:MaterialExpressionNamedRerouteUsage_77}}
   ]
  }
  n64 = NamedRerouteUsage()
  n65 = Multiply(A=@n64) {constB=32}
  n66 = Subtract(A=Multiply(A=Frac(None=@n65)) {constB=2})
  n67 = Min(A=OneMinus(None=ComponentMask(None=@n64) {r=true}), B=ComponentMask(None=@n64) {r=true})
  n68 = Power(Base=Abs(None=ComponentMask(None=@n66) {g=true}), Exp=LinearInterpolate(Alpha=@n67) {constA=0.005, constAlpha=1, constB=0.01})
  n69 = Multiply(A=Power(Base=@n63, Exp=Divide() {constB=2.2}) {constExponent=0.5}, B=Multiply(A=OneMinus(None=@n68), B=@n67))
  n70 = Multiply(A=@n67) {constA=20, constB=20}
  NamedRerouteDeclaration(None=Multiply(A=Multiply(A=@n69, B=@n70) {constB=10}, B=ScalarParameter() {defaultValue=2, parameterName=RangerColor_Intensity}) {constB=0}) {name=RangerColor_Acent, nodeColor=<rgba 1 0.5859 0 1>}
  n71 = NamedRerouteUsage()
  n72 = NamedRerouteUsage()
  n73 = Add(A=Add(A=@n71, B=Multiply(A=@n71, B=@n72)), B=NamedRerouteUsage())
  n74 = NamedRerouteUsage()
  NamedRerouteDeclaration(None=Add(A=Add(A=@n73, B=@n74), B=NamedRerouteUsage())) {name=FinalOutput, nodeColor=<rgba 0.9745 0.9684 1 1>}
  n75 = TextureSample(UVs=Add(A=@n21) {constB=0.8}, Tex=@n22, "DDX(UVs)"=NamedRerouteUsage(), "DDY(UVs)"=NamedRerouteUsage()) {mipValueMode=TMVM_Derivative, samplerSource=SSM_Wrap_WorldGroupSettings}
  n76 = NamedRerouteUsage()
  n77 = TextureSample!(UVs=Panner(Coordinate=Multiply(A=@n76) {constB=4}) {speedX=0.01, speedY=0.01}, Tex=@n22, Apply View MipBias) {samplerSource=SSM_Wrap_WorldGroupSettings}
  n78 = Constant() {r=0.005}
  NamedRerouteDeclaration(None=Add(A=Multiply(A=@n75.R) {constB=0.02}, B=Multiply(A=@n77.G, B=@n78) {constB=0.01})) {name=Distort_2, nodeColor=<rgba 0 1 0.6484 1>}
  n79 = Constant2Vector() {g=0.2}
  n80 = Add(A=@n3, B=@n79)
  n81 = NamedRerouteUsage()
  n82 = Frac(None=Multiply(A=Add(A=@n80, B=@n81) {desc=Distort}) {constB=4})
  n83 = ComponentMask(None=Subtract(A=Multiply(A=@n82) {constB=2})) {g=true}
  n84 = Power(Base=Divide(A=LinearInterpolate() {constAlpha=0.1, constB=0.07}, B=Abs(None=@n83))) {constExponent=1}
  n85 = Add(A=@n3, B=Multiply(A=@n79) {constB=-1})
  n86 = Frac(None=Multiply(A=Add(A=@n85, B=@n81) {desc=Distort}) {constB=8})
  n87 = ComponentMask(None=Subtract(A=Multiply(A=@n86) {constB=2})) {g=true}
  n88 = Power(Base=Divide(A=LinearInterpolate() {constAlpha=0.1, constB=0.07}, B=Abs(None=@n87))) {constExponent=1}
  NamedRerouteDeclaration(None=Saturate(None=Add(A=@n84, B=@n88))) {name=LightningCloudNoise_2, nodeColor=<rgba 0 0 1 1>}
  n89 = NamedRerouteUsage()
  n90 = NamedRerouteUsage()
  n91 = TextureSample(UVs=Panner(Coordinate=@n89) {speedX=0.05}, Tex=@n90) {samplerSource=SSM_Wrap_WorldGroupSettings}
  n92 = TextureSample(UVs=BumpOffset(Coordinate=Panner(Coordinate=@n89) {speedX=0.1}, Height=Constant() {r=-2}), Tex=@n90) {samplerSource=SSM_Wrap_WorldGroupSettings}
  NamedRerouteDeclaration(None=Multiply(A=Add(A=@n91.R, B=@n92.R), B=ScalarParameter() {defaultValue=1, parameterName=BG_Cloud_Intensity}) {constB=2}) {name=BGCloud, nodeColor=<rgba 1 1 1 1>}
  n93 = NamedRerouteUsage()
  n94 = TextureSample!(
   UVs=Add(A=Panner(Coordinate=@n93, Time=@n18, Speed=@n20.Result) {speedX=0.5, speedY=0.1}) {constB=0.5}
   Tex=@n22
   "DDX(UVs)"=NamedRerouteUsage()
   "DDY(UVs)"=NamedRerouteUsage()
   Apply View MipBias
  ) {mipValueMode=TMVM_Derivative, samplerSource=SSM_Wrap_WorldGroupSettings}
  n95 = Multiply(A=@n77.R, B=@n78) {constB=0.01}
  NamedRerouteDeclaration(None=Add(A=Multiply(A=@n94.R) {constB=0.02}, B=@n95)) {name=Distort_3, nodeColor=<rgba 0 1 0.6484 1>}
  n96 = Add(A=NamedRerouteUsage()) {constB=0}
  n97 = NamedRerouteUsage()
  n98 = Frac(None=Multiply(A=Add(A=@n96, B=@n97) {desc=Distort}) {constB=8})
  n99 = ComponentMask(None=Subtract(A=Multiply(A=@n98) {constB=2})) {g=true}
  n100 = Power(Base=Divide(A=LinearInterpolate() {constAlpha=0.2, constB=0.07}, B=Abs(None=@n99)))
  n101 = Add(A=NamedRerouteUsage()) {constB=0}
  n102 = NamedRerouteUsage()
  n103 = Frac(None=Multiply(A=Add(A=@n101, B=@n102) {desc=Distort}) {constB=8})
  n104 = ComponentMask(None=Subtract(A=Multiply(A=@n103) {constB=2})) {g=true}
  n105 = Power(Base=Divide(A=LinearInterpolate() {constAlpha=0.2, constB=0.07}, B=Abs(None=@n104)))
  NamedRerouteDeclaration(None=Saturate(None=Add(A=@n100, B=@n105))) {name=LightningCloudNoise_3, nodeColor=<rgba 0 0 1 1>}
  n106 = NamedRerouteUsage()
  n107 = VectorParameter() {parameterName=Center, defaultValue=<rgba 0 0 0 1>}
  n108 = Add(
   A=@n106
   B=MaterialFunctionCall(X=@n107.R, Y=@n107.R) {materialFunction=&MakeFloat2}.Result
  )
  n109 = ScalarParameter() {defaultValue=1, parameterName=PolarUVLengthScale}
  n110 = ScalarParameter() {defaultValue=2, parameterName=PolarUVRadialScale}
  n111 = MaterialFunctionCall(UV Coord=@n108, LenghtScale=@n109, RadialScale=@n110) {materialFunction=&MF_PolarCoordinates}
  NamedRerouteDeclaration(None=@n111.Result) {name=PolarUV, nodeColor=<rgba 0.9688 0 1 1>}
  NamedRerouteDeclaration(None=@n42) {name=LightFracMask, nodeColor=<rgba 0 0.9062 1 1>}
  NamedRerouteDeclaration(None=ScalarParameter() {defaultValue=8, group=Lightning, parameterName=Num_LightningBeam, sortPriority=0}) {name=NumbersOfBeam, nodeColor=<rgba 0 1 0.9059 1>}
  n112 = Constant2Vector() {r=0.01}
  NamedRerouteDeclaration(None=MaterialFunctionCall(UV Coord=Add(A=@n108, B=@n112), LenghtScale=@n109, RadialScale=@n110) {materialFunction=&MF_PolarCoordinates}.Result) {name=OffsetedPolarUV_01, nodeColor=<rgba 0 0.2969 1 1>}
  n113 = Multiply(A=@n112) {constB=-1}
  NamedRerouteDeclaration(None=MaterialFunctionCall(UV Coord=Add(A=@n108, B=@n113), LenghtScale=@n109, RadialScale=@n110) {materialFunction=&MF_PolarCoordinates}.Result) {name=OffsetedPolarUV_02, nodeColor=<rgba 0 0.2969 1 1>}
  NamedRerouteDeclaration(None=@n111.Result) {name=PolarUV_1, nodeColor=<rgba 0.9688 0 1 1>}
  NamedRerouteDeclaration(None=@n95) {name=Distort_4, nodeColor=<rgba 0 1 0.6484 1>}
  NamedRerouteDeclaration(None=TextureObject() {texture=&T_Tile_RainbowColor}) {name=ChromaticTex, nodeColor=<rgba 0 0.6719 1 1>}
  n114 = OneMinus(None=NamedRerouteUsage())
  n115 = Add(A=NamedRerouteUsage(), B=NamedRerouteUsage())
  n116 = Constant2Vector() {g=1}
  n117 = NamedRerouteUsage()
  n118 = Floor(None=@n117)
  n119 = Floor(None=Multiply(A=Multiply(A=@n115, B=@n116), B=@n118))
  n120 = NamedRerouteUsage()
  n121 = NamedRerouteUsage()
  n122 = NamedRerouteUsage()
  n123 = TextureSample(UVs=Panner(Coordinate=Divide(A=@n119, B=@n118)) {speedY=0.5}, Tex=@n120, "DDX(UVs)"=@n121, "DDY(UVs)"=@n122) {mipValueMode=TMVM_Derivative, samplerSource=SSM_Wrap_WorldGroupSettings}
  NamedRerouteDeclaration(None=Saturate(None=Multiply(A=@n114, B=@n123.G))) {name=RandomFlash, nodeColor=<rgba 1 0 1 1>}
  NamedRerouteDeclaration(None=DDX(Value=@n106)) {name=DDX, nodeColor=<rgba 0 1 0.0625 1>}
  n124 = VectorParameter() {parameterName=CoreLightColorB, defaultValue=<rgba 0.48 0.2 1 1>}
  n125 = NamedRerouteUsage()
  n126 = LinearInterpolate(Alpha=Length(None=Subtract(A=@n125) {constB=0.5})) {constB=0.05}
  n127 = Add(A=@n125, B=@n126)
  n128 = MaterialFunctionCall(UVs=@n127, Radius=Constant() {r=1}, Density=Constant() {r=1}) {materialFunction=&RadialGradientExponential}
  n129 = MaterialFunctionCall(Vector or UVs=Subtract(A=@n127) {constB=0.5}) {materialFunction=&VectorToRadialValue}
  n130 = Panner(Coordinate=@n129.Vector Converted to Angle) {speedX=0.1, speedY=-0.1}
  n131 = NamedRerouteUsage()
  n132 = NamedRerouteUsage()
  n133 = NamedRerouteUsage()
  n134 = MaterialFunctionCall(UVs=@n125, Radius=Constant() {r=1}, Density=Constant() {r=1}) {materialFunction=&RadialGradientExponential}
  n135 = MaterialFunctionCall(Vector or UVs=Subtract(A=@n125) {constB=0.5}) {materialFunction=&VectorToRadialValue}
  n136 = NamedRerouteUsage()
  n137 = NamedRerouteUsage()
  n138 = NamedRerouteUsage()
  n139 = Multiply(
   A=Power(Base=@n134.RadialGradientExponential)
   B=TextureSample(UVs=Panner(Coordinate=@n135.Vector Converted to Angle) {speedX=0.1, speedY=-0.1}, Tex=@n136, "DDX(UVs)"=@n137, "DDY(UVs)"=@n138) {mipValueMode=TMVM_Derivative, samplerSource=SSM_Wrap_WorldGroupSettings}.R
  )
  n140 = Subtract(A=@n125, B=@n126) {constB=0.5}
  n141 = MaterialFunctionCall(UVs=@n140, Radius=Constant() {r=1}, Density=Constant() {r=1}) {materialFunction=&RadialGradientExponential}
  n142 = MaterialFunctionCall(Vector or UVs=Subtract(A=@n140) {constB=0.5}) {materialFunction=&VectorToRadialValue}
  n143 = Panner(Coordinate=@n142.Vector Converted to Angle) {speedX=0.1, speedY=-0.1}
  n144 = NamedRerouteUsage()
  n145 = NamedRerouteUsage()
  n146 = NamedRerouteUsage()
  n147 = MaterialFunctionCall(
   X=Multiply(
    A=Power(Base=@n128.RadialGradientExponential)
    B=TextureSample(UVs=@n130, Tex=@n131, "DDX(UVs)"=@n132, "DDY(UVs)"=@n133) {mipValueMode=TMVM_Derivative, samplerSource=SSM_Wrap_WorldGroupSettings}.R
   )
   Y=@n139
   Z=Multiply(
    A=Power(Base=@n141.RadialGradientExponential)
    B=TextureSample(UVs=@n143, Tex=@n144, "DDX(UVs)"=@n145, "DDY(UVs)"=@n146) {mipValueMode=TMVM_Derivative, samplerSource=SSM_Wrap_WorldGroupSettings}.R
   )
  ) {materialFunction=&MakeFloat3}
  n148 = MaterialFunctionCall(Low=@n147.Result, High=@n147.Result, Epic=@n147.Result, Mobile=@n139, Medium=@n139) {materialFunction=&MF_QualitySwitch_Vector}
  n149 = MaterialFunctionCall(UVs=@n127, Radius=Constant() {r=0.2}, Density=Constant() {r=1}) {materialFunction=&RadialGradientExponential}
  n150 = Power(Base=Add(A=@n149.RadialGradientExponential, B=@n128.RadialGradientExponential))
  n151 = Power(Base=@n128.RadialGradientExponential) {constExponent=4}
  n152 = TextureSample(UVs=Panner(Coordinate=@n129.Vector Converted to Angle) {speedX=-0.1, speedY=0.1}, Tex=@n131, "DDX(UVs)"=@n132, "DDY(UVs)"=@n133) {mipValueMode=TMVM_Derivative, samplerSource=SSM_Wrap_WorldGroupSettings}
  n153 = Power(Base=Multiply(A=OneMinus(None=@n150), B=Max(A=@n151, B=@n152.R))) {constExponent=4}
  n154 = MaterialFunctionCall(UVs=@n125, Radius=Constant() {r=0.2}, Density=Constant() {r=1}) {materialFunction=&RadialGradientExponential}
  n155 = Multiply(
   A=Power(Base=Add(A=@n154.RadialGradientExponential, B=@n134.RadialGradientExponential))
   B=TextureSample(UVs=Panner(Coordinate=@n135.Vector Converted to Angle) {speedX=-0.1, speedY=0.1}, Tex=@n136, "DDX(UVs)"=@n137, "DDY(UVs)"=@n138) {mipValueMode=TMVM_Derivative, samplerSource=SSM_Wrap_WorldGroupSettings}.R
  )
  n156 = Add(
   A=MaterialFunctionCall(UVs=@n140, Radius=Constant() {r=0.2}, Density=Constant() {r=1}) {materialFunction=&RadialGradientExponential}.RadialGradientExponential
   B=@n141.RadialGradientExponential
  )
  n157 = Panner(Coordinate=@n142.Vector Converted to Angle) {speedX=-0.1, speedY=0.1}
  n158 = MaterialFunctionCall(
   X=Multiply(A=@n150, B=@n152.R)
   Y=@n155
   Z=Multiply(
    A=Power(Base=@n156)
    B=TextureSample(UVs=@n157, Tex=@n144, "DDX(UVs)"=@n145, "DDY(UVs)"=@n146) {mipValueMode=TMVM_Derivative, samplerSource=SSM_Wrap_WorldGroupSettings}.R
   )
  ) {materialFunction=&MakeFloat3}
  n159 = Power(Base=@n149.RadialGradientExponential) {constExponent=4}
  n160 = Add(
   A=Saturate(None=Multiply(A=@n153))
   B=Add(
    A=MaterialFunctionCall(Low=@n158.Result, High=@n158.Result, Epic=@n158.Result, Mobile=@n155, Medium=@n155) {materialFunction=&MF_QualitySwitch_Vector}.Result
    B=Multiply(A=@n159) {constB=10}
   )
  )
  n161 = LinearInterpolate(
   A=LinearInterpolate(B=Multiply(A=@n124.RGB, B=@n148.Result), Alpha=@n160)
   B=Multiply(A=VectorParameter() {parameterName=CoreLightColorA, defaultValue=<rgba 0.2 0.3 0.8 1>}.RGB, B=@n160)
   Alpha=@n160
  )
  n162 = Max(A=Power(Base=@n161), B=Power(Base=@n161) {constExponent=1})
  NamedRerouteDeclaration(None=Multiply(A=@n162, B=ScalarParameter() {defaultValue=3, parameterName=CoreLight_Intensity}) {constB=20}) {name=Core, nodeColor=<rgba 1 0.7969 0 1>}
  NamedRerouteDeclaration(None=TextureObject() {texture=/Game/Global/Textures/Default/Noise/T_Tiling_Noise_01}) {name=NoiseTex, nodeColor=<rgba 0 0.4475 1 1>}
  n163 = Subtract(A=NamedRerouteUsage()) {constB=0.5}
  n164 = LinearInterpolate(Alpha=Length(None=Subtract(A=@n163) {constB=0.5})) {constA=0.01, constB=0.04}
  n165 = Multiply(A=Add(A=@n163, B=@n164))
  n166 = Multiply(A=Frac(None=MaterialFunctionCall(Vector or UVs=@n165) {materialFunction=&VectorToRadialValue}.Radial Coordinates)) {constB=2}
  n167 = MaterialFunctionCall(Float2=Abs(None=Subtract(A=@n166))) {materialFunction=&BreakOutFloat2Components}
  n168 = Multiply(A=@n16, B=@n167.R)
  n169 = Panner(Coordinate=@n168) {speedX=-0.1}
  n170 = NamedRerouteUsage()
  n171 = NamedRerouteUsage()
  n172 = NamedRerouteUsage()
  n173 = Panner(Coordinate=@n168) {speedX=0.1}
  n174 = NamedRerouteUsage()
  n175 = NamedRerouteUsage()
  n176 = NamedRerouteUsage()
  n177 = Add(
   A=Add(
    A=TextureSample(UVs=@n169, Tex=@n170, "DDX(UVs)"=@n171, "DDY(UVs)"=@n172) {mipValueMode=TMVM_Derivative, samplerSource=SSM_Wrap_WorldGroupSettings}.R
    B=TextureSample(UVs=@n173, Tex=@n174, "DDX(UVs)"=@n175, "DDY(UVs)"=@n176) {mipValueMode=TMVM_Derivative, samplerSource=SSM_Wrap_WorldGroupSettings}.R
   )
   B=@n167.G
  )
  n178 = Multiply(A=@n163) {constB=1.1}
  n179 = Multiply(A=Frac(None=MaterialFunctionCall(Vector or UVs=@n178) {materialFunction=&VectorToRadialValue}.Radial Coordinates)) {constB=2}
  n180 = MaterialFunctionCall(Float2=Abs(None=Subtract(A=@n179))) {materialFunction=&BreakOutFloat2Components}
  n181 = Multiply(A=@n14, B=@n180.R)
  n182 = Add(
   A=TextureSample(UVs=Panner(Coordinate=@n181) {speedX=-0.1}, Tex=NamedRerouteUsage(), "DDX(UVs)"=NamedRerouteUsage(), "DDY(UVs)"=NamedRerouteUsage()) {mipValueMode=TMVM_Derivative, samplerSource=SSM_Wrap_WorldGroupSettings}.R
   B=TextureSample(UVs=Panner(Coordinate=@n181) {speedX=0.1}, Tex=NamedRerouteUsage(), "DDX(UVs)"=NamedRerouteUsage(), "DDY(UVs)"=NamedRerouteUsage()) {mipValueMode=TMVM_Derivative, samplerSource=SSM_Wrap_WorldGroupSettings}.R
  )
  n183 = OneMinus(None=SmoothStep(Value=Add(A=@n182, B=@n180.G)) {constMin=-5})
  n184 = Multiply(A=Subtract(A=@n163, B=@n164) {constB=0.5}) {constB=0.9}
  n185 = Multiply(A=Frac(None=MaterialFunctionCall(Vector or UVs=@n184) {materialFunction=&VectorToRadialValue}.Radial Coordinates)) {constB=2}
  n186 = MaterialFunctionCall(Float2=Abs(None=Subtract(A=@n185))) {materialFunction=&BreakOutFloat2Components}
  n187 = Multiply(A=@n15, B=@n186.R)
  n188 = Panner(Coordinate=@n187) {speedX=-0.1}
  n189 = NamedRerouteUsage()
  n190 = NamedRerouteUsage()
  n191 = NamedRerouteUsage()
  n192 = Panner(Coordinate=@n187) {speedX=0.1}
  n193 = NamedRerouteUsage()
  n194 = NamedRerouteUsage()
  n195 = NamedRerouteUsage()
  n196 = Add(
   A=Add(
    A=TextureSample(UVs=@n188, Tex=@n189, "DDX(UVs)"=@n190, "DDY(UVs)"=@n191) {mipValueMode=TMVM_Derivative, samplerSource=SSM_Wrap_WorldGroupSettings}.R
    B=TextureSample(UVs=@n192, Tex=@n193, "DDX(UVs)"=@n194, "DDY(UVs)"=@n195) {mipValueMode=TMVM_Derivative, samplerSource=SSM_Wrap_WorldGroupSettings}.R
   )
   B=@n186.G
  )
  n197 = MaterialFunctionCall(X=OneMinus(None=SmoothStep(Value=@n177) {constMin=-5}), Y=@n183, Z=OneMinus(None=SmoothStep(Value=@n196) {constMin=-5})) {materialFunction=&MakeFloat3}
  n198 = MaterialFunctionCall(Low=@n197.Result, High=@n197.Result, Epic=@n197.Result, Mobile=@n183, Medium=@n183) {materialFunction=&MF_QualitySwitch_Vector}
  NamedRerouteDeclaration(None=Multiply(A=Power(Base=@n198.Result), B=ScalarParameter() {defaultValue=4, parameterName=Flare_Intensity}) {constB=10}) {name=Flare, nodeColor=<rgba 1 0.6094 0 1>}
  NamedRerouteDeclaration(None=DDY(Value=@n106)) {name=DDY, nodeColor=<rgba 0 1 0.0625 1>}
  n199 = OneMinus(None=NamedRerouteUsage())
  n200 = Panner(Coordinate=@n115) {speedX=-0.5}
  n201 = Floor(None=@n117)
  n202 = Multiply(A=Multiply(A=@n115, B=@n116), B=@n201)
  n203 = TextureSample(UVs=Divide(A=Floor(None=@n202), B=@n201), Tex=@n120, "DDX(UVs)"=@n121, "DDY(UVs)"=@n122) {mipValueMode=TMVM_Derivative, samplerSource=SSM_Wrap_WorldGroupSettings}
  n204 = Frac(None=ComponentMask(None=Add(A=@n200, B=@n203.R) {constB=0, desc=Scan Panning Offset}) {r=true})
  n205 = SmoothStep(Value=Min(A=@n204, B=OneMinus(None=@n204))) {constMax=0.8, constMin=0.2}
  NamedRerouteDeclaration(None=Saturate(None=Multiply(A=@n199, B=@n205))) {name=RandomScan, nodeColor=<rgba 1 0 0.5391 1>}
  NamedRerouteUsage()
  NamedRerouteUsage()
  NamedRerouteUsage()
  QualitySwitch()
  Saturate(None=@n162)
  ScalarParameter() {defaultValue=2.1357, parameterName=ScaleX}
  ScalarParameter() {defaultValue=0.5286, parameterName=ScaleY}
  VectorParameter() {parameterName=ThunderColor, defaultValue=<rgba 0.2 0.3 0.8 1>}
 ]
}
