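# M_ToyFox_Pedestal_Tank is an unlit material for a pedestal tank showing a glowing, animated face amid electricity, reacting to an interaction progress parameter.
# The face uses stabilized screen-space UVs with flowmap distortion and flipbook blending.
# Custom HLSL scans, noise-panned lines and bolt flipbooks form lightning, with cheaper mobile variants through quality switches.
# Fill masks, Fresnel, skylight-cubemap reflection and time-of-day scaling shape the final emissive.
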Material /ToyFoxGameplay/Assets/Material/M_ToyFox_Pedestal_Tank
graph {
 properties {bAutomaticallySetUsageInEditor=false, shadingModel=MSM_Unlit, usedShadingModels=Unlit}
 nodes [
  n1 = NamedRerouteUsage()
  n2 = Divide(
   A=Constant() {r=0.08}
   B=ComponentMask(None=MaterialFunctionCall!(UV Coord=@n1, LenghtScale, RadialScale) {materialFunction=/Game/Global/MaterialFunctions/TechArt/MF_PolarCoordinates}.Result) {r=true}
  )
  n3 = NamedRerouteUsage()
  n4 = NamedRerouteUsage()
  n5 = TextureSample(UVs=@n4, Tex=NamedRerouteUsage()) {samplerSource=SSM_Clamp_WorldGroupSettings}
  n6 = MaterialFunctionCall!(A=Add(A=@n4) {constB=-0.5}, B, Bounds=Constant2Vector() {g=1, r=1}, Edge Falloff=Constant() {r=0.01}) {materialFunction=/Engine/Functions/Engine_MaterialFunctions01/Gradient/BoxMask-2D}
  n7 = Multiply(
   A=Multiply(A=Multiply(A=@n5.G) {constB=2}, B=NamedRerouteUsage())
   B=NamedRerouteDeclaration(None=Multiply(A=@n5.A, B=@n6.Result)) {name=HeadMask, nodeColor=<rgba 0.5938 0 1 1>}
  )
  n8 = MaterialFunctionCall!(Size=Constant2Vector() {g=8, r=8}, Selector=Fmod(A=Time(), B=Constant() {r=8}), UV=@n4) {materialFunction=/Game/Global/MaterialFunctions/MF_FlipbookUV}
  n9 = NamedRerouteUsage()
  n10 = TextureSample(UVs=@n8.FlipbookUV1, Tex=@n9) {samplerSource=SSM_Wrap_WorldGroupSettings}
  n11 = TextureSample(UVs=@n8.FlipbookUV2, Tex=@n9) {samplerSource=SSM_Wrap_WorldGroupSettings}
  n12 = LinearInterpolate(
   A=@n5.R
   B=Multiply(A=LinearInterpolate(A=@n10.R, B=@n11.R, Alpha=@n8.Blend1), B=NamedRerouteUsage())
   Alpha=ScalarParameter() {defaultValue=1, parameterName=LerpToFlipbook, sliderMax=1}
  )
  n13 = LinearInterpolate(
   A=Saturate(None=Power(Base=@n2))
   B=Add(
    A=Multiply(A=@n3, B=@n7)
    B=MaterialFunctionCall!(Low=@n12, High=@n12, Epic=@n12, --------, Mobile=@n5.R, Medium=@n5.R) {materialFunction=/FortniteBaseMaterials/Functions/MF_QualitySwitch_Vector}.Result
   )
   Alpha=NamedRerouteUsage()
  )
  Add(A=@n13, B=NamedRerouteUsage())
  Frac(None=Divide(A=Time()) {constB=10})
  n14 = Constant4Vector() {constant=<rgba 0.9 1 0 1>}
  n15 = MaterialFunctionCall(Input=Clamp(None=NamedRerouteUsage()) {minDefault=0.9}, Input Low=@n14.R, Input High=@n14.R, Target Low=@n14.R, Target High=@n14.R) {materialFunction=&RemapValueRange}
  LinearInterpolate(Alpha=Saturate(None=Sine(None=@n15.Result))) {constB=2, desc=Final Blink}
  Multiply()
  n16 = MaterialFunctionCall(In=NamedRerouteUsage(), Power=Constant() {r=2}) {materialFunction=&SCurve}
  NamedRerouteDeclaration(None=Add(A=@n13, B=Multiply(A=@n13, B=@n16.Result))) {name=Zordon Face, nodeColor=<rgba 1 0 0.4922 1>}
  n17 = ObjectPositionWS()
  n18 = Sine(None=Time()) {period=6.28}
  n19 = Add(
   A=@n17
   B=MaterialFunctionCall(Z=Multiply(A=@n18) {constB=4}) {materialFunction=&MakeFloat3}.Result
  )
  n20 = Divide(
   A=Multiply(
    A=ScalarParameter() {defaultValue=0.6, parameterName=UVRadialScale}
    B=MaterialFunctionCall() {materialFunction=&ObjectScale}.Scale XYZ
   )
   B=MaterialFunctionCall() {materialFunction=/Engine/Functions/Engine_MaterialFunctions02/Utility/FOV}."FOV "
  )
  n21 = ScalarParameter() {defaultValue=0.2, parameterName=Face_OffsetY}
  NamedRerouteDeclaration(None=Add(
    A=MaterialFunctionCall!(Position=@n19, UVRadialScale=@n20) {materialFunction=/Game/Characters/Player/Common/MaterialFunctions/Superlevel/CharacterStabilizedScreenSpaceUVs}.StabilizedUVs
    B=MaterialFunctionCall(Y=@n21) {materialFunction=&MakeFloat2}.Result
   )) {name=FloatStablizedUV, nodeColor=<rgba 0 1 0.2031 1>}
  MaterialExpressionNamedRerouteUsage_45 = NamedRerouteUsage()
  MaterialExpressionTime_8 = Time()
  n22 = Custom!(uv=@MaterialExpressionNamedRerouteUsage_45, Time=@MaterialExpressionTime_8) {
   code="float3 col = float3(1.0,1.0,1.0);\t\t// Drop Colour\r\nuv.x = uv.x*1.0;\t\t\t\t\t\t// H-Count\r\nfloat dx = frac(uv.x);\r\nuv.x = floor(uv.x);\r\nfloat t =  Time*1.0;\r\nuv.y *= 0.02;\t\t\t\t\t\t\t// stretch\r\nfloat o=sin(uv.x*215.4);\t\t\t\t// offset\r\nfloat s=cos(uv.x*33.1)*.3 +.7;\t\t\t// speed\r\nfloat trail = lerp(90.0,20.0,s);\t\t\t// trail length\r\nfloat yv = frac(uv.y + t*s + o) * trail;\r\nyv = 1.0/yv;\r\nyv = smoothstep(0.0,1.0,yv*yv);\r\nyv = sin(yv*PI)*(s*5.0);\r\nfloat d2 = sin(dx*PI);\r\nyv *= d2*d2;\r\ncol = col*yv;\r\nreturn yv;"
   description=LigtningScan1
   inputs [
    {inputName=uv, input {expression=&self:MaterialExpressionNamedRerouteUsage_45}}
    {inputName=Time, input {expression=&self:MaterialExpressionTime_8}}
   ]
  }
  n23 = ConstantBiasScale(None=MaterialFunctionCall(Float=Time()) {materialFunction=&MF_Hash}.Float)
  n24 = NamedRerouteUsage()
  n25 = NamedRerouteUsage()
  n26 = Time()
  n27 = NamedRerouteUsage()
  n28 = TextureSample(UVs=Panner(Coordinate=Divide(A=@n25) {constB=2.2}, Time=@n26) {speedX=0.01}, Tex=@n27) {samplerSource=SSM_Wrap_WorldGroupSettings}
  n29 = Constant3Vector() {constant=<rgba 0.5 -0.5 0.6 0>}
  n30 = Multiply(A=@n28.R, B=@n29.R) {constB=0.1}
  n31 = Frac(None=Multiply(A=Add(A=@n24, B=@n30)) {constB=2})
  n32 = ComponentMask(None=Subtract(A=Multiply(A=@n31) {constB=2})) {r=true}
  n33 = Multiply(A=@n22, B=Divide(A=LinearInterpolate(Alpha=@n23) {constA=0.1, constAlpha=0.2}, B=Abs(None=@n32)))
  MaterialExpressionAdd_13 = Add(A=@MaterialExpressionNamedRerouteUsage_45, B=Constant2Vector() {r=5}) {constB=5, desc=offset}
  n34 = Custom(uv=@MaterialExpressionAdd_13, Time=@MaterialExpressionTime_8) {
   code="float3 col = float3(1.0,1.0,1.0);\t\t// Drop Colour\r\nuv.x = uv.x*1.0;\t\t\t\t\t\t// H-Count\r\nfloat dx = frac(uv.x);\r\nuv.x = floor(uv.x);\r\nfloat t =  Time*1.0;\r\nuv.y *= 0.02;\t\t\t\t\t\t\t// stretch\r\nfloat o=sin(uv.x*215.4);\t\t\t\t// offset\r\nfloat s=cos(uv.x*33.1)*.3 +.7;\t\t\t// speed\r\nfloat trail = lerp(90.0,20.0,s);\t\t\t// trail length\r\nfloat yv = frac(uv.y + t*s + o) * trail;\r\nyv = 1.0/yv;\r\nyv = smoothstep(0.0,1.0,yv*yv);\r\nyv = sin(yv*PI)*(s*5.0);\r\nfloat d2 = sin(dx*PI);\r\nyv *= d2*d2;\r\ncol = col*yv;\r\nreturn yv;"
   description=LigtningScan1
   inputs [
    {inputName=uv, input {expression=&self:MaterialExpressionAdd_13}}
    {inputName=Time, input {expression=&self:MaterialExpressionTime_8}}
   ]
  }
  n35 = Multiply(A=@n30) {constB=2}
  n36 = Frac(None=Multiply(A=Add(A=@n24, B=@n35)))
  n37 = ComponentMask(None=Subtract(A=Multiply(A=@n36) {constB=2})) {r=true}
  n38 = Multiply(A=@n34, B=Divide(A=LinearInterpolate(Alpha=@n23) {constA=0.1, constAlpha=0.2}, B=Abs(None=@n37)))
  MaterialExpressionAdd_30 = Add(A=@MaterialExpressionNamedRerouteUsage_45, B=Constant2Vector() {r=2}) {constB=5, desc=offset}
  n39 = Custom(uv=@MaterialExpressionAdd_30, Time=@MaterialExpressionTime_8) {
   code="float3 col = float3(1.0,1.0,1.0);\t\t// Drop Colour\r\nuv.x = uv.x*1.0;\t\t\t\t\t\t// H-Count\r\nfloat dx = frac(uv.x);\r\nuv.x = floor(uv.x);\r\nfloat t =  Time*0.4;\r\nuv.y *= 0.02;\t\t\t\t\t\t\t// stretch\r\nfloat o=sin(uv.x*215.4);\t\t\t\t// offset\r\nfloat s=cos(uv.x*33.1)*.3 +.7;\t\t\t// speed\r\nfloat trail = lerp(90.0,20.0,s);\t\t\t// trail length\r\nfloat yv = frac(uv.y + t*s + o) * trail;\r\nyv = 1.0/yv;\r\nyv = smoothstep(0.0,1.0,yv*yv);\r\nyv = sin(yv*PI)*(s*5.0);\r\nfloat d2 = sin(dx*PI);\r\nyv *= d2*d2;\r\ncol = col*yv;\r\nreturn yv;"
   description=LigtningScan2
   inputs [
    {inputName=uv, input {expression=&self:MaterialExpressionAdd_30}}
    {inputName=Time, input {expression=&self:MaterialExpressionTime_8}}
   ]
  }
  n40 = TextureSample(UVs=Panner(Coordinate=Divide(A=@n25), Time=@n26) {speedX=0.02}, Tex=@n27) {samplerSource=SSM_Wrap_WorldGroupSettings}
  n41 = Multiply(A=Add(A=@n24, B=Multiply(A=@n40.R, B=@n29.G) {constB=0.1})) {constB=2}
  n42 = Subtract(A=Multiply(A=Frac(None=@n41)) {constB=2})
  n43 = Divide(A=LinearInterpolate(Alpha=@n23) {constA=0.1, constAlpha=0.2}, B=Abs(None=ComponentMask(None=@n42) {r=true}))
  NamedRerouteDeclaration(None=Add(A=Add(A=@n33, B=@n38), B=Multiply(A=@n39, B=@n43))) {name=SecondaryLightning, nodeColor=<rgba 0 0.3438 1 1>}
  NamedRerouteDeclaration(None=TextureObject() {texture=/Game/Global/Textures/Default/Noise/T_godray_Noise}) {name=Noise_TEX, nodeColor=<rgba 0 0.4232 1 1>}
  n44 = OneMinus(None=MaterialFunctionCall(
    Axis=Constant3Vector() {constant=<rgba 0 0 1 0>}
    Exponent=ScalarParameter() {defaultValue=1, parameterName=CylinderDistortionFresnelMaskExponent}
   ) {materialFunction=&AxisAlignedFresnel}.Fresnel)
  NamedRerouteDeclaration(None=Saturate(None=@n44)) {name=Tank Fresnel, nodeColor=<rgba 0 0.5312 1 1>}
  NamedRerouteDeclaration(None=MaterialFunctionCall(Position=@n17, UVRadialScale=@n20) {materialFunction=&CharacterStabilizedScreenSpaceUVs}.StabilizedUVs) {name=StablizedUV, nodeColor=<rgba 0 1 0.2031 1>}
  n45 = Add(A=NamedRerouteUsage(), B=NamedRerouteUsage())
  n46 = OneMinus(None=NamedRerouteUsage())
  n47 = LinearInterpolate(A=Saturate(None=@n45), B=@n45, Alpha=@n46)
  n48 = Multiply(A=Multiply(A=NamedRerouteUsage(), B=Multiply(A=@n47, B=@n46)), B=NamedRerouteUsage())
  n49 = NamedRerouteUsage()
  NamedRerouteDeclaration(None=Multiply(A=Multiply(A=@n48, B=@n49), B=NamedRerouteUsage())) {name=Filler, nodeColor=<rgba 0.6641 1 0 1>}
  n50 = ScalarParameter() {parameterName=Interaction_Progession, sliderMax=1, sortPriority=0}
  n51 = Constant4Vector() {constant=<rgba 0 1 0.24 0.87>}
  NamedRerouteDeclaration(None=MaterialFunctionCall(Input=@n50, Input Low=@n51.R, Input High=@n51.R, Target Low=@n51.R, Target High=@n51.R) {materialFunction=&RemapValueRange}.Result) {name=ProgressionData, nodeColor=<rgba 1 0 0.0937 1>}
  NamedRerouteDeclaration(None=Add(A=@n44) {constB=0.1}) {name=Tank_CenterDarken_Fresnel, nodeColor=<rgba 0 0.5312 1 1>}
  n52 = Constant4Vector() {constant=<rgba 0 0.2 0 1>}
  n53 = MaterialFunctionCall!(Mask In=NamedRerouteUsage(), Amount=NamedRerouteUsage(), Range=@n52.G, Start Offset=@n52.G, Sharpness=@n52.G) {materialFunction=/Game/Global/MaterialFunctions/MF_MaskShifting}
  NamedRerouteDeclaration(None=Saturate(None=Multiply(A=@n53.Result) {constB=-1})) {name=FillMsk, nodeColor=<rgba 0 0.9297 1 1>}
  NamedRerouteDeclaration(None=TextureObject() {texture=/ToyFoxGameplay/Assets/Texture/T_ToyFox_PedestalFace_Flipbook}) {name=FlipbookTex, nodeColor=<rgba 0 0.4232 1 1>}
  NamedRerouteDeclaration(None=@n50) {name=RawProgession, nodeColor=<rgba 0 0.2734 1 1>}
  n54 = NamedRerouteUsage()
  n55 = Constant2Vector() {g=6, r=2}
  n56 = TextureSample(UVs=Panner(Coordinate=Multiply(A=@n54, B=@n55) {constB=2}) {speedX=0.1, speedY=-0.5}) {
   samplerSource=SSM_Wrap_WorldGroupSettings
   samplerType=SAMPLERTYPE_LinearColor
   texture=/Game/Global/Textures/Default/Flowmaps/T_TorusFlowmap
  }
  NamedRerouteDeclaration(None=Multiply(A=AppendVector(A=@n56.R, B=@n56.R)) {constB=0.02}) {name=Face_Distort, nodeColor=<rgba 0 1 0.1328 1>}
  n57 = Time()
  NamedRerouteDeclaration(None=ConstantBiasScale(None=MaterialFunctionCall(Float=@n57) {materialFunction=&MF_Hash}.Float)) {name=Flicker, nodeColor=<rgba 0 1 0.5313 1>}
  MaterialExpressionAdd_31 = Add(A=@MaterialExpressionNamedRerouteUsage_45, B=Constant2Vector() {r=8}) {constB=5, desc=offset}
  n58 = Custom(uv=@MaterialExpressionAdd_31, Time=@MaterialExpressionTime_8) {
   code="float3 col = float3(1.0,1.0,1.0);\t\t// Drop Colour\r\nuv.x = uv.x*1.0;\t\t\t\t\t\t// H-Count\r\nfloat dx = frac(uv.x);\r\nuv.x = floor(uv.x);\r\nfloat t =  Time*0.3;\r\nuv.y *= 0.01;\t\t\t\t\t\t\t// stretch\r\nfloat o=sin(uv.x*215.4);\t\t\t\t// offset\r\nfloat s=cos(uv.x*33.1)*.3 +.7;\t\t\t// speed\r\nfloat trail = lerp(90.0,20.0,s);\t\t\t// trail length\r\nfloat yv = frac(uv.y + t*s + o) * trail;\r\nyv = 1.0/yv;\r\nyv = smoothstep(0.0,1.0,yv*yv);\r\nyv = sin(yv*PI)*(s*5.0);\r\nfloat d2 = sin(dx*PI);\r\nyv *= d2*d2;\r\ncol = col*yv;\r\nreturn yv;"
   description=LigtningScan3
   inputs [
    {inputName=uv, input {expression=&self:MaterialExpressionAdd_31}}
    {inputName=Time, input {expression=&self:MaterialExpressionTime_8}}
   ]
  }
  NamedRerouteDeclaration(None=Add(A=Add(A=@n22, B=@n39), B=@n58)) {name=LightScan, nodeColor=<rgba 0 0.2266 1 1>}
  n59 = CameraVectorWS()
  n60 = VertexNormalWS()
  n61 = Constant() {r=0.5}
  n62 = SkyLightEnvMapSample(
   Direction=Add(A=Multiply(A=@n59) {constB=-1}, B=Multiply(A=@n60, B=@n61))
   Roughness=ScalarParameter() {defaultValue=0.2, parameterName=SkylightEnvMapRoughness}
  )
  n63 = LinearInterpolate(
   A=VectorParameter() {parameterName=CubeMap_Darken_Color, defaultValue=<rgba 0.1579 0.2901 0.3698 1>}.RGB
   B=VectorParameter() {parameterName=CubeMap_Filled_Color, defaultValue=<rgba 0.1609 0.4845 0.776 1>}.RGB
   Alpha=NamedRerouteUsage()
  )
  n64 = Multiply(A=NamedRerouteUsage()) {constB=0.5}
  NamedRerouteDeclaration(None=Multiply(A=Multiply(A=@n62, B=@n63), B=OneMinus(None=@n64))) {name=Cubemap, nodeColor=<rgba 1 0 0.7266 1>}
  n65 = Subtract(A=Multiply(A=NamedRerouteUsage()) {constB=2})
  n66 = SmoothStep(Min=LinearInterpolate(Alpha=NamedRerouteUsage()) {constA=0.4}, Value=Abs(None=ComponentMask(None=@n65) {r=true})) {constMax=0, constMin=0.5}
  NamedRerouteDeclaration(None=Multiply(A=Power(Base=@n66) {constExponent=4}) {constB=40}) {name=BackProcessbar, nodeColor=<rgba 0.6406 1 0 1>}
  NamedRerouteDeclaration(None=Constant3Vector() {constant=<rgba 0.2 0.3 0.8 0>}) {name=LightningColor, nodeColor=<rgba 0.1981 0.2961 0.7913 1>}
  n67 = NamedRerouteUsage()
  n68 = NamedRerouteUsage()
  n69 = ComponentMask(None=NamedRerouteUsage()) {r=true}
  n70 = OneMinus(None=@n69)
  NamedRerouteDeclaration(None=BumpOffset(Coordinate=Add(A=@n67, B=@n68), Height=Min(A=@n69, B=@n70)) {heightRatio=0.08}) {name=BentUV, nodeColor=<rgba 0.1262 0.2642 1 1>}
  NamedRerouteDeclaration(None=TextureObject() {texture=/ToyFoxGameplay/Assets/Texture/T_ToyFox_PedestalFace}) {name=Face_TEX, nodeColor=<rgba 0 0.4232 1 1>}
  n71 = NamedRerouteUsage()
  n72 = TextureSample(UVs=Panner(Coordinate=Multiply(A=@n71) {constB=2}) {speedX=0.5}, Tex=NamedRerouteUsage()) {samplerSource=SSM_Wrap_WorldGroupSettings}
  n73 = TextureSample(UVs=Add(A=NamedRerouteUsage(), B=Multiply(A=@n72.R) {constB=0.05}), Tex=NamedRerouteUsage()) {samplerSource=SSM_Clamp_WorldGroupSettings}
  n74 = Multiply(A=Multiply(A=Multiply(A=@n73.B, B=@n73.B)) {constB=4}, B=@n72.R)
  n75 = NamedRerouteUsage()
  NamedRerouteDeclaration(None=Multiply(A=Multiply(A=@n74, B=@n75), B=NamedRerouteUsage())) {name=FaceElectric, nodeColor=<rgba 0.7812 1 0 1>}
  n76 = Add(A=NamedRerouteDeclaration(None=MaterialFunctionCall() {materialFunction=&BoundingBoxBased_0-1_UVW}.B) {name=BoundingZ, nodeColor=<rgba 0.4297 0 1 1>}) {constB=-0.048}
  n77 = Min(A=@n76, B=OneMinus(None=@n76))
  n78 = Constant4Vector() {constant=<rgba 0.15 1 0 1>}
  NamedRerouteDeclaration(None=Saturate(None=MaterialFunctionCall(Mask In=@n77, Amount=@n78.R, Range=@n78.R, Start Offset=@n78.R, Sharpness=@n78.R) {materialFunction=&MF_MaskShifting}.Result)) {name=TankMask, nodeColor=<rgba 0.8984 0 1 1>}
  n79 = Divide(A=Time())
  n80 = Add(A=NamedRerouteUsage(), B=LinearInterpolate(Alpha=Frac(None=@n79)) {constA=1, constB=-1})
  n81 = OneMinus(None=@n80)
  n82 = SmoothStep(Value=Multiply(A=Min(A=@n81, B=@n80)) {constB=0.5}) {constMax=0.3, constMin=0.2}
  n83 = LinearInterpolate(A=@n82, B=Multiply(A=@n82) {constB=0.2}, Alpha=NamedRerouteUsage())
  n84 = SmoothStep(Value=NamedRerouteUsage()) {constMax=0.1, constValue=0.08}
  n85 = Constant() {r=0.2}
  n86 = NamedRerouteUsage()
  n87 = MaterialFunctionCall(UV Coord=Add(A=@n86, B=Constant2Vector() {g=1.3})) {materialFunction=&MF_PolarCoordinates}
  n88 = Power(Base=Divide(A=@n85, B=ComponentMask(None=@n87.Result) {r=true})) {constExponent=3}
  n89 = MaterialFunctionCall(UV Coord=Subtract(A=@n86, B=Constant2Vector() {g=1})) {materialFunction=&MF_PolarCoordinates}
  n90 = Power(Base=Divide(A=@n85, B=ComponentMask(None=@n89.Result) {r=true})) {constExponent=3}
  n91 = Add(A=Multiply(A=Multiply(A=@n83) {constB=2}, B=OneMinus(None=@n84)), B=Multiply(A=Add(A=@n88, B=@n90)) {constB=4})
  n92 = NamedRerouteUsage()
  n93 = NamedRerouteUsage()
  n94 = VectorParameter() {parameterName=Face_Tint, defaultValue=<rgba 0.3005 0.6638 0.9948 1>}
  n95 = NamedRerouteUsage()
  n96 = Multiply(A=NamedRerouteUsage(), B=OneMinus(None=@n95))
  n97 = VectorParameter() {parameterName=Container_Fresnel_Color, defaultValue=<rgba 0.2951 0.7146 1 1>}
  n98 = NamedRerouteUsage()
  n99 = Multiply(A=@n98, B=NamedRerouteUsage())
  n100 = Fmod(A=Time() {bOverride_Period=true, period=8}, B=Constant() {r=8})
  n101 = NamedRerouteUsage()
  n102 = Constant2Vector() {g=-0.4, r=2}
  n103 = MaterialFunctionCall(Size=Constant2Vector() {g=2, r=8}, Selector=@n100, UV=Add(A=Multiply(A=@n101, B=@n102), B=Constant2Vector() {g=-0.3, r=1}) {constB=0.5}) {materialFunction=&MF_FlipbookUV}
  n104 = Constant2Vector() {g=1, r=-1}
  n105 = Constant2Vector() {r=0.5}
  n106 = Add(A=Multiply(A=@n103.FlipbookUV1, B=@n104) {constB=2}, B=@n105) {constB=0.5}
  n107 = Add(A=Multiply(A=@n103.FlipbookUV2, B=@n104) {constB=2}, B=@n105) {constB=0.5}
  n108 = LinearInterpolate(
   A=LinearInterpolate(
    A=TextureSample(UVs=@n103.FlipbookUV1) {samplerSource=SSM_Wrap_WorldGroupSettings, texture=/ToyFoxGameplay/Assets/Texture/T_Toyfox_Pedestal_LightningBlot}.R
    B=TextureSample(UVs=@n103.FlipbookUV2) {samplerSource=SSM_Wrap_WorldGroupSettings, texture=&T_Toyfox_Pedestal_LightningBlot}.R
    Alpha=@n103.Blend2
   )
   B=LinearInterpolate(
    A=TextureSample(UVs=@n106) {samplerSource=SSM_Wrap_WorldGroupSettings, texture=&T_Toyfox_Pedestal_LightningBlot}.R
    B=TextureSample(UVs=@n107) {samplerSource=SSM_Wrap_WorldGroupSettings, texture=&T_Toyfox_Pedestal_LightningBlot}.R
    Alpha=@n103.Blend2
   )
   Alpha=Step(Y=ComponentMask(None=@n101) {r=true}) {constX=0.5}
  )
  n109 = Multiply(A=NamedRerouteUsage(), B=Constant2Vector() {g=-0.2, r=2})
  n110 = Constant2Vector() {g=-0.15, r=1}
  n111 = TextureSample(UVs=MaterialFunctionCall(Size=Constant2Vector() {g=1, r=8}, Selector=@n100, UV=Add(A=@n109, B=@n110) {constB=0.5}) {materialFunction=&MF_FlipbookUV}.FlipbookUV1) {
   samplerSource=SSM_Wrap_WorldGroupSettings
   samplerType=SAMPLERTYPE_Masks
   texture=/ToyFoxGameplay/Assets/Texture/T_ToyFox_Pedestal_LightningBlot_SDF
  }
  n112 = SmoothStep(Value=@n111.R) {constMax=0.95, constMin=0.75}
  n113 = Power(Base=Add(A=Power(Base=@n112) {constExponent=4}, B=Multiply(A=@n111.R) {constB=0.5})) {constExponent=4}
  n114 = MaterialFunctionCall!(
   Low _ PC Low and Med=@n108
   "High _ XB1, PS4"=@n108
   "Epic _ PC High, XSX, PS5"=@n108
   --------
   "Mobile Low _ PC Perf, Android"=@n113
   Mobile Med _ Switch=@n113
  ) {materialFunction=/Game/Global/MaterialFunctions/TechArt/MF_Fortnite_QualitySwitch_Scalar}
  n115 = NamedRerouteUsage()
  n116 = Multiply(
   A=Multiply(A=Add(A=@n98, B=@n99), B=NamedRerouteUsage())
   B=LinearInterpolate(A=@n114.Result, B=Multiply(A=@n114.Result, B=@n115), Alpha=NamedRerouteUsage()) {constB=0}
  )
  n117 = LinearInterpolate(Alpha=NamedRerouteUsage()) {constA=20, constB=40}
  n118 = Add(
   A=Add(A=Multiply(A=@n93, B=@n94.RGB), B=LinearInterpolate(A=@n96, B=@n97.RGB, Alpha=@n95) {constB=0})
   B=Multiply(A=NamedRerouteUsage(), B=Multiply(A=@n116, B=@n117) {constB=20}) {constB=0.25}
  )
  n119 = NamedRerouteUsage()
  n120 = NamedRerouteUsage()
  n121 = Multiply(A=Add(A=Multiply(A=@n91, B=@n92) {constB=10}, B=Add(A=@n118, B=@n119)), B=Saturate(None=Power(Base=@n120) {constExponent=0.4}))
  n122 = Constant() {r=1.5}
  n123 = Constant() {r=2}
  n124 = MaterialFunctionCall(Value_Morning=@n122, Value_Day=@n122, Value_Evening=@n123, Value_Night=@n123) {materialFunction=&Scalar_Time_of_Day}
  n125 = Constant() {r=4}
  NamedRerouteDeclaration(None=MaterialFunctionCall!(
    Base=Multiply(A=@n121, B=@n124.Result)
    Add=MaterialFunctionCall(In=@n15.Result, Power=@n125) {materialFunction=&SCurve}.Result
   ) {materialFunction=/Engine/Functions/Engine_MaterialFunctions02/Math/MultiplyAdd}.Result) {name=Final_Emissive, nodeColor=<rgba 0 1 0.0156 1>}
  n126 = TextureSample(UVs=Panner(Coordinate=Divide(A=@n25) {constB=1.8}, Time=@n26) {speedX=0.03}, Tex=@n27) {samplerSource=SSM_Wrap_WorldGroupSettings}
  n127 = Multiply(A=@n126.R, B=@n29.B) {constB=0.1}
  n128 = Frac(None=Multiply(A=Add(A=@n127, B=@n24)) {constB=4})
  n129 = ComponentMask(None=Subtract(A=Multiply(A=@n128) {constB=2})) {r=true}
  n130 = Power(Base=Divide(A=LinearInterpolate(Alpha=@n23) {constA=1, constAlpha=0.2, constB=4}, B=Abs(None=@n129))) {constExponent=1}
  n131 = Saturate(None=Subtract(A=@n28.R, B=Min(A=@n40.R, B=@n126.R)))
  NamedRerouteDeclaration(None=Multiply(A=@n58, B=Multiply(A=@n130, B=@n131))) {name=ToppingLightning, nodeColor=<rgba 0 0.5312 1 1>}
  NamedRerouteUsage()
  NamedRerouteUsage()
 ]
}
