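# Master material for the ToyFox supply drop meteor, layering animated glowing cracks and heat over the Fortnite GigaMix base material.
# Hard triplanar mask textures, heat spot and impulse masks feed a curve-atlas emissive, while the heat mask also alters base color, roughness, specular and metallic.
# CurrentStage and a Blueprint-supplied game time drive staged animation, with shake and curve-driven stomp computed procedurally.
# DirectFX, Veneer, biome, RVT and damage functions follow, plus a simplified low-quality mobile path.
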Material /ToyFoxGameplay/Assets/Material/M_ToyFox_SupplyDrops_Meteor
graph {
 properties {
  bAutomaticallySetUsageInEditor=false
  bTangentSpaceNormal=false
  bUseMaterialAttributes=true
  bUsedWithInstancedStaticMeshes=true
  bUsedWithNanite=true
  bUsedWithSplineMeshes=true
  ditheredLODTransition=true
  pixelDepthOffsetMode=PDOM_Legacy
  shadingModel=MSM_FromMaterialExpression
  usedShadingModels=Default Lit | Subsurface | Clear Coat | Two Sided Foliage
 }
 nodes [
  n1 = MaterialFunctionCall() {materialFunction=/Game/Global/Base4Layers_2022/MaterialFunctions/MF_GigaMix_2023}
  MaterialExpressionReroute_70 = Reroute(None=@n1.Result)
  n2 = ScalarParameter() {
   desc=Documentation link
   group="0 (Expand)  Documentation (RMB - Copy display name) and Debug view"
   parameterName=https://confluence.it.epicgames.com/display/FORT/Fortnite+Base+Material+2022
  }
  n3 = SetMaterialAttributes!(MaterialAttributes=@MaterialExpressionReroute_70.MaterialAttributes, World Position Offset=Multiply(A=@n2)) {attributeSetTypes [#1]}
  MaterialExpressionReroute_71 = Reroute(None=@n3)
  Composite!(None=@MaterialExpressionReroute_71.Output) {desc=Transform to World in case we will use RVT after this point, subgraphName=Tangent to World}
  Composite!(MaterialAttributes=@n1.Result) {subgraphName=Documentation Link}
  Composite() {desc=Game Shared WPO and Fire Effect, subgraphName=WPO And Shared Functions}
  MaterialExpressionReroute_1 = Reroute(None=@n15.RVT H)
  MaterialExpressionReroute_3 = Reroute(None=@MaterialExpressionReroute_71.Output)
  n4 = GetMaterialAttributes(None=@MaterialExpressionReroute_3.Input) {attributeGetTypes [#2]}
  n5 = Transform(None=@n4.Normal)
  n6 = LinearInterpolate(A=Multiply(A=@n5) {constB=-1}, B=@n5, Alpha=Saturate(None=TwoSidedSign())) {desc="Fixes normals\r\non backfaces"}
  n7 = SetMaterialAttributes!(
   MaterialAttributes=@n4.MaterialAttributes
   Normal=StaticSwitchParameter(True=MaterialProxyReplace(Realtime=@n6, MaterialProxy=@n4.Normal), False=@n6) {parameterName=Force TangentSpaceNormalsFor HLOD ProxyReplace}
  ) {attributeSetTypes [#2]}
  MaterialExpressionReroute_2 = Reroute(None=@n7)
  n8 = GetMaterialAttributes(None=@MaterialExpressionReroute_2.Output) {attributeGetTypes [
    #3
    #4
    #5
    #2
    #6
   ]}
  n9 = NamedRerouteUsage()
  n10 = Add(A=@n8.Roughness) {constB=0.2}
  n11 = Add(A=Multiply(A=NamedRerouteUsage()) {constB=4}, B=@n9)
  n12 = SetMaterialAttributes!(
   MaterialAttributes=@n8.MaterialAttributes
   Emissive Color=NamedRerouteUsage()
   Base Color=LinearInterpolate(A=@n8.Base Color, B=Multiply(A=@n8.Base Color) {constB=0.5}, Alpha=@n9)
   Roughness=LinearInterpolate(A=Saturate(None=@n10), Alpha=Saturate(None=@n11))
   Specular=LinearInterpolate(A=@n8.Specular, Alpha=@n9) {constB=0.1}
   Normal=NamedRerouteUsage()
   Metallic=LinearInterpolate(A=@n8.Metallic, Alpha=@n9) {constB=0}
   World Position Offset=Add(A=NamedRerouteUsage(), B=NamedRerouteUsage())
  ) {attributeSetTypes [
    #7
    #3
    #4
    #5
    #2
    #6
    #1
   ]}
  n13 = MaterialFunctionCall!(In=MaterialFunctionCall!(In=@n12) {materialFunction=/Game/ArtTools/Tech_Art/DirectFX/Assets/MaterialFunctions/MF_DirectFX}.Result) {materialFunction=/Game/ArtTools/Tech_Art/Veneer/MaterialFunctions/MF_Veneer}
  MaterialExpressionReroute_8 = Reroute(None=@n13.Result)
  n14 = MaterialFunctionCall!(AO, RVT_Height=@MaterialExpressionReroute_1."RVT_Height (S)", For Buildings=@MaterialExpressionReroute_8."For Moss (MA)", For Moss) {materialFunction=/Game/Environments/Asteria/Rocks/Materials/MaterialFunctions/MF_JungleBiome}
  MaterialExpressionReroute_0 = Reroute(None=@n14.Buildings Jungl)
  n15 = MaterialFunctionCall!(
   In=@MaterialExpressionReroute_0.Result
   "RVT Height Map (Roughness)"=GetMaterialAttributes(None=@n13.Result) {attributeGetTypes [#4]}.Roughness
  ) {materialFunction=/Game/Global/Base4Layers_2022/MaterialFunctions/MF_RVTBlending}
  n16 = MaterialFunctionCall!(
   In=MaterialFunctionCall(Material_In=@n15.Result, UseAdditiveWPO=StaticBool() {value=true}) {materialFunction=&MF_FN_DestructibleDamage}.Material_Out
   UseSphericalMappingOverride
  ) {materialFunction=/Game/Packages/Fortress_SharedMaterials/Base_Material_Functions/MF_CurieFallbackEffect}
  Composite!(InputPin=@n16.Result) {subgraphName=Mobile Opt}
  Composite!("For Moss (MA)"=@n13.Result, "RVT_Height (S)"=@n15.RVT H) {subgraphName="Seasonal "}
  LocalPosition() {localOrigin=InstancePreSkinning}
  MaterialExpressionReroute_7 = Reroute(None=@n16.Result)
  n17 = GetMaterialAttributes(None=@MaterialExpressionReroute_7.InputPin)
  n18 = Constant() {r=0.5}
  n19 = Constant() {r=0.8}
  n20 = ShadingPathSwitch(
   Default=@MaterialExpressionReroute_7.InputPin
   Mobile=QualitySwitch(
    Default=@MaterialExpressionReroute_7.InputPin
    Low=SetMaterialAttributes!(MaterialAttributes=@n17.MaterialAttributes, Specular=@n18, Roughness=@n19) {attributeSetTypes [#5, #4]}
    High=@MaterialExpressionReroute_7.InputPin
    Medium=@MaterialExpressionReroute_7.InputPin
    Epic=@MaterialExpressionReroute_7.InputPin
   )
  )
  MaterialExpressionReroute_6 = Reroute(None=@n20)
  MaterialFunctionCall(MA In=@MaterialExpressionReroute_6.Output) {materialFunction=&MF_CreativeMoveTool}
  n21 = Divide(A=Multiply(A=NamedRerouteUsage()) {constB=-1}, B=Abs(None=Constant() {r=32})) {constB=1024}
  n22 = VectorParameter() {group=Meteor, parameterName=Triplannar3DTilling, defaultValue=<rgba 1 1 1 1>}
  n23 = Divide(A=WorldPosition() {worldPositionShaderOffset=WPT_ExcludeAllShaderOffsets}.XYZ) {constB=128}
  n24 = Constant3Vector() {constant=<rgba 1 1 0 0>}
  n25 = MaterialFunctionCall(Float3=Add(A=Multiply(A=@n21, B=@n22.RGB), B=Multiply(A=@n23, B=@n24))) {materialFunction=&BreakOutFloat3Components}
  n26 = NamedRerouteUsage()
  MaterialExpressionComponentMask_9 = ComponentMask(None=@n26) {r=true}
  n27 = Custom!(x=@MaterialExpressionComponentMask_9) {
   code="return float(x < 0) ? -1 : 1;"
   description=StepReturn
   inputs [{inputName=x, input {expression=&self:MaterialExpressionComponentMask_9}}]
  }
  MaterialExpressionComponentMask_8 = ComponentMask(None=@n26) {g=true}
  n28 = Custom(x=@MaterialExpressionComponentMask_8) {
   code="return float(x < 0) ? 1 : -1;"
   description=StepReturn
   inputs [{inputName=x, input {expression=&self:MaterialExpressionComponentMask_8}}]
  }
  MaterialExpressionComponentMask_7 = ComponentMask(None=@n26) {b=true}
  n29 = Custom(x=@MaterialExpressionComponentMask_7) {
   code="return float(x < 0) ? -1 : 1;"
   description=StepReturn
   inputs [{inputName=x, input {expression=&self:MaterialExpressionComponentMask_7}}]
  }
  n30 = MaterialFunctionCall(Float3=MaterialFunctionCall(X=@n27, Y=@n28, Z=@n29) {materialFunction=&MakeFloat3}.Result) {materialFunction=&BreakOutFloat3Components}
  n31 = Abs(None=@n26)
  n32 = MaterialFunctionCall(Float3=@n31) {materialFunction=&BreakOutFloat3Components}
  n33 = Max(A=@n32.R, B=@n32.R)
  n34 = MaterialFunctionCall(Float3=Step(Y=Max(A=@n33, B=@n32.B), X=@n31)) {materialFunction=&BreakOutFloat3Components}
  n35 = LinearInterpolate(
   A=MaterialFunctionCall(X=Multiply(A=@n25.G, B=@n30.R), Y=@n25.B) {desc=Side, materialFunction=&MakeFloat2}.Result
   B=MaterialFunctionCall(X=Multiply(A=@n25.R, B=@n30.G), Y=@n25.B) {desc=Front, materialFunction=&MakeFloat2}.Result
   Alpha=@n34.G
  )
  n36 = MaterialFunctionCall(X=Multiply(A=@n25.R, B=@n30.B), Y=@n25.G) {desc=Top, materialFunction=&MakeFloat2}
  NamedRerouteDeclaration(None=ConstantBiasScale(None=LinearInterpolate(A=@n35, B=@n36.Result, Alpha=@n34.B))) {name=HardTriplannarUV, nodeColor=<rgba 0.0312 0 1 1>}
  n37 = SphereMask(A=LocalPosition().XYZ, B=Constant3Vector() {constant=<rgba -7.215 -25 100.124 0>}) {attenuationRadius=250, hardnessPercent=0}
  n38 = ShaderStageSwitch(PixelShader=VertexInterpolator(VS=@n37).PS, VertexShader=@n37)
  n39 = VectorParameter() {
   group=Meteor
   parameterName=HeatCrackMask
   sortPriority=33
   channelNames {a=Exp, b=StartOffset, g=Range, r=Amount}
   defaultValue=<rgba 1 1.5 0 3>
  }
  NamedRerouteDeclaration(None=MaterialFunctionCall!(
    Mask In=Power(Base=@n38, Exp=@n39.Exp) {constExponent=6}
    Amount=OneMinus(None=@n39.Amount)
    Range=@n39.Range
    Start Offset=@n39.Range
    Sharpness=Constant() {r=1}
   ) {materialFunction=/Game/Global/MaterialFunctions/MF_MaskShifting}.Result) {name=CreakArea_Msk, nodeColor=<rgba 1 0 0.75 1>}
  n40 = NamedRerouteUsage()
  NamedRerouteDeclaration(None=TextureSample(
    UVs=Multiply(A=@n40)
    Tex=TextureObject() {samplerType=SAMPLERTYPE_Normal, texture=/Game/Athena/Environments/Meteor/Textures/T_Meteor_Gravel_01_N}
   ) {samplerSource=SSM_Wrap_WorldGroupSettings}.RGB) {name=DetialNormal, nodeColor=<rgba 0.2 0.2 1 1>}
  n41 = NamedRerouteUsage()
  n42 = NamedRerouteUsage()
  n43 = NamedRerouteUsage()
  n44 = Add(A=Max(A=Multiply(A=@n41) {constB=0.1}, B=Multiply(A=@n42, B=@n43)), B=NamedRerouteUsage())
  n45 = VectorParameter() {group=Meteor, parameterName=HeatMaskTweak, channelNames {a=Exp, b=StartOffset, g=Range, r=Amount}, defaultValue=<rgba 0.544 0.5 0.04 2.5509>}
  n46 = Add(A=@n45.Amount, B=LinearInterpolate(Alpha=NamedRerouteUsage()) {constA=0.121, constB=0})
  n47 = Divide(A=NamedRerouteUsage())
  n48 = LinearInterpolate(Alpha=Saturate(None=Sine(None=@n47))) {constB=0.06}
  n49 = MaterialFunctionCall(
   Mask In=LinearInterpolate(A=Multiply(A=@n44) {constB=0.5}, B=@n44, Alpha=NamedRerouteUsage())
   Amount=OneMinus(None=Add(A=@n46, B=@n48))
   Range=@n45.Range
   Start Offset=Add(A=@n45.StartOffset) {constB=0.5}
   Sharpness=Constant() {r=1}
  ) {materialFunction=&MF_MaskShifting}
  n50 = NamedRerouteUsage()
  n51 = NamedRerouteUsage()
  n52 = NamedRerouteUsage()
  n53 = Sine(None=Divide(A=NamedRerouteUsage()))
  n54 = Add(
   A=Multiply(A=Multiply(A=@n50, B=@n51), B=LinearInterpolate(Alpha=@n52) {constA=0.5, constB=4})
   B=Multiply(A=Saturate(None=@n53), B=NamedRerouteUsage())
  )
  n55 = Add(A=NamedRerouteUsage(), B=Add(A=Multiply(A=@n49.Result, B=@n54), B=@n54))
  n56 = Saturate(None=Power(Base=@n49.Result, Exp=@n45.Exp) {constExponent=4})
  NamedRerouteDeclaration(None=Max(A=LinearInterpolate(A=@n55, B=@n49.Result, Alpha=@n56), B=NamedRerouteUsage())) {name=Final_Heat_Mask, nodeColor=<rgba 0 0.3906 1 1>}
  n57 = NamedRerouteUsage()
  n58 = TextureSample(UVs=Clamp(None=NamedRerouteUsage()) {minDefault=0.01}) {
   samplerSource=SSM_Clamp_WorldGroupSettings
   samplerType=SAMPLERTYPE_LinearColor
   texture=/ToyFoxGameplay/Assets/Texture/Curves/CA_ToyFox_CannonHeat
  }
  n59 = MaterialFunctionCall(Float3=RgbToHsv!(None=Multiply(A=@n57, B=@n58.RGB))) {materialFunction=&BreakOutFloat3Components}
  n60 = ScalarParameter() {defaultValue=0.92, group=Meteor, parameterName=HeatSeaturationPercentage}
  n61 = HsvToRgb!(None=MaterialFunctionCall(X=@n59.R, Y=Multiply(A=@n59.G, B=@n60), Z=@n59.R) {materialFunction=&MakeFloat3}.Result)
  n62 = VectorParameter() {group=Meteor, parameterName=MeteorEmissiveColorOverlay}
  NamedRerouteDeclaration(None=Multiply(A=Multiply(A=@n61, B=@n62.RGB), B=ScalarParameter() {defaultValue=10, group=Meteor, parameterName=MeteorHeatIntensity})) {name=OUT_Emissive, nodeColor=<rgba 0.7109 1 0 1>}
  n63 = Add(A=NamedRerouteUsage(), B=Constant3Vector() {desc=Object Center, constant=<rgba 7.215 25 -100.124 0>})
  n64 = Multiply(A=DotProduct(A=NamedRerouteUsage(), B=Normalize(VectorInput=@n63))) {constB=-1, desc=*-1 invert Dot Product}
  n65 = Multiply(
   A=LinearInterpolate(A=Length(None=@n63) {is=class MaterialExpressionLength}, B=Saturate(None=@n64), Alpha=Constant() {desc=0.5, r=0.5}) {desc=Distance}
   B=ScalarParameter() {defaultValue=0.3, group=Meteor, parameterName=HeatSpot Size Exponent, sortPriority=21}
  ) {desc=*FalloffDamp}
  n66 = Power(
   Base=Saturate(None=Divide(B=@n65))
   Exp=ScalarParameter() {defaultValue=2, group=Meteor, parameterName=HeatSpot Falloff Exponent, sortPriority=20}
  )
  n67 = Time()
  n68 = NamedRerouteUsage()
  n69 = Sine(None=Divide(A=Add(A=@n67, B=@n68)) {constB=4})
  n70 = Add(
   A=ScalarParameter() {defaultValue=4, group=Meteor, parameterName=HeatSpot Emissive Multiplier, sortPriority=19}
   B=LinearInterpolate(Alpha=ConstantBiasScale(None=@n69)) {constA=10, constB=18}
  )
  n71 = NamedRerouteUsage()
  n72 = ScalarParameter() {defaultValue=0.18, group=Meteor, parameterName=HeatSpot Emissive Min Intensity, sliderMax=1, sortPriority=15}
  NamedRerouteDeclaration(None=Add(A=Multiply(A=@n66, B=@n70), B=Multiply(A=@n71, B=@n72))) {name=HeatSpot, nodeColor=<rgba 1 0 0.3281 1>}
  NamedRerouteDeclaration(None=@n8.Normal) {name=Original_N, nodeColor=<rgba 1 0.9375 0 1>}
  NamedRerouteDeclaration(None=TextureSampleParameter2D(UVs=TextureCoordinate()) {parameterName=SpecularMasks, samplerType=SAMPLERTYPE_Masks, texture=&T_BlackMask}.B) {name=MeteroSpecularMask, nodeColor=<rgba 0.4297 0 1 1>}
  n73 = CameraVectorWS()
  n74 = DotProduct(A=VertexNormalWS(), B=@n73)
  NamedRerouteDeclaration(None=Power(Base=OneMinus(None=@n74))) {name=MeteorFresnel, nodeColor=<rgba 0 0.1797 1 1>}
  NamedRerouteDeclaration(None=@n23) {name=WSRandomOffset, nodeColor=<rgba 1 0.4922 0 1>}
  n75 = NamedRerouteUsage()
  n76 = Time() {period=1}
  n77 = NamedRerouteUsage()
  n78 = OneMinus(None=Divide(A=Add(A=@n76, B=@n77)) {constB=8})
  n79 = Constant4Vector() {constant=<rgba 0 0.05 0 0.1>}
  n80 = Sine(None=Multiply(A=MaterialFunctionCall(Mask In=@n75, Amount=@n78, Range=@n79.G, Start Offset=@n79.G, Sharpness=@n79.G) {materialFunction=&MF_MaskShifting}.Result) {constB=2})
  n81 = Multiply(A=Multiply(A=ConstantBiasScale(None=@n80), B=@n75)) {constB=2}
  n82 = NamedRerouteUsage()
  NamedRerouteDeclaration(None=Saturate(None=Max(A=@n81, B=@n82))) {name=ImpulseMask, nodeColor=<rgba 0.4062 0 1 1>}
  NamedRerouteDeclaration(None=ScalarParameter() {defaultValue=1, group=Meteor, parameterName=Animation_Duration, sortPriority=101}) {name=Duration, nodeColor=<rgba 0.6641 1 0 1>}
  n83 = NamedRerouteUsage()
  n84 = Multiply(A=ComponentMask(None=@n83) {b=true}) {constB=-1}
  n85 = OneMinus(None=DotProduct(A=@n83, B=@n73))
  NamedRerouteDeclaration(None=Multiply(A=Multiply(A=@n84, B=@n85)) {constB=0.25}) {name=BounceLight, nodeColor=<rgba 0.4766 0 1 1>}
  n86 = CurveAtlasRowParameter(CurveTime=NamedRerouteUsage()) {
   atlas=/DropPodParentGameplay/Assets/Textures/Curves/CA_DropPodParent
   curve=/DropPodParentGameplay/Assets/Textures/Curves/C_DropPodParent_01
   parameterName=DropPod_CurvesAnimation
  }
  NamedRerouteDeclaration(None=LinearInterpolate(A=Constant3Vector(), B=Constant3Vector() {constant=<rgba 0 0 8 0>}, Alpha=@n86.R) {constA=1, constB=0.8}) {name=Stomp, nodeColor=<rgba 0 1 0.4844 1>}
  n87 = ScalarParameter() {group=Meteor, parameterName=CurrentStage}
  n88 = LinearInterpolate(A=Subtract(A=@n87), B=@n87, Alpha=NamedRerouteUsage()) {constB=20}
  n89 = Constant4Vector() {constant=<rgba 1 2 0 1>}
  NamedRerouteDeclaration(None=Saturate(None=MaterialFunctionCall(Input=@n88, Input Low=@n89.R, Input High=@n89.R, Target Low=@n89.R, Target High=@n89.R) {materialFunction=&RemapValueRange}.Result)) {name=Stages02, nodeColor=<rgba 0.3812 1 0.0069 1>}
  n90 = Constant4Vector() {constant=<rgba 2 3 0 1>}
  NamedRerouteDeclaration(None=Saturate(None=MaterialFunctionCall(Input=@n88, Input Low=@n90.R, Input High=@n90.R, Target Low=@n90.R, Target High=@n90.R) {materialFunction=&RemapValueRange}.Result)) {name=Stages03, nodeColor=<rgba 0.3812 1 0.0069 1>}
  n91 = Constant3Vector() {desc=Random, constant=<rgba -0.3864 0.6 0.4464 0>}
  n92 = Add(
   A=Multiply(A=Time(), B=ScalarParameter() {defaultValue=100, group=Meteor, parameterName=Shake_Frequence, sortPriority=1})
   B=ScalarParameter() {group=Meteor, parameterName=Wobble_Offset, sortPriority=2}
  )
  n93 = MaterialFunctionCall(Float3=ObjectPositionWS()) {materialFunction=&BreakOutFloat3Components}
  n94 = Sine(None=Multiply(A=@n91.R, B=Add(A=@n92, B=@n93.R))) {period=6.283}
  n95 = Sine(None=Multiply(A=@n91.G, B=Add(A=@n92, B=@n93.G))) {period=6.283}
  n96 = Sine(None=Multiply(A=@n91.B, B=Add(A=@n92, B=@n93.B))) {period=6.283}
  n97 = MaterialFunctionCall(Float3=Multiply(
    A=MaterialFunctionCall(X=@n94, Y=@n95, Z=@n96) {materialFunction=&MakeFloat3}.Result
    B=ScalarParameter() {defaultValue=50, group=Meteor, parameterName=Shake_Amplitude, sortPriority=4}
   ) {desc=wind factor}) {materialFunction=&BreakOutFloat3Components}
  n98 = Add(A=@n97.G, B=@n97.G)
  n99 = Add(A=@n97.R, B=@n97.R)
  n100 = Add(A=@n97.R, B=@n97.R)
  n101 = ScalarParameter() {defaultValue=0.015, group=Meteor, parameterName=Shake_Strength, sortPriority=3}
  NamedRerouteDeclaration(None=Multiply(
    A=MaterialFunctionCall(X=@n98, Y=@n99, Z=@n100) {materialFunction=&MakeFloat3}.Result
    B=Multiply(A=@n101) {constB=0.1, desc=adjustStrength}
   )) {name=ShakeNoise, nodeColor=<rgba 0 1 0.3203 1>}
  n102 = Divide(A=NamedRerouteUsage())
  n103 = LinearInterpolate(B=NamedRerouteUsage(), Alpha=Saturate(None=Sine(None=@n102)))
  NamedRerouteDeclaration(None=@n103) {name=Stage2Shake, nodeColor=<rgba 0.5703 1 0 1>}
  NamedRerouteDeclaration(None=Multiply(A=ComponentMask(None=@n103) {r=true}) {constB=0.025}) {name=Stage2Blink, nodeColor=<rgba 1 0 0.8906 1>}
  n104 = Constant4Vector() {constant=<rgba 0 1 0 1>}
  NamedRerouteDeclaration(None=Saturate(None=MaterialFunctionCall(Input=@n88, Input Low=@n104.R, Input High=@n104.R, Target Low=@n104.R, Target High=@n104.R) {materialFunction=&RemapValueRange}.Result)) {name=Stages01, nodeColor=<rgba 0.3812 1 0.0069 1>}
  NamedRerouteDeclaration(None=VertexInterpolator(VS=PreSkinnedNormal()).PS) {name=PreSkinnedNormal, nodeColor=<rgba 0 0.9297 1 1>}
  NamedRerouteDeclaration(None=VertexInterpolator(VS=LocalPosition() {localOrigin=InstancePreSkinning}.XYZ).PS) {name=PreSkinnedLoc, nodeColor=<rgba 1 0.75 0 1>}
  n105 = Subtract(A=Time(), B=ScalarParameter() {defaultValue=-99999, group=Meteor, parameterName=GameTimeFromBP})
  n106 = Add(A=NamedRerouteUsage()) {constB=0}
  NamedRerouteDeclaration(None=Saturate(None=Divide(A=@n105, B=@n106))) {name=TimeNative, nodeColor=<rgba 0.2891 1 0 1>}
  n107 = MaterialFunctionCall!(Height=NamedRerouteUsage(), Absolute World Position=NamedRerouteUsage(), World Space Vertex Normals=VertexNormalWS()) {materialFunction=/Engine/Functions/Engine_MaterialFunctions03/Procedurals/HeightToNormalSmooth}
  NamedRerouteDeclaration(None=MaterialFunctionCall(
    Normal=Transform(None=@n107.WorldSpaceNormal) {transformSourceType=TRANSFORMSOURCE_World, transformType=TRANSFORM_Tangent}
    Flatness=Constant() {r=0.25}
   ) {materialFunction=&FlattenNormal}.Result) {name=GeneratedCracksNormal, nodeColor=<rgba 0 0.8359 1 1>}
  n108 = NamedRerouteUsage()
  n109 = Add(A=@n108, B=Constant3Vector() {constant=<rgba 0 0 1 0>})
  n110 = Multiply(
   A=MaterialFunctionCall(Normal=NamedRerouteUsage(), Flatness=Constant() {r=0.55}) {materialFunction=&FlattenNormal}.Result
   B=Constant3Vector() {constant=<rgba -1 -1 1 0>}
  )
  n111 = DotProduct(A=@n109, B=@n110)
  n112 = Subtract(A=Divide(A=Multiply(A=@n109, B=@n111), B=ComponentMask(None=@n109) {b=true}), B=@n110)
  n113 = Add(
   A=MaterialFunctionCall!(Low=@n112, High=@n112, Epic=@n112, --------, Mobile=@n108, Medium=@n108) {materialFunction=/FortniteBaseMaterials/Functions/MF_QualitySwitch_Vector}.Result
   B=Constant3Vector() {constant=<rgba 0 0 1 0>}
  )
  n114 = Multiply(A=NamedRerouteUsage(), B=Constant3Vector() {constant=<rgba -1 -1 1 0>})
  n115 = Divide(A=Multiply(A=@n113, B=DotProduct(A=@n113, B=@n114)), B=ComponentMask(None=@n113) {b=true})
  NamedRerouteDeclaration(None=Normalize(VectorInput=Subtract(A=@n115, B=@n114))) {name=CombinedNormal, nodeColor=<rgba 0 0.7656 1 1>}
  n116 = MaterialFunctionCall() {materialFunction=&ObjectLocalBounds}
  n117 = Subtract(
   A=Subtract(A=ObjectPositionWS(), B=WorldPosition() {worldPositionShaderOffset=WPT_ExcludeAllShaderOffsets}.XYZ)
   B=@n116.Local Bounds Minimum
  )
  n118 = Multiply(
   A=@n116.Local Bounds Size
   B=MaterialFunctionCall() {materialFunction=&ObjectScale}.Scale XYZ
  )
  NamedRerouteDeclaration(None=MaterialFunctionCall(Float3=Divide(A=@n117, B=@n118)) {materialFunction=&BreakOutFloat3Components}.B) {name=BoundingHeatMask, nodeColor=<rgba 0 1 0.9766 1>}
  n119 = Multiply(A=@n40) {constB=0.5}
  n120 = TextureObject() {texture=/ToyFoxGameplay/Assets/Texture/T_ToyFox_SupplyDrops_Meteor_Mask}
  n121 = TextureSample(UVs=@n40, Tex=@n120) {samplerSource=SSM_Wrap_WorldGroupSettings}
  NamedRerouteDeclaration(None=Max(A=TextureSample(UVs=@n119, Tex=@n120) {samplerSource=SSM_Wrap_WorldGroupSettings}.G, B=@n121.B)) {name=Meteor_Detail_Mask, nodeColor=<rgba 1 0.6563 0 1>}
  n122 = TextureSample(UVs=Multiply(A=Add(A=@n40) {constB=0.5}) {constB=0.5}, Tex=@n120) {samplerSource=SSM_Wrap_WorldGroupSettings}
  NamedRerouteDeclaration(None=Max(A=Multiply(A=@n122.R) {constB=0.25}, B=@n121.R)) {name=Cracks_Mask, nodeColor=<rgba 0.5469 1 0 1>}
  NamedRerouteUsage()
  PinBase() {reroutePins [Input {expression=&self:MaterialExpressionReroute_3}]}
  PinBase!(Output=@n7) {reroutePins [Output {expression=&self:MaterialExpressionReroute_2}]}
  PinBase!() {reroutePins [
    "For Moss (MA)" {expression=&self:MaterialExpressionReroute_8}
    "RVT_Height (S)" {expression=&self:MaterialExpressionReroute_1}
   ]}
  PinBase!(Result=@n14.Buildings Jungl) {reroutePins [Result {expression=&self:MaterialExpressionReroute_0}]}
  PinBase!() {reroutePins [MaterialAttributes {expression=&self:MaterialExpressionReroute_70}]}
  PinBase!(Output=@n3) {reroutePins [Output {expression=&self:MaterialExpressionReroute_71}]}
  PinBase!() {reroutePins [MaterialAttributes {expression=&self:MaterialExpressionReroute_69}]}
  MaterialExpressionReroute_69 = Reroute()
  n123 = MaterialFunctionCall(Material_In=@MaterialExpressionReroute_69.MaterialAttributes) {materialFunction=&MF_FN_DestructibleDamage}
  n124 = MaterialFunctionCall(In=MaterialFunctionCall!(Material Attributes=MaterialFunctionCall!(In=@n123.Material_Out) {materialFunction=/Game/Athena/WinterArt/Materials/MF_WPO_WindNew}.Result) {materialFunction=/Game/Creative/Sets/MilitaryBase/Props/MF_FanSpin}.Result) {materialFunction=&MF_CurieFallbackEffect}
  n125 = MaterialFunctionCall!(In=MaterialFunctionCall!(In=MaterialFunctionCall!(In=@n124.Result, Custom Mask, Custom Rotation Axis) {materialFunction=/Game/Athena/WinterArt/Materials/MF_WindWPO_v2}.Result) {materialFunction=/Game/Global/Base4Layers_2022/MaterialFunctions/MF_SnowAndIceArchitecture}.Result) {materialFunction=/Game/Global/Base4Layers_2022/MaterialFunctions/MF_VibrationRattle}
  PinBase!(Result=@n125.Result) {reroutePins [Result {expression=&self:MaterialExpressionReroute_68}]}
  PinBase!() {reroutePins [InputPin {expression=&self:MaterialExpressionReroute_7}]}
  PinBase!(Output=@n20) {reroutePins [Output {expression=&self:MaterialExpressionReroute_6}]}
  MaterialExpressionReroute_68 = Reroute(None=@n125.Result)
 ]
}
