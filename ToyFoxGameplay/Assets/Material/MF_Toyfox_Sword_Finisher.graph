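# Material function for the ToyFox sword finisher, layering animated lightning and power-up glow onto incoming material attributes.
# Hard triplanar UVs, panned noise and hashed flicker form procedural lightning lines, alongside HLSL scan streaks.
# Z-travelling sphere masks create lifting energy and burst glows, shaped by texture, bounding-box and fresnel masks.
# Curve atlas rows time the effects from an animation parameter or Blueprint-supplied start time, also adjusting base color, specular and roughness.
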MaterialFunction /ToyFoxGameplay/Assets/Material/MF_Toyfox_Sword_Finisher
graph {nodes [
  Frac(None=Divide(A=Time()) {constB=6})
  n1 = GetMaterialAttributes(None=FunctionInput() {inputName=IN MA, inputType=FunctionInput_MaterialAttributes}) {attributeGetTypes [
    #1
    #2
    #3
    #4
   ]}
  n2 = NamedRerouteUsage()
  n3 = Saturate(None=ComponentMask(None=NamedRerouteUsage()) {r=true})
  n4 = SetMaterialAttributes!(
   MaterialAttributes=@n1.MaterialAttributes
   BaseColor=LinearInterpolate(A=@n1.BaseColor, B=LinearInterpolate(Alpha=@n2) {constA=0.25, constB=0.75}, Alpha=@n3)
   Specular=LinearInterpolate(A=@n1.Specular, Alpha=@n3) {constB=0}
   Roughness=LinearInterpolate(A=@n1.Roughness, Alpha=@n3)
   EmissiveColor=Add(A=NamedRerouteUsage(), B=NamedRerouteUsage())
  ) {attributeSetTypes [
    #1
    #2
    #3
    #5
   ]}
  FunctionOutput(None=@n4) {outputName=MA OUT}
  n5 = NamedRerouteUsage()
  n6 = Multiply(A=NamedRerouteUsage()) {constB=2}
  n7 = NamedRerouteUsage()
  n8 = Multiply(A=NamedRerouteUsage()) {constB=0.2}
  n9 = Add(A=@n5, B=Add(A=Add(A=@n6, B=@n7), B=@n8)) {constB=0}
  Multiply(A=@n9) {desc=Count}
  n10 = Divide(A=Multiply(A=NamedRerouteUsage()) {constB=-1}, B=Abs(None=Constant() {r=32})) {constB=1024}
  n11 = VectorParameter() {group=FinisherAnimation, parameterName=Triplannar3DTilling, defaultValue=<rgba 1 1 1 1>}
  n12 = MaterialFunctionCall(Float3=Add(
    A=Multiply(A=@n10, B=@n11.RGB)
    B=VectorParameter() {group=FinisherAnimation, parameterName=Triplannar3DOffset, defaultValue=<rgba 0 0 0 1>}.RGB
   )) {materialFunction=&BreakOutFloat3Components}
  n13 = NamedRerouteUsage()
  MaterialExpressionComponentMask_3 = ComponentMask(None=@n13) {r=true}
  n14 = Custom!(x=@MaterialExpressionComponentMask_3) {
   code="return float(x < 0) ? -1 : 1;"
   description=StepReturn
   inputs [{inputName=x, input {expression=&self:MaterialExpressionComponentMask_3}}]
  }
  MaterialExpressionComponentMask_2 = ComponentMask(None=@n13) {g=true}
  n15 = Custom(x=@MaterialExpressionComponentMask_2) {
   code="return float(x < 0) ? 1 : -1;"
   description=StepReturn
   inputs [{inputName=x, input {expression=&self:MaterialExpressionComponentMask_2}}]
  }
  MaterialExpressionComponentMask_1 = ComponentMask(None=@n13) {b=true}
  n16 = Custom(x=@MaterialExpressionComponentMask_1) {
   code="return float(x < 0) ? -1 : 1;"
   description=StepReturn
   inputs [{inputName=x, input {expression=&self:MaterialExpressionComponentMask_1}}]
  }
  n17 = MaterialFunctionCall(Float3=MaterialFunctionCall(X=@n14, Y=@n15, Z=@n16) {materialFunction=&MakeFloat3}.Result) {materialFunction=&BreakOutFloat3Components}
  n18 = Abs(None=@n13)
  n19 = MaterialFunctionCall(Float3=@n18) {materialFunction=&BreakOutFloat3Components}
  n20 = Max(A=Max(A=@n19.R, B=@n19.R), B=@n19.B)
  n21 = MaterialFunctionCall(Float3=SmoothStep(Max=@n20, Value=@n18)) {materialFunction=&BreakOutFloat3Components}
  n22 = LinearInterpolate(
   A=MaterialFunctionCall(X=Multiply(A=@n12.G, B=@n17.R), Y=@n12.B) {desc=Side, materialFunction=&MakeFloat2}.Result
   B=MaterialFunctionCall(X=Multiply(A=@n12.R, B=@n17.G), Y=@n12.B) {desc=Front, materialFunction=&MakeFloat2}.Result
   Alpha=@n21.G
  )
  n23 = MaterialFunctionCall(X=Multiply(A=@n12.R, B=@n17.B), Y=@n12.G) {desc=Top, materialFunction=&MakeFloat2}
  NamedRerouteDeclaration(None=ConstantBiasScale(None=LinearInterpolate(A=@n22, B=@n23.Result, Alpha=@n21.B))) {name=HardTriplannarUV, nodeColor=<rgba 0.0312 0 1 1>}
  n24 = NamedRerouteUsage()
  n25 = Add(A=NamedRerouteUsage(), B=Multiply(A=ComponentMask(None=@n24) {g=true, r=true}) {constB=0.1}) {desc=Add normal dirstort}
  n26 = Panner(Coordinate=@n25) {speedX=0.05}
  n27 = NamedRerouteUsage()
  n28 = TextureSample(UVs=Multiply(A=@n26, B=Constant() {r=0.5}) {constB=4}, Tex=@n27) {samplerSource=SSM_Wrap_WorldGroupSettings}
  n29 = TextureSample(UVs=Multiply(A=@n26) {constB=-1}, Tex=@n27) {samplerSource=SSM_Wrap_WorldGroupSettings}
  NamedRerouteDeclaration(None=Multiply(A=Subtract(A=@n28.R, B=@n29.R), B=ScalarParameter() {defaultValue=0.2, group=FinisherAnimation, parameterName=LightningDistort_Intensity})) {name=DistortNoise, nodeColor=<rgba 0.2891 1 0 1>}
  n30 = MaterialFunctionCall() {materialFunction=&BoundingBoxBased_0-1_UVW}
  n31 = OneMinus(None=@n30.R)
  n32 = OneMinus(None=@n30.B)
  NamedRerouteDeclaration(None=Min(A=Min(A=@n30.R, B=@n31), B=Min(A=@n30.B, B=@n32))) {name=SwordSoftBoundingMsk, nodeColor=<rgba 0 0.2969 1 1>}
  NamedRerouteDeclaration(None=TextureObject() {texture=/Game/Global/Textures/Default/Noise/T_godray_Noise}) {name=Noise_TEX, nodeColor=<rgba 0 0.4232 1 1>}
  n33 = OneMinus(None=DotProduct(A=VertexNormalWS(), B=CameraVectorWS()))
  NamedRerouteDeclaration(None=Power(Base=Saturate(None=@n33)) {constExponent=4}) {name=SwordFresnel, nodeColor=<rgba 0.6172 1 0 1>}
  MaterialExpressionNamedRerouteUsage_23 = NamedRerouteUsage()
  MaterialExpressionTime_8 = Time()
  n34 = Custom!(uv=@MaterialExpressionNamedRerouteUsage_23, Time=@MaterialExpressionTime_8) {
   code="float3 col = float3(1.0,1.0,1.0);\t\t// Drop Colour\r\nuv.x = uv.x*1.0;\t\t\t\t\t\t// H-Count\r\nfloat dx = frac(uv.x);\r\nuv.x = floor(uv.x);\r\nfloat t =  Time*0.5;\r\nuv.y *= 0.05;\t\t\t\t\t\t\t// stretch\r\nfloat o=sin(uv.x*215.4);\t\t\t\t// offset\r\nfloat s=cos(uv.x*33.1)*.3 +.7;\t\t\t// speed\r\nfloat trail = lerp(20.0,10.0,s);\t\t\t// trail length\r\nfloat yv = frac(uv.y + t*s + o) * trail;\r\nyv = 1.0/yv;\r\nyv = smoothstep(0.0,1.0,yv*yv);\r\nyv = sin(yv*PI)*(s*5.0);\r\nfloat d2 = sin(dx*PI);\r\nyv *= d2*d2;\r\ncol = col*yv;\r\nreturn yv;"
   description=LigtningScan1
   inputs [
    {inputName=uv, input {expression=&self:MaterialExpressionNamedRerouteUsage_23}}
    {inputName=Time, input {expression=&self:MaterialExpressionTime_8}}
   ]
  }
  NamedRerouteDeclaration(None=@n34) {name=LightScanShifted, nodeColor=<rgba 1 0 0.9375 1>}
  n35 = NamedRerouteUsage()
  n36 = NamedRerouteUsage()
  n37 = Add(A=@n35, B=Multiply(A=ComponentMask(None=@n36) {g=true, r=true}) {constB=0.25}) {desc=Add normal dirstort}
  n38 = MaterialFunctionCall(UVs=Panner(Coordinate=@n37) {speedX=0.5}, "Rotation Angle (0-1)"=Constant() {r=0.1}) {materialFunction=&CustomRotator}
  NamedRerouteDeclaration(None=@n38.Rotated Values) {name=LightningUV_01, nodeColor=<rgba 0 1 0.7422 1>}
  n39 = Add(A=@n37, B=Constant2Vector() {r=-0.2})
  NamedRerouteDeclaration(None=MaterialFunctionCall(UVs=Panner(Coordinate=@n39) {speedX=0.1, speedY=0.5}, "Rotation Angle (0-1)"=Constant() {r=0.02}) {materialFunction=&CustomRotator}.Rotated Values) {name=LightningUV_02, nodeColor=<rgba 1 0.4219 0 1>}
  n40 = Constant() {r=0.2}
  n41 = Frac(None=Multiply(A=MaterialFunctionCall(UVs=@n35, "Rotation Angle (0-1)"=@n40) {materialFunction=&CustomRotator}.Rotated Values) {constB=2})
  n42 = ComponentMask(None=Subtract(A=Multiply(A=@n41) {constB=2})) {r=true}
  NamedRerouteDeclaration(None=SmoothStep(Value=Abs(None=@n42)) {constMax=2}) {name=DistortPattern, nodeColor=<rgba 0.1016 0 1 1>}
  n43 = NamedRerouteUsage()
  NamedRerouteDeclaration(None=Abs(None=ComponentMask(None=@n43) {g=true})) {name=SwordFrontMsk, nodeColor=<rgba 0.2187 1 0 1>}
  n44 = ConstantBiasScale(None=MaterialFunctionCall(Float=Time()) {materialFunction=&MF_Hash}.Float)
  n45 = LinearInterpolate(Alpha=@n44) {constA=0.01, constAlpha=0.2, constB=0.07}
  n46 = Add(A=NamedRerouteUsage(), B=@n8)
  n47 = Frac(None=Multiply(A=Add(A=@n5, B=@n46) {constB=0}) {constB=3, desc=Count})
  n48 = ComponentMask(None=Subtract(A=Multiply(A=@n47) {constB=2})) {r=true}
  n49 = Saturate(None=Divide(A=@n45, B=Abs(None=@n48)) {constA=0.02})
  n50 = Multiply(A=Frac(None=@n9)) {constB=2}
  n51 = Abs(None=ComponentMask(None=Subtract(A=@n50)) {r=true})
  n52 = Multiply(A=Saturate(None=Divide(A=@n45, B=@n51) {constA=0.02}), B=NamedRerouteUsage())
  n53 = Multiply(A=Power(Base=Add(A=@n49, B=@n52)) {constExponent=1}, B=NamedRerouteUsage())
  n54 = LinearInterpolate(Alpha=@n44) {constA=0.007, constAlpha=0.2, constB=0.01}
  n55 = Add(A=NamedRerouteUsage(), B=NamedRerouteUsage())
  n56 = Multiply(A=NamedRerouteUsage(), B=NamedRerouteUsage()) {constB=2}
  n57 = Frac(None=Multiply(A=Add(A=@n55, B=@n56)) {constB=5, desc=Count})
  n58 = ComponentMask(None=Subtract(A=Multiply(A=@n57) {constB=2})) {r=true}
  n59 = Saturate(None=Divide(A=@n54, B=Abs(None=@n58)) {constA=0.008})
  n60 = Add(A=@n5, B=NamedRerouteUsage())
  n61 = Multiply(A=NamedRerouteUsage(), B=NamedRerouteUsage()) {constB=0.2}
  n62 = Frac(None=Multiply(A=Add(A=@n60, B=@n61)) {constB=3, desc=Count})
  n63 = ComponentMask(None=Subtract(A=Multiply(A=@n62) {constB=2})) {r=true}
  n64 = Saturate(None=Divide(A=@n54, B=Abs(None=@n63)) {constA=0.01})
  n65 = Multiply(A=Power(Base=Add(A=@n59, B=@n64)) {constExponent=1}, B=NamedRerouteUsage())
  n66 = Multiply(A=Multiply(A=Add(A=@n53, B=@n65), B=NamedRerouteUsage()), B=ScalarParameter() {defaultValue=10, parameterName=Lightning_Intensity}) {constB=40}
  n67 = LinearInterpolate(
   A=VectorParameter() {group=FinisherAnimation, parameterName=Lightning_Color, defaultValue=<rgba 0.2 0.3 0.8 1>}.RGB
   B=Constant3Vector() {constant=<rgba 1 1 0.85 0>}
   Alpha=NamedRerouteUsage()
  )
  NamedRerouteDeclaration(None=Multiply(A=Multiply(A=@n66, B=@n67), B=NamedRerouteUsage())) {name=SwordLightning, nodeColor=<rgba 0 0.6016 1 1>}
  NamedRerouteDeclaration(None=@n1.Normal) {name=OrigialNormal, nodeColor=<rgba 0 1 0.5078 1>}
  NamedRerouteDeclaration(None=OneMinus(None=Step(Y=@n30.B) {constX=0.3112})) {name=SwordFace, nodeColor=<rgba 1 0 0.9141 1>}
  n68 = Multiply(A=@n25, B=Constant2Vector() {g=0.1, r=4})
  n69 = Multiply(A=TextureSample(UVs=Panner(Coordinate=@n68) {speedY=1}, Tex=@n27) {samplerSource=SSM_Wrap_WorldGroupSettings}.R, B=NamedRerouteUsage())
  n70 = TextureSampleParameter2D() {
   parameterName=SRM
   samplerSource=SSM_Wrap_WorldGroupSettings
   samplerType=SAMPLERTYPE_Masks
   texture=&T_BlackMask
  }
  n71 = Subtract(A=@n70.R, B=@n70.R)
  NamedRerouteDeclaration(None=Multiply(A=Multiply(A=@n69) {constB=10}, B=@n71)) {name=ChargedNoise, nodeColor=<rgba 0.9453 0 1 1>}
  MaterialExpressionNamedRerouteUsage_10 = NamedRerouteUsage()
  n72 = Custom(uv=@MaterialExpressionNamedRerouteUsage_10, Time=@MaterialExpressionTime_8) {
   code="float3 col = float3(1.0,1.0,1.0);\t\t// Drop Colour\r\nuv.x = uv.x*1.0;\t\t\t\t\t\t// H-Count\r\nfloat dx = frac(uv.x);\r\nuv.x = floor(uv.x);\r\nfloat t =  Time*0.5;\r\nuv.y *= 0.05;\t\t\t\t\t\t\t// stretch\r\nfloat o=sin(uv.x*215.4);\t\t\t\t// offset\r\nfloat s=cos(uv.x*33.1)*.3 +.7;\t\t\t// speed\r\nfloat trail = lerp(20.0,10.0,s);\t\t\t// trail length\r\nfloat yv = frac(uv.y + t*s + o) * trail;\r\nyv = 1.0/yv;\r\nyv = smoothstep(0.0,1.0,yv*yv);\r\nyv = sin(yv*PI)*(s*5.0);\r\nfloat d2 = sin(dx*PI);\r\nyv *= d2*d2;\r\ncol = col*yv;\r\nreturn yv;"
   description=LigtningScan1
   inputs [
    {inputName=uv, input {expression=&self:MaterialExpressionNamedRerouteUsage_10}}
    {inputName=Time, input {expression=&self:MaterialExpressionTime_8}}
   ]
  }
  NamedRerouteDeclaration(None=@n72) {name=LightScanVertical, nodeColor=<rgba 0 0.2266 1 1>}
  n73 = TextureSampleParameter2D() {
   parameterName=Mask
   samplerSource=SSM_Wrap_WorldGroupSettings
   samplerType=SAMPLERTYPE_Masks
   texture=&T_BlackMask
  }
  n74 = Multiply(A=TextureSample(UVs=NamedRerouteUsage(), Tex=@n27) {samplerSource=SSM_Wrap_WorldGroupSettings}.R, B=@n73.R)
  NamedRerouteDeclaration(None=Saturate(None=Max(A=@n74, B=@n71))) {name=PowerLiftingNoise, nodeColor=<rgba 1 0 0.1875 1>}
  NamedRerouteDeclaration(None=@n73.G) {name=sword_msk, nodeColor=<rgba 0 0.0391 1 1>}
  n75 = OneMinus(None=@n73.G)
  n76 = NamedRerouteUsage()
  n77 = Add(A=SmoothStep(Value=@n71) {constMin=0.1}, B=Multiply(A=SmoothStep(Value=@n75) {constMin=0.35}, B=Step(X=@n76) {constX=0.35, constY=0.02}))
  n78 = NamedRerouteUsage()
  NamedRerouteDeclaration(None=Saturate(None=Multiply(A=@n77, B=@n78))) {name=SwordFacePattern, nodeColor=<rgba 0 0.1328 1 1>}
  NamedRerouteDeclaration(None=SmoothStep(Value=@n30.B) {constMin=0.3}) {name=SwordFaceGradient, nodeColor=<rgba 0.4297 0 1 1>}
  NamedRerouteDeclaration(None=VertexInterpolator(VS=PreSkinnedNormal()).PS) {name=PreSkinnedNormal, nodeColor=<rgba 0 0.9297 1 1>}
  NamedRerouteDeclaration(None=VertexInterpolator(VS=LocalPosition() {localOrigin=InstancePreSkinning}.XYZ).PS) {name=PreSkinnedLoc, nodeColor=<rgba 1 0.75 0 1>}
  NamedRerouteDeclaration(None=@n38.Rotated Values) {name=RotatedTriUV, nodeColor=<rgba 1 0 0.3281 1>}
  n79 = ScalarParameter() {group=FinisherAnimation, parameterName=ProcessAnimationCurve, sortPriority=3}
  n80 = Subtract(A=@n79)
  n81 = NamedRerouteUsage()
  n82 = Constant4Vector() {desc=Stage 0 to 1, constant=<rgba 0 1 0 1>}
  n83 = Saturate(None=MaterialFunctionCall(
    Input=LinearInterpolate(A=@n80, B=@n79, Alpha=@n81) {desc=Native lerp toward}
    Input Low=@n82.R
    Input High=@n82.R
    Target Low=@n82.R
    Target High=@n82.R
   ) {materialFunction=&RemapValueRange}.Result)
  n84 = CurveAtlasRowParameter(CurveTime=@n83) {
   atlas=/ToyFoxGameplay/Assets/Texture/Curves/CA_ToyFox_Sword_Finisher
   curve=/ToyFoxGameplay/Assets/Texture/Curves/C_ToyFox_Sword_Lightning
   defaultValue=1
   group=FinisherAnimation
   parameterName=FinisherCurves_Lightning
  }
  NamedRerouteDeclaration(None=@n84.R) {name=LightningCurve_Intensity, nodeColor=<rgba 1 0.0469 0 1>}
  NamedRerouteDeclaration(None=@n84.G) {name=LightningCurve_Color, nodeColor=<rgba 1 0.0469 0 1>}
  n85 = Add(A=NamedRerouteUsage()) {constB=0.5}
  n86 = NamedRerouteUsage()
  n87 = CurveAtlasRowParameter(CurveTime=@n83) {
   atlas=&CA_ToyFox_Sword_Finisher
   curve=/ToyFoxGameplay/Assets/Texture/Curves/C_ToyFox_Sword_Powerup
   group=FinisherAnimation
   parameterName=FinisherCurves_Powerup
  }
  n88 = LinearInterpolate(Alpha=@n87.R) {constB=180}
  n89 = NamedRerouteUsage()
  n90 = Add(
   A=SphereMask(
    A=@n86
    B=MaterialFunctionCall(Z=@n88) {materialFunction=&MakeFloat3}.Result
    Radius=LinearInterpolate(Alpha=@n87.R) {constA=20, constB=60}
   ) {attenuationRadius=20, desc=Lifting Energy, hardnessPercent=0}
   B=@n89
  )
  n91 = VectorParameter() {group=FinisherAnimation, parameterName=PowerLiftingColor, channelNames {a=Strength, b="", g="", r=""}, defaultValue=<rgba 1 0.2789 0.0056 4>}
  n92 = Multiply(A=@n91, B=@n91) {constB=4}
  n93 = VectorParameter() {group=FinisherAnimation, parameterName=PowerGlowingColor, channelNames {a=Strength, b="", g="", r=""}, defaultValue=<rgba 1 1 0.666 3>}
  n94 = Multiply(A=@n93, B=@n93) {constB=4}
  n95 = Power(Base=Multiply(A=Multiply(A=@n85, B=@n90), B=LinearInterpolate(A=@n92, B=@n94, Alpha=@n87.R))) {constExponent=5}
  n96 = LinearInterpolate(Alpha=@n87.R) {constAlpha=1, constB=10}
  n97 = LinearInterpolate(A=@n94, Alpha=@n87.G) {constB=3}
  n98 = SphereMask(
   A=@n86
   B=MaterialFunctionCall(Z=LinearInterpolate(Alpha=@n87.G) {constA=-10, constB=240}) {materialFunction=&MakeFloat3}.Result
   Radius=LinearInterpolate(Alpha=@n87.G) {constB=360}
  ) {attenuationRadius=20, desc=Burst, hardnessPercent=0}
  n99 = LinearInterpolate(A=@n98, B=Multiply(A=@n98) {constB=0.2}, Alpha=@n89) {constB=0.1}
  n100 = LinearInterpolate(B=Power(Base=Multiply(A=@n97, B=@n99)) {constExponent=10}, Alpha=NamedRerouteUsage())
  n101 = LinearInterpolate(A=@n91, B=@n93, Alpha=NamedRerouteUsage())
  n102 = Multiply(A=Multiply(A=Multiply(A=@n101, B=@n89)) {constB=4}, B=@n99) {constB=10}
  n103 = Multiply(A=Add(A=Multiply(A=@n95, B=@n96) {constA=20}, B=Add(A=@n100, B=@n102)), B=@n87.B)
  n104 = NamedRerouteUsage()
  NamedRerouteDeclaration(None=Multiply(A=Multiply(A=@n103, B=@n104), B=NamedRerouteUsage())) {name=SwordCharge_Glow, nodeColor=<rgba 0 1 0.6016 1>}
  n105 = Subtract(A=Time(), B=ScalarParameter() {defaultValue=-99999, group=FinisherAnimation, parameterName=GameTimeFromBP, sortPriority=100})
  n106 = Add(A=ScalarParameter() {defaultValue=1, group=FinisherAnimation, parameterName=Animation_Duration, sortPriority=101}) {constB=0}
  NamedRerouteDeclaration(None=Saturate(None=Divide(A=@n105, B=@n106))) {name=TimeNative, nodeColor=<rgba 0.2891 1 0 1>}
  Normalize()
  Step(Y=@n20, X=@n18)
  Subtract(A=@n70.R, B=@n70.R)
 ]}
