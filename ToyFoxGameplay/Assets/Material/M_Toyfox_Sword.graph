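# Surface material for the ToyFox sword, built on the shared 2023 character parent function with material attributes.
# The parent output and its Sword Finisher variant feed MF_ToyFox_Outro as original and weakpoint looks, alongside an inverted bounding-box gradient and a constant visor mask.
# The outro also receives pre-skinned normals and hard triplanar UVs, built from tiled, offset pre-skinned positions using dominant-axis selection and sign-based flipping.
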Material /ToyFoxGameplay/Assets/Material/M_Toyfox_Sword
graph {
 properties {
  bUseMaterialAttributes=true
  bUsedWithClothing=true
  bUsedWithMeshDeformer=true
  bUsedWithMorphTargets=true
  bUsedWithNanite=true
  bUsedWithNiagaraMeshParticles=true
  bUsedWithSkeletalMesh=true
  pixelDepthOffsetMode=PDOM_Legacy
  shadingModel=MSM_FromMaterialExpression
  usedShadingModels=Default Lit | Subsurface | Clear Coat | Subsurface Profile | Cloth
 }
 nodes [
  n1 = MaterialFunctionCall() {materialFunction=&MF_CharacterParent_2023_PreFX}
  n2 = Divide(A=Multiply(A=@n1.Pre skinned Local Position) {constB=-1}, B=Abs(None=Constant() {r=32})) {constB=1024}
  n3 = VectorParameter() {group=Toyfox, parameterName=Triplannar3DTilling, defaultValue=<rgba 1 1 1 1>}
  n4 = MaterialFunctionCall(Float3=Add(A=Multiply(A=@n2, B=@n3.RGB), B=VectorParameter() {group=Toyfox, parameterName=Triplannar3DOffset, defaultValue=<rgba 0 0 0 1>}.RGB)) {materialFunction=&BreakOutFloat3Components}
  n5 = NamedRerouteUsage()
  MaterialExpressionComponentMask_2 = ComponentMask(None=@n5) {r=true}
  n6 = Custom!(x=@MaterialExpressionComponentMask_2) {
   code="return float(x < 0) ? -1 : 1;"
   description=StepReturn
   inputs [{inputName=x, input {expression=&self:MaterialExpressionComponentMask_2}}]
  }
  MaterialExpressionComponentMask_1 = ComponentMask(None=@n5) {g=true}
  n7 = Custom(x=@MaterialExpressionComponentMask_1) {
   code="return float(x < 0) ? 1 : -1;"
   description=StepReturn
   inputs [{inputName=x, input {expression=&self:MaterialExpressionComponentMask_1}}]
  }
  MaterialExpressionComponentMask_0 = ComponentMask(None=@n5) {b=true}
  n8 = Custom(x=@MaterialExpressionComponentMask_0) {
   code="return float(x < 0) ? -1 : 1;"
   description=StepReturn
   inputs [{inputName=x, input {expression=&self:MaterialExpressionComponentMask_0}}]
  }
  n9 = MaterialFunctionCall(Float3=MaterialFunctionCall(X=@n6, Y=@n7, Z=@n8) {materialFunction=&MakeFloat3}.Result) {materialFunction=&BreakOutFloat3Components}
  n10 = Multiply(A=@n4.G, B=@n9.R)
  n11 = Multiply(A=@n4.R, B=@n9.G)
  n12 = Abs(None=@n5)
  n13 = MaterialFunctionCall(Float3=@n12) {materialFunction=&BreakOutFloat3Components}
  n14 = Max(A=Max(A=@n13.R, B=@n13.R), B=@n13.B)
  n15 = MaterialFunctionCall(Float3=Step(Y=@n14, X=@n12)) {materialFunction=&BreakOutFloat3Components}
  n16 = LinearInterpolate(
   A=LinearInterpolate(
    A=MaterialFunctionCall(X=@n10, Y=@n4.B) {desc=Side, materialFunction=&MakeFloat2}.Result
    B=MaterialFunctionCall(X=@n11, Y=@n4.B) {desc=Front, materialFunction=&MakeFloat2}.Result
    Alpha=@n15.G
   )
   B=MaterialFunctionCall(X=Multiply(A=@n4.R, B=@n9.B), Y=@n4.G) {desc=Top, materialFunction=&MakeFloat2}.Result
   Alpha=@n15.B
  )
  n17 = PreSkinnedNormal()
  MaterialFunctionCall!(
   IN OriginalVisual MA=@n1.Result
   IN WeakpointVisual MA=MaterialFunctionCall!(IN MA=@n1.Result) {materialFunction=&/MF_Toyfox_Sword_Finisher}.MA OUT
   IN Bounding=OneMinus(None=MaterialFunctionCall() {materialFunction=&BoundingBoxBased_0-1_UVW}.B)
   IN VisorMask=Constant() {r=1}
   IN TriplannarUV=NamedRerouteDeclaration(None=ConstantBiasScale(None=@n16)) {name=HardTriplannarUV, nodeColor=<rgba 0.0312 0 1 1>}
   IN PreSkinnedLoc=@n1.Result
   IN PreSkinnedNormal=NamedRerouteDeclaration(None=VertexInterpolator(VS=@n17).PS) {name=PreskinedNormal, nodeColor=<rgba 1 0.0938 0 1>}
  ) {materialFunction=&/MF_ToyFox_Outro}
  SmoothStep(Max=@n14, Value=@n12)
 ]
}
