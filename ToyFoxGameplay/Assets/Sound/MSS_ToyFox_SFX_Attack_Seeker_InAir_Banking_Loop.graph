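# Looping stereo MetaSound for the in-air ToyFox Seeker attack projectile, layering a whoosh with a banking sound.
# Azimuth and elevation movement speeds are normalized, smoothed and made absolute, then gated to drive separate ADSR envelopes.
# The larger envelope scales the banking layer and starts or stops it at thresholds, while the whoosh starts on play.
# Both layers mix through MasterVolume; a Mobile page drops the banking logic, keeping only the whoosh.
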MetaSoundSource /ToyFoxGameplay/Assets/Sound/MSS_ToyFox_SFX_Attack_Seeker_InAir_Banking_Loop
inputs [
 Banking {type=WaveAsset}
 MasterVolume {type=Float}
 Movement.Speed {type=Float}
 Movement.SpeedAzimuth {type=Float}
 Movement.SpeedElevation {type=Float}
 UE.Source.OnPlay {type=Trigger}
 Whoosh {type=WaveAsset}
]
interfaces [Fort.Movement:1.0, UE.OutputFormat.Stereo:1.0, UE.Source:1.1]
outputs [
 "UE.OutputFormat.Stereo.Audio:0" {type=Audio}
 "UE.OutputFormat.Stereo.Audio:1" {type=Audio}
]
pages [
 Default {
  is_build_graph=true
  input_defaults {
   Banking=/Game/Athena/Sounds/Weapons/Guided_Missile/Unvault/GuidedMissile_Unvault_InAir_Loop_Distort_01
   MasterVolume=0.4
   Movement.Speed=0
   Movement.SpeedAzimuth=0
   Movement.SpeedElevation=0
   UE.Source.OnPlay=null
   Whoosh=/Game/Athena/Sounds/Weapons/Guided_Missile/GuidedMissile_InAir_Banking_Loop_04
  }
  nodes [
   n1 = MapRange.MapRange:Float(In=^Movement.SpeedElevation) {also {defaults {In Range A=-25, In Range B=25, Out Range A=-1}}}
   n2 = &MSP_OneshotGate(Value=AbsoluteValue.Abs:Float(Input=UE.InterpTo:Audio(Target=@n1.Out Value).Value).Value) {also {defaults {OneshotThreshold=0.1, RetriggerDelay=0.2}}}
   ADSREnvelopeADSREnvelope_0 = ADSR Envelope.ADSR Envelope:Float(Trigger Attack=@n2.OnGateEnter, Trigger Release=@n2.OnGateExit) {also {defaults {Attack Time=1, Release Time=0.5, Sustain Level=1}}}
   n3 = MapRange.MapRange:Float(In=^Movement.SpeedAzimuth) {also {defaults {In Range A=-25, In Range B=25, Out Range A=-1}}}
   n4 = &MSP_OneshotGate(Value=AbsoluteValue.Abs:Float(Input=UE.InterpTo:Audio(Target=@n3.Out Value).Value).Value) {also {defaults {OneshotThreshold=0.1, RetriggerDelay=0.2}}}
   ADSREnvelopeADSREnvelope_1 = ADSR Envelope.ADSR Envelope:Float(Trigger Attack=@n4.OnGateEnter, Trigger Release=@n4.OnGateExit) {also {defaults {Attack Time=1, Release Time=0.5, Sustain Level=1}}}
   n5 = Max.Max:Float(B=@ADSREnvelopeADSREnvelope_0.Out Envelope, A=@ADSREnvelopeADSREnvelope_1.Out Envelope)
   n6 = VariableAccessor.Float(Variable=VariableMutator.Float(Variable=InitVariable.Float().Variable, Value=@n5.Value).Variable)
   n7 = InitVariable.Trigger()
   n8 = UE.TriggerOnThreshold:Float(In=@n5.Value) {also {defaults {Threshold=0}}}
   n9 = InitVariable.Trigger()
   n10 = UE.TriggerOnThreshold:Float(In=@n5.Value) {also {defaults {Threshold=0.01, Type=1}}}
   WavePlayer_0 = UE.Wave Player:Stereo(
    Wave Asset=^Banking
    Play=VariableAccessor.Trigger(Variable=VariableMutator.Trigger(Variable=@n7.Variable, Value=@n8.Out).Variable).Value
    Stop=VariableAccessor.Trigger(Variable=VariableMutator.Trigger(Variable=@n9.Variable, Value=@n10.Out).Variable).Value
   ) {also {defaults {Loop="<MetaSoundBuilderResult.SUCCEEDED: 0>"}, version [1,1]}}
   Multiply_0 = UE.Multiply:Audio by Float(AdditionalOperands=@n6.Value, PrimaryOperand=@WavePlayer_0.Out Right)
   Multiply_1 = UE.Multiply:Audio by Float(AdditionalOperands=@n6.Value, PrimaryOperand=@WavePlayer_0.Out Left)
   WavePlayer_1 = UE.Wave Player:Stereo(Play=^UE.Source.OnPlay, Wave Asset=^Whoosh) {also {defaults {Loop="<MetaSoundBuilderResult.SUCCEEDED: 0>"}, version [1,1]}}
   AudioMixerAudioMixerStereo2_0 = "AudioMixer.Audio Mixer (Stereo, 2)"(In 1 R=@Multiply_0.Out, In 1 L=@Multiply_1.Out, In 0 L=@WavePlayer_1.Out Left, In 0 R=@WavePlayer_1.Out Right) {also {defaults {Gain 1=0.8}}}
   Multiply_2 = UE.Multiply:Audio by Float(AdditionalOperands=^MasterVolume, PrimaryOperand=@AudioMixerAudioMixerStereo2_0.Out L)
   Multiply_3 = UE.Multiply:Audio by Float(AdditionalOperands=^MasterVolume, PrimaryOperand=@AudioMixerAudioMixerStereo2_0.Out R)
  ]
  links [{dst="@UE.OutputFormat.Stereo.Audio:0", src=Multiply_2.Out}, {dst="@UE.OutputFormat.Stereo.Audio:1", src=Multiply_3.Out}]
 }
 Mobile {
  nodes [
   WavePlayer_0 = UE.Wave Player:Stereo(Play=^UE.Source.OnPlay, Wave Asset=^Whoosh) {also {defaults {Loop="<MetaSoundBuilderResult.SUCCEEDED: 0>"}, version [1,1]}}
   AudioMixerAudioMixerStereo2_0 = "AudioMixer.Audio Mixer (Stereo, 2)"(In 0 L=@WavePlayer_0.Out Left, In 0 R=@WavePlayer_0.Out Right) {also {defaults {Gain 1=0.8}}}
   Multiply_0 = UE.Multiply:Audio by Float(AdditionalOperands=^MasterVolume, PrimaryOperand=@AudioMixerAudioMixerStereo2_0.Out L)
   Multiply_1 = UE.Multiply:Audio by Float(AdditionalOperands=^MasterVolume, PrimaryOperand=@AudioMixerAudioMixerStereo2_0.Out R)
  ]
  links [{dst="@UE.OutputFormat.Stereo.Audio:0", src=Multiply_0.Out}, {dst="@UE.OutputFormat.Stereo.Audio:1", src=Multiply_1.Out}]
 }
]
properties {
 assetImportData=&self:AssetImportData
 attenuationSettings=&/ATT_ToyFox_Attacks_Seeker_Proj
 bLooping=true
 bOverrideConcurrency=true
 duration=10000
 maxDistance=9000
 numChannels=2
 outputFormat=Stereo
 sampleRate=48000
 soundClassObject=&Fort_WeaponRanged_Fire_3P_PSM
 volume=1.5
 concurrencyOverrides {maxCount=1}
}
