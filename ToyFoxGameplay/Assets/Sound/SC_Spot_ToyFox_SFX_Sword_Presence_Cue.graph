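# Looping spatial sound cue for the ToyFox sword presence ambience, classed as outdoor ambient and limited to two concurrent instances.
# One looping wave passes through two endlessly looping envelopers of different periods, one slowly swaying pitch, the other volume.
# That chain feeds two parallel capsule attenuations: a short-range layer and a long-range layer with occlusion, listener focus and custom falloff.
# A mixer sums both before slight random pitch and volume modulation.
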SoundCue /ToyFoxGameplay/Assets/Sound/SC_Spot_ToyFox_SFX_Sword_Presence_Cue
root=Modulator_1
nodes [
 {
  id=Attenuation_0
  children [Enveloper_7]
  properties {
   bOverrideAttenuation=true
   attenuationOverrides {
    attenuationShape=Capsule
    bAttenuateWithLPF=true
    bEnableFocusInterpolation=true
    bEnableListenerFocus=true
    bEnableOcclusion=true
    distanceAlgorithm=Custom
    falloffDistance=6000
    focusAttackInterpSpeed=2
    focusReleaseInterpSpeed=0.9
    focusVolumeAttenuation=1.2
    hPFFrequencyAtMax=3000
    hPFFrequencyAtMin=400
    lPFRadiusMin=2000
    nonFocusDistanceScale=0.9
    nonFocusVolumeAttenuation=0
    occlusionInterpolationTime=0.2
    occlusionLowPassFilterFrequency=1000
    occlusionTraceChannel=ECC_GameTraceChannel11
    occlusionVolumeAttenuation=0.6
    stereoSpread=800
    attenuationShapeExtents=<1500 500>
    customAttenuationCurve {editorCurveData {keys [<at 0 1>,<at 0.5 1>,<at 1 0>]}}
   }
  }
 }
 {
  id=Attenuation_2
  children [Enveloper_7]
  properties {
   bOverrideAttenuation=true
   attenuationOverrides {
    attenuationShape=Capsule
    bAttenuateWithLPF=true
    falloffDistance=2000
    hPFFrequencyAtMax=2000
    hPFFrequencyAtMin=100
    lPFRadiusMin=2000
    nonFocusVolumeAttenuation=0.4
    nonSpatializedRadiusEnd=500
    nonSpatializedRadiusStart=1000
    stereoSpread=800
    attenuationShapeExtents=<1500 500>
   }
  }
 }
 {
  id=Enveloper_6
  children [WavePlayer_0]
  properties {
   bLoop=true
   bLoopIndefinitely=true
   loopEnd=30
   pitchCurve {editorCurveData {keys [{arriveTangent=-0.0013, interpMode=RCIM_Cubic, time=0}, <at 15 1.05 RCIM_Cubic>, {interpMode=RCIM_Cubic, leaveTangent=0.0013, time=30}]}}
   volumeCurve {editorCurveData {keys [<at 0 1>]}}
  }
 }
 {
  id=Enveloper_7
  children [Enveloper_6]
  properties {
   bLoop=true
   bLoopIndefinitely=true
   loopEnd=20
   pitchCurve {editorCurveData {keys [{time=0}]}}
   volumeCurve {editorCurveData {keys [
      {arriveTangent=-0.01, interpMode=RCIM_Cubic, time=0, value=1}
      <at 10 0.8 RCIM_Cubic>
      {interpMode=RCIM_Cubic, leaveTangent=0.01, time=20, value=1}
     ]}}
  }
 }
 {id=Mixer_0, children [Attenuation_2, Attenuation_0], properties {inputVolume [1,1]}}
 {id=Modulator_1, children [Mixer_0], properties {pitchMax=1.1, pitchMin=0.9, volumeMax=1.1, volumeMin=0.9}}
 {id=WavePlayer_0, properties {bLooping=true, soundWaveAssetPtr=/ToyFoxGameplay/Assets/Sound/SoundWaves/ToyFox_SFX_Sword_Presence_Loop_02}}
]
properties {
 bEnableBusSends=false
 bOverrideConcurrency=true
 maxDistance=7500
 pitchMultiplier=1.1
 soundClassObject=&Fort_Ambient_Outdoors_Athena
 volumeMultiplier=1
 concurrencyOverrides {maxCount=2}
 preEffectBusSends [{}]
}
