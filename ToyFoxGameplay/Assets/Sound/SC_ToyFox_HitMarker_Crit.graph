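# Sound Cue for the ToyFox critical hit marker, playing a weakpoint impact sound.
# A Random node picks one of four equally weighted impact waves, then a Modulator varies pitch and volume at or below their original levels.
# It uses the 2D instigator hit-notification sound class and shared hit-marker attenuation settings.
# A shared concurrency set limits hit notifications to two, stopping the oldest.
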SoundCue /ToyFoxGameplay/Assets/Sound/SC_ToyFox_HitMarker_Crit
root=Modulator_0
nodes [
 {id=Modulator_0, children [Random_1], properties {pitchMax=1, pitchMin=0.7, volumeMax=1, volumeMin=0.6}}
 {id=Random_1, children [WavePlayer_0,WavePlayer_7,WavePlayer_8,WavePlayer_1], properties {weights [1,1,1,1]}}
 {id=WavePlayer_0, properties {soundWaveAssetPtr=/ToyFoxGameplay/Assets/Sound/SoundWaves/ToyFox_SFX_Weakpoint_Impact_01}}
 {id=WavePlayer_1, properties {soundWaveAssetPtr=/ToyFoxGameplay/Assets/Sound/SoundWaves/ToyFox_SFX_Weakpoint_Impact_04}}
 {id=WavePlayer_7, properties {soundWaveAssetPtr=/ToyFoxGameplay/Assets/Sound/SoundWaves/ToyFox_SFX_Weakpoint_Impact_02}}
 {id=WavePlayer_8, properties {soundWaveAssetPtr=/ToyFoxGameplay/Assets/Sound/SoundWaves/ToyFox_SFX_Weakpoint_Impact_03}}
]
properties {
 attenuationSettings=/Game/Athena/Sounds/HitNotifications/Fort_HitMarker_Instigate_Att
 maxDistance=2097152
 soundClassObject=&Fort_2DHitNotifications_Instigate_PSM
 volumeMultiplier=1
 attenuationOverrides {bAttenuate=false, bEnableListenerFocus=true, nonFocusVolumeAttenuation=0.1}
 concurrencyOverrides {maxCount=2, resolutionRule=StopOldest}
 concurrencySet [&HitNotifications_TwoMaximum_StopOldest_VolumeScale]
}
