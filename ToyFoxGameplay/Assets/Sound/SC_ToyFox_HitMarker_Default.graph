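# Sound cue for the default ToyFox hit marker, played as a hit notification through an instigator-side 2D sound class.
# A random node picks one of six equally weighted Ferret vehicle metal impact waves, then a modulator varies pitch and volume downward for variety.
# It references shared hit marker attenuation and a concurrency set that allows two voices, stopping the oldest.
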SoundCue /ToyFoxGameplay/Assets/Sound/SC_ToyFox_HitMarker_Default
root=Modulator_0
nodes [
 {id=Modulator_0, children [Random_1], properties {pitchMax=1, pitchMin=0.7, volumeMax=1, volumeMin=0.6}}
 {id=Random_1, children [WavePlayer_0,WavePlayer_1,WavePlayer_2,WavePlayer_3,WavePlayer_5,WavePlayer_6], properties {weights [1,1,1,1,1,1]}}
 {id=WavePlayer_0, properties {soundWaveAssetPtr=/Game/Athena/Sounds/Vehicles/Ferret/Metal_Impact_Ferret_Player_10}}
 {id=WavePlayer_1, properties {soundWaveAssetPtr=/Game/Athena/Sounds/Vehicles/Ferret/Metal_Impact_Ferret_Player_11}}
 {id=WavePlayer_2, properties {soundWaveAssetPtr=/Game/Athena/Sounds/Vehicles/Ferret/Metal_Impact_Ferret_Player_01}}
 {id=WavePlayer_3, properties {soundWaveAssetPtr=/Game/Athena/Sounds/Vehicles/Ferret/Metal_Impact_Ferret_Player_02}}
 {id=WavePlayer_5, properties {soundWaveAssetPtr=/Game/Athena/Sounds/Vehicles/Ferret/Metal_Impact_Ferret_Player_03}}
 {id=WavePlayer_6, properties {soundWaveAssetPtr=/Game/Athena/Sounds/Vehicles/Ferret/Metal_Impact_Ferret_Player_04}}
]
properties {
 attenuationSettings=/Game/Athena/Sounds/HitNotifications/Fort_HitMarker_Instigate_Att
 maxDistance=2097152
 soundClassObject=&Fort_2DHitNotifications_Instigate_PSM
 volumeMultiplier=0.6
 attenuationOverrides {bAttenuate=false, bEnableListenerFocus=true, nonFocusVolumeAttenuation=0.1}
 concurrencyOverrides {maxCount=2, resolutionRule=StopOldest}
 concurrencySet [&HitNotifications_TwoMaximum_StopOldest_VolumeScale]
}
