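# Sound cue for the ToyFox stand activation sound effect, routed through the scripted 3D events sound class.
# A single wave player plays the Toyfox_SFX_Stand_01 wave directly, with no variation or mixing.
# It references the ToyFox attacks attenuation asset for 3D falloff.
# A concurrency override allows only one instance at a time and blocks new plays while it is active.
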SoundCue /ToyFoxGameplay/Assets/Sound/SC_ToyFox_SFX_Activate_Stand
root=WavePlayer_0
nodes [{id=WavePlayer_0, properties {soundWaveAssetPtr=/ToyFoxGameplay/Assets/Sound/SoundWaves/Toyfox_SFX_Stand_01}}]
properties {
 attenuationSettings=&/ATT_ToyFox_Attacks
 bOverrideConcurrency=true
 maxDistance=90000
 soundClassObject=/Game/Sounds/FortSoundClassesAndModes/Fort_Events_Scripted_3D
 volumeMultiplier=1
 attenuationOverrides {falloffDistance=10000, stereoSpread=5000, attenuationShapeExtents {x=1000}}
 concurrencyOverrides {maxCount=1, resolutionRule=PreventNew}
}
