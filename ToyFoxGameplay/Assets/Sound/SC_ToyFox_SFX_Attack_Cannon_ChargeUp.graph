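# Sound cue for the ToyFox cannon attack charge-up, using the boss fight SFX sound class and ToyFox attack attenuation.
# A mixer starts the charge wave at boosted volume through a modulator with pitch variation disabled, and layers in the ready wave after a fixed delay, so readiness sounds on a set timeline.
# A concurrency override limits simultaneous instances to two.
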SoundCue /ToyFoxGameplay/Assets/Sound/SC_ToyFox_SFX_Attack_Cannon_ChargeUp
root=Mixer_1
nodes [
 {id=Delay_0, children [WavePlayer_10], properties {delayMax=1.6, delayMin=1.6}}
 {id=Mixer_1, children [Modulator_2,Delay_0], properties {inputVolume [1.5,1]}}
 {id=Modulator_2, children [WavePlayer_11], properties {pitchMax=1, pitchMin=1}}
 {id=WavePlayer_10, properties {soundWaveAssetPtr=/ToyFoxGameplay/Assets/Sound/SoundWaves/ToyFox_SFX_Attack_Cannon_Ready_01}}
 {id=WavePlayer_11, properties {soundWaveAssetPtr=/ToyFoxGameplay/Assets/Sound/SoundWaves/ToyFox_SFX_Attack_Cannon_Charge_01}}
]
properties {
 attenuationSettings=&/ATT_ToyFox_Attacks
 bOverrideConcurrency=true
 maxDistance=90000
 soundClassObject=&Fort_BossFight_SFX
 volumeMultiplier=1.5
 concurrencyOverrides {maxCount=2}
}
