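# Sound cue for the looping power sound of ToyFox's charged cannon attack.
# A looping wave player feeds an Enveloper whose volume curve rises from silence to full over about one second, then holds, so the loop swells in.
# It routes to the boss-fight SFX sound class, uses the ToyFox attack attenuation, and overrides concurrency to cap simultaneous instances.
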SoundCue /ToyFoxGameplay/Assets/Sound/SC_ToyFox_SFX_Attack_Cannon_Charged_Loop
root=Enveloper_0
nodes [
 {
  id=Enveloper_0
  children [WavePlayer_0]
  properties {
   pitchCurve {editorCurveData {keys [{time=0}]}}
   volumeCurve {editorCurveData {keys [<at 0 0>,<at 0.0032 0>,<at 0.5936 0.1487>,<at 1 1>,<at 1.6 1>]}}
  }
 }
 {
  id=WavePlayer_0
  properties {bLooping=true, soundWaveAssetPtr=/ToyFoxGameplay/Assets/Sound/SoundWaves/ToyFox_SFX_Attack_Cannon_Power_Loop_01}
 }
]
properties {
 attenuationSettings=&/ATT_ToyFox_Attacks
 bOverrideConcurrency=true
 maxDistance=90000
 soundClassObject=&Fort_BossFight_SFX
 volumeMultiplier=1
 concurrencyOverrides {maxCount=2}
}
