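# Sound cue for the ToyFox cannon attack's projectile explosion, routed to the boss fight SFX sound class.
# A Random node picks one of four equally weighted explosion waves, and a Modulator at default settings varies its pitch and volume.
# It reuses the ToyFox rocket explosion attenuation with a boosted volume.
# A concurrency override limits it to one playing instance per owner.
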SoundCue /ToyFoxGameplay/Assets/Sound/SC_ToyFox_SFX_Attack_Cannon_Expl
root=Modulator_0
nodes [
 {id=Modulator_0, children [Random_0]}
 {id=Random_0, children [WavePlayer_0,WavePlayer_1,WavePlayer_2,WavePlayer_3], properties {weights [1,1,1,1]}}
 {id=WavePlayer_0, properties {soundWaveAssetPtr=/ToyFoxGameplay/Assets/Sound/SoundWaves/ToyFox_SFX_Attack_Cannon_Proj_Expl_01}}
 {id=WavePlayer_1, properties {soundWaveAssetPtr=/ToyFoxGameplay/Assets/Sound/SoundWaves/ToyFox_SFX_Attack_Cannon_Proj_Expl_02}}
 {id=WavePlayer_2, properties {soundWaveAssetPtr=/ToyFoxGameplay/Assets/Sound/SoundWaves/ToyFox_SFX_Attack_Cannon_Proj_Expl_03}}
 {id=WavePlayer_3, properties {soundWaveAssetPtr=/ToyFoxGameplay/Assets/Sound/SoundWaves/ToyFox_SFX_Attack_Cannon_Proj_Expl_04}}
]
properties {
 attenuationSettings=&/ATT_ToyFox_Attacks_Rocket_Expl
 bOverrideConcurrency=true
 maxDistance=82500
 soundClassObject=&Fort_BossFight_SFX
 volumeMultiplier=1.5
 concurrencyOverrides {bLimitToOwner=true, maxCount=1}
}
