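# Sound cue for the flyby sound of ToyFox's cannon attack projectile, classed as boss fight SFX.
# A Random node picks one of four equally weighted flyby wave variations and feeds it through a Modulator at the root.
# The cue boosts volume, references the cannon projectile attenuation asset, and overrides concurrency to cap simultaneous instances.
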SoundCue /ToyFoxGameplay/Assets/Sound/SC_ToyFox_SFX_Attack_Cannon_Prj_Flyby
root=Modulator_0
nodes [
 {id=Modulator_0, children [Random_0]}
 {id=Random_0, children [WavePlayer_1,WavePlayer_2,WavePlayer_3,WavePlayer_4], properties {weights [1,1,1,1]}}
 {id=WavePlayer_1, properties {soundWaveAssetPtr=/ToyFoxGameplay/Assets/Sound/SoundWaves/ToyFox_SFX_Attack_Cannon_Proj_FlyBy_01}}
 {id=WavePlayer_2, properties {soundWaveAssetPtr=/ToyFoxGameplay/Assets/Sound/SoundWaves/ToyFox_SFX_Attack_Cannon_Proj_FlyBy_02}}
 {id=WavePlayer_3, properties {soundWaveAssetPtr=/ToyFoxGameplay/Assets/Sound/SoundWaves/ToyFox_SFX_Attack_Cannon_Proj_FlyBy_03}}
 {id=WavePlayer_4, properties {soundWaveAssetPtr=/ToyFoxGameplay/Assets/Sound/SoundWaves/ToyFox_SFX_Attack_Cannon_Proj_FlyBy_04}}
]
properties {
 attenuationSettings=&/ATT_ToyFox_Attacks_Cannon_Proj
 bOverrideConcurrency=true
 maxDistance=10000
 soundClassObject=&Fort_BossFight_SFX
 volumeMultiplier=2
 attenuationOverrides {falloffDistance=20000}
 concurrencyOverrides {maxCount=3}
}
