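# Sound cue for the projectile launch of the ToyFox cannon attack.
# A Random node picks one of four launch waves with equal weight.
# The cue plays through the boss-fight SFX class with boosted volume and references the shared ToyFox attack attenuation.
# It overrides concurrency to cap simultaneous instances at a small count.
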SoundCue /ToyFoxGameplay/Assets/Sound/SC_ToyFox_SFX_Attack_Cannon_Prj_Launch
root=Random_1
nodes [
 {id=Random_1, children [WavePlayer_0,WavePlayer_5,WavePlayer_6,WavePlayer_7], properties {weights [1,1,1,1]}}
 {id=WavePlayer_0, properties {soundWaveAssetPtr=/ToyFoxGameplay/Assets/Sound/SoundWaves/ToyFox_SFX_Attack_Cannon_Proj_Launch_01}}
 {id=WavePlayer_5, properties {soundWaveAssetPtr=/ToyFoxGameplay/Assets/Sound/SoundWaves/ToyFox_SFX_Attack_Cannon_Proj_Launch_02}}
 {id=WavePlayer_6, properties {soundWaveAssetPtr=/ToyFoxGameplay/Assets/Sound/SoundWaves/ToyFox_SFX_Attack_Cannon_Proj_Launch_03}}
 {id=WavePlayer_7, properties {soundWaveAssetPtr=/ToyFoxGameplay/Assets/Sound/SoundWaves/ToyFox_SFX_Attack_Cannon_Proj_Launch_04}}
]
properties {
 attenuationSettings=&/ATT_ToyFox_Attacks
 bOverrideConcurrency=true
 maxDistance=90000
 soundClassObject=&Fort_BossFight_SFX
 volumeMultiplier=1.5
 attenuationOverrides {falloffDistance=20000}
 concurrencyOverrides {maxCount=3}
}
