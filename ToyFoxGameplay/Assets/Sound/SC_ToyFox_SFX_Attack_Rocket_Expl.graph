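# Sound cue for the ToyFox rocket attack explosion, layering close and distant variations in the boss fight SFX sound class.
# A mixer blends the two layers with the close one boosted; each randomly picks one of four equally weighted waves.
# Each layer passes through a modulator and its own attenuation node, so close and distant sounds follow separate distance settings.
# A concurrency override limits playback to one instance per owner.
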SoundCue /ToyFoxGameplay/Assets/Sound/SC_ToyFox_SFX_Attack_Rocket_Expl
root=Mixer_1
nodes [
 {id=Attenuation_2, children [Random_4], properties {attenuationSettings=&/ATT_ToyFox_Attacks_Distant}}
 {id=Attenuation_3, children [Modulator_2], properties {attenuationSettings=&/ATT_ToyFox_Attacks_Rocket_Expl}}
 {id=Mixer_1, children [Attenuation_3, Modulator_3], properties {inputVolume [1.5,1]}}
 {id=Modulator_2, children [Random_3]}
 {id=Modulator_3, children [Attenuation_2]}
 {id=Random_3, children [WavePlayer_11, WavePlayer_12, WavePlayer_13, WavePlayer_14], properties {weights [1,1,1,1]}}
 {id=Random_4, children [WavePlayer_15, WavePlayer_16, WavePlayer_17, WavePlayer_18], properties {weights [1,1,1,1]}}
 {id=WavePlayer_11, properties {soundWaveAssetPtr=/ToyFoxGameplay/Assets/Sound/SoundWaves/ToyFox_SFX_Attack_Rocket_Expl_Close_01}}
 {id=WavePlayer_12, properties {soundWaveAssetPtr=/ToyFoxGameplay/Assets/Sound/SoundWaves/ToyFox_SFX_Attack_Rocket_Expl_Close_02}}
 {id=WavePlayer_13, properties {soundWaveAssetPtr=/ToyFoxGameplay/Assets/Sound/SoundWaves/ToyFox_SFX_Attack_Rocket_Expl_Close_03}}
 {id=WavePlayer_14, properties {soundWaveAssetPtr=/ToyFoxGameplay/Assets/Sound/SoundWaves/ToyFox_SFX_Attack_Rocket_Expl_Close_04}}
 {id=WavePlayer_15, properties {soundWaveAssetPtr=/ToyFoxGameplay/Assets/Sound/SoundWaves/ToyFox_SFX_Attack_Rocket_Expl_Distant_01}}
 {id=WavePlayer_16, properties {soundWaveAssetPtr=/ToyFoxGameplay/Assets/Sound/SoundWaves/ToyFox_SFX_Attack_Rocket_Expl_Distant_02}}
 {id=WavePlayer_17, properties {soundWaveAssetPtr=/ToyFoxGameplay/Assets/Sound/SoundWaves/ToyFox_SFX_Attack_Rocket_Expl_Distant_03}}
 {id=WavePlayer_18, properties {soundWaveAssetPtr=/ToyFoxGameplay/Assets/Sound/SoundWaves/ToyFox_SFX_Attack_Rocket_Expl_Distant_04}}
]
properties {
 attenuationSettings=&/ATT_ToyFox_Attacks
 bOverrideConcurrency=true
 maxDistance=100000
 soundClassObject=&Fort_BossFight_SFX
 volumeMultiplier=1
 concurrencyOverrides {bLimitToOwner=true, maxCount=1}
}
