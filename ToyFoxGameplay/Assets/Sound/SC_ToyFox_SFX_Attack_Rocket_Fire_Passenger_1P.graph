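# Sound cue for the ToyFox rocket attack firing sound, named as the first-person passenger variant.
# A Random node picks one of three equally weighted close-range fire waves and feeds a Modulator at the root.
# It uses the boss fight SFX sound class and ToyFox attack attenuation, with boosted volume and a concurrency override capping simultaneous instances.
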SoundCue /ToyFoxGameplay/Assets/Sound/SC_ToyFox_SFX_Attack_Rocket_Fire_Passenger_1P
root=Modulator_0
nodes [
 {id=Modulator_0, children [Random_0]}
 {id=Random_0, children [WavePlayer_1,WavePlayer_2,WavePlayer_3], properties {weights [1,1,1]}}
 {id=WavePlayer_1, properties {soundWaveAssetPtr=/ToyFoxGameplay/Assets/Sound/SoundWaves/ToyFox_SFX_Attack_Rocket_Fire_Close_01}}
 {id=WavePlayer_2, properties {soundWaveAssetPtr=/ToyFoxGameplay/Assets/Sound/SoundWaves/ToyFox_SFX_Attack_Rocket_Fire_Close_02}}
 {id=WavePlayer_3, properties {soundWaveAssetPtr=/ToyFoxGameplay/Assets/Sound/SoundWaves/ToyFox_SFX_Attack_Rocket_Fire_Close_03}}
]
properties {
 attenuationSettings=&/ATT_ToyFox_Attacks
 bOverrideConcurrency=true
 maxDistance=90000
 soundClassObject=&Fort_BossFight_SFX
 volumeMultiplier=1.5
 attenuationOverrides {falloffDistance=20000}
 concurrencyOverrides {maxCount=2}
}
