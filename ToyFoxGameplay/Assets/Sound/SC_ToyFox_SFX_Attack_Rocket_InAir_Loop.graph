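# Sound cue for the looping in-flight sound of ToyFox's rocket attack projectile, on the boss-fight SFX class.
# It loops a reused DrunkenQuad rocket wave through slight upward pitch randomization and an enveloper that fades volume in over the first second.
# A projectile attenuation asset spatializes it, and an outer modulator fixes a lower volume.
# A concurrency override caps it at a few simultaneous instances.
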SoundCue /ToyFoxGameplay/Assets/Sound/SC_ToyFox_SFX_Attack_Rocket_InAir_Loop
root=Modulator_0
nodes [
 {id=Attenuation_1, children [Enveloper_4], properties {attenuationSettings=&/ATT_ToyFox_Attacks_Rocket_Proj}}
 {
  id=Enveloper_4
  children [Modulator_2]
  properties {
   pitchCurve {editorCurveData {keys [{time=0}]}}
   volumeCurve {editorCurveData {keys [{interpMode=RCIM_Cubic, tangentMode=RCTM_User, time=0, value=0}, {interpMode=RCIM_Cubic, tangentMode=RCTM_User, time=1, value=1}]}}
  }
 }
 {id=Modulator_0, children [Attenuation_1], properties {pitchMax=1, pitchMin=1, volumeMax=0.4, volumeMin=0.4}}
 {id=Modulator_2, children [WavePlayer_3], properties {pitchMax=1.2, pitchMin=1.1}}
 {id=WavePlayer_3, properties {bLooping=true, soundWaveAssetPtr=/Game/Sounds/Fort_Weapons/Launchers/DrunkenQuad/DrunkenRocket_InAir_Loop_01}}
]
properties {
 bOverrideConcurrency=true
 maxDistance=11500
 soundClassObject=&Fort_BossFight_SFX
 volumeMultiplier=1.5
 attenuationOverrides {falloffDistance=20000}
 concurrencyOverrides {maxCount=3}
}
