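# Sound cue for the flyby whoosh of ToyFox's rocket attack projectile, classed as boss fight SFX.
# A random node picks one of four equally weighted whoosh-by waves and feeds a modulator at the root for per-play variation.
# It uses the ToyFox rocket projectile attenuation asset, boosts volume, and overrides concurrency to cap how many flybys play at once.
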SoundCue /ToyFoxGameplay/Assets/Sound/SC_ToyFox_SFX_Attack_Rocket_Prj_Flyby
root=Modulator_0
nodes [
 {id=Modulator_0, children [Random_1]}
 {id=Random_1, children [WavePlayer_0,WavePlayer_5,WavePlayer_6,WavePlayer_7], properties {weights [1,1,1,1]}}
 {id=WavePlayer_0, properties {soundWaveAssetPtr=/ToyFoxGameplay/Assets/Sound/SoundWaves/ToyFox_SFX_Attack_Rocket_Whoosh_By_01}}
 {id=WavePlayer_5, properties {soundWaveAssetPtr=/ToyFoxGameplay/Assets/Sound/SoundWaves/ToyFox_SFX_Attack_Rocket_Whoosh_By_02}}
 {id=WavePlayer_6, properties {soundWaveAssetPtr=/ToyFoxGameplay/Assets/Sound/SoundWaves/ToyFox_SFX_Attack_Rocket_Whoosh_By_03}}
 {id=WavePlayer_7, properties {soundWaveAssetPtr=/ToyFoxGameplay/Assets/Sound/SoundWaves/ToyFox_SFX_Attack_Rocket_Whoosh_By_04}}
]
properties {
 attenuationSettings=&/ATT_ToyFox_Attacks_Rocket_Proj
 bOverrideConcurrency=true
 maxDistance=11500
 soundClassObject=&Fort_BossFight_SFX
 volumeMultiplier=1.5
 attenuationOverrides {falloffDistance=20000}
 concurrencyOverrides {maxCount=3}
}
