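# Sound cue for the third-person firing sound of a ToyFox seeker attack, passenger variant, routed to the boss fight SFX sound class.
# A mixer layers two branches, each picking one of three waves at random.
# The near branch plays guided missile activation sounds with the attack attenuation and overridden falloff and shape.
# The distant branch plays louder-mixed rocket fire tails with distant attenuation, and concurrency is capped by an override.
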SoundCue /ToyFoxGameplay/Assets/Sound/SC_ToyFox_SFX_Attack_Seeker_Fire_Passenger_3P
root=Mixer_3
nodes [
 {id=Attenuation_0, children [Random_1], properties {attenuationSettings=&/ATT_ToyFox_Attacks_Distant}}
 {
  id=Attenuation_10
  children [Random_0]
  properties {
   attenuationSettings=&/ATT_ToyFox_Attacks
   attenuationOverrides {falloffDistance=10000, stereoSpread=512, attenuationShapeExtents {x=2000}}
  }
 }
 {id=Mixer_3, children [Attenuation_10, Attenuation_0], properties {inputVolume [1,2]}}
 {id=Random_0, children [WavePlayer_0,WavePlayer_1,WavePlayer_5], properties {weights [1,1,1]}}
 {id=Random_1, children [WavePlayer_6,WavePlayer_7,WavePlayer_8], properties {weights [1,1,1]}}
 {id=WavePlayer_0, properties {soundWaveAssetPtr=/Game/Athena/Sounds/Weapons/Guided_Missile/GuidedMissile_Activate_01}}
 {id=WavePlayer_1, properties {soundWaveAssetPtr=/Game/Athena/Sounds/Weapons/Guided_Missile/GuidedMissile_Activate_02}}
 {id=WavePlayer_5, properties {soundWaveAssetPtr=/Game/Athena/Sounds/Weapons/Guided_Missile/GuidedMissile_Activate_03}}
 {id=WavePlayer_6, properties {soundWaveAssetPtr=/ToyFoxGameplay/Assets/Sound/SoundWaves/ToyFox_SFX_Attack_Rocket_Fire_Distant_01}}
 {id=WavePlayer_7, properties {soundWaveAssetPtr=/ToyFoxGameplay/Assets/Sound/SoundWaves/ToyFox_SFX_Attack_Rocket_Fire_Distant_02}}
 {id=WavePlayer_8, properties {soundWaveAssetPtr=/ToyFoxGameplay/Assets/Sound/SoundWaves/ToyFox_SFX_Attack_Rocket_Fire_Distant_03}}
]
properties {
 bOverrideConcurrency=true
 maxDistance=100000
 soundClassObject=&Fort_BossFight_SFX
 volumeMultiplier=1
 attenuationOverrides {falloffDistance=20000}
 concurrencyOverrides {maxCount=3}
}
