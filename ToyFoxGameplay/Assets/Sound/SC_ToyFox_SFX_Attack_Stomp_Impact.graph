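# Sound cue for the ToyFox stomp attack impact SFX, classed as boss fight SFX with boosted volume.
# A Branch on the "islocal" bool parameter picks between two Attenuation nodes that share one wave player.
# The true path has no attenuation asset assigned, the false path uses the ToyFox attacks attenuation, and the parameter-unset input is empty.
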SoundCue /ToyFoxGameplay/Assets/Sound/SC_ToyFox_SFX_Attack_Stomp_Impact
root=Branch_0
nodes [
 {id=Attenuation_0, children [WavePlayer_0]}
 {id=Attenuation_1, children [WavePlayer_0], properties {attenuationSettings=&/ATT_ToyFox_Attacks}}
 {id=Branch_0, children [Attenuation_0, Attenuation_1, null], properties {boolParameterName=islocal}}
 {id=WavePlayer_0, properties {soundWaveAssetPtr=/ToyFoxGameplay/Assets/Sound/SoundWaves/ToyFox_SFX_Attack_Stomp_Impact_01}}
]
properties {
 maxDistance=2097152
 soundClassObject=&Fort_BossFight_SFX
 volumeMultiplier=1.5
 attenuationOverrides {falloffDistance=25000, attenuationShapeExtents {x=9000}}
 concurrencyOverrides {maxCount=1, resolutionRule=PreventNew}
}
