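# Sound cue for the windup sound effect of the ToyFox stomp attack.
# A single wave player feeds the windup wave into a modulator at the root, which applies default random variation.
# The cue boosts volume, routes to the boss fight SFX sound class, uses the shared ToyFox attack attenuation, and overrides concurrency to cap simultaneous instances at two.
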SoundCue /ToyFoxGameplay/Assets/Sound/SC_ToyFox_SFX_Attack_Stomp_Windup
root=Modulator_2
nodes [
 {id=Modulator_2, children [WavePlayer_0]}
 {id=WavePlayer_0, properties {soundWaveAssetPtr=/ToyFoxGameplay/Assets/Sound/SoundWaves/ToyFox_SFX_Attack_Stomp_Windup_01}}
]
properties {
 attenuationSettings=&/ATT_ToyFox_Attacks
 bOverrideConcurrency=true
 maxDistance=90000
 soundClassObject=&Fort_BossFight_SFX
 volumeMultiplier=2.233
 concurrencyOverrides {maxCount=2}
}
