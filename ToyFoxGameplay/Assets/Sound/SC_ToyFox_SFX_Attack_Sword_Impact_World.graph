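# Sound cue for the ToyFox sword attack's world impact, routed to the boss fight SFX sound class.
# A mixer layers randomly chosen close impact waves with randomly chosen, quieter distant ones, each layer using its own attenuation asset.
# Each layer passes through a modulator for variation, with the distant one randomly pitched downward.
# A concurrency override caps simultaneous instances at two, stopping the oldest.
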SoundCue /ToyFoxGameplay/Assets/Sound/SC_ToyFox_SFX_Attack_Sword_Impact_World
root=Mixer_0
nodes [
 {id=Attenuation_0, children [Random_2], properties {attenuationSettings=&/ATT_ToyFox_Attacks}}
 {id=Attenuation_1, children [Random_1], properties {attenuationSettings=&/ATT_ToyFox_Attacks_Distant}}
 {id=Mixer_0, children [Modulator_0,Modulator_1], properties {inputVolume [1,0.7]}}
 {id=Modulator_0, children [Attenuation_0]}
 {id=Modulator_1, children [Attenuation_1], properties {pitchMax=1, pitchMin=0.8}}
 {id=Random_1, children [WavePlayer_3,WavePlayer_4,WavePlayer_5], properties {weights [1,1,1]}}
 {id=Random_2, children [WavePlayer_7, WavePlayer_8, WavePlayer_9, WavePlayer_10], properties {weights [1,1,1,1]}}
 {id=WavePlayer_10, properties {soundWaveAssetPtr=/ToyFoxGameplay/Assets/Sound/SoundWaves/Toyfox_Melee_Impact_Close_04}}
 {id=WavePlayer_3, properties {soundWaveAssetPtr=/ToyFoxGameplay/Assets/Sound/SoundWaves/Toyfox_Melee_Impact_Distant_01}}
 {id=WavePlayer_4, properties {soundWaveAssetPtr=/ToyFoxGameplay/Assets/Sound/SoundWaves/Toyfox_Melee_Impact_Distant_02}}
 {id=WavePlayer_5, properties {soundWaveAssetPtr=/ToyFoxGameplay/Assets/Sound/SoundWaves/Toyfox_Melee_Impact_Distant_03}}
 {id=WavePlayer_7, properties {soundWaveAssetPtr=/ToyFoxGameplay/Assets/Sound/SoundWaves/Toyfox_Melee_Impact_Close_01}}
 {id=WavePlayer_8, properties {soundWaveAssetPtr=/ToyFoxGameplay/Assets/Sound/SoundWaves/Toyfox_Melee_Impact_Close_02}}
 {id=WavePlayer_9, properties {soundWaveAssetPtr=/ToyFoxGameplay/Assets/Sound/SoundWaves/Toyfox_Melee_Impact_Close_03}}
]
properties {
 bOverrideConcurrency=true
 maxDistance=100000
 soundClassObject=&Fort_BossFight_SFX
 volumeMultiplier=0.7943
 attenuationOverrides {falloffDistance=25000, attenuationShapeExtents {x=6000}}
 concurrencyOverrides {maxCount=2, resolutionRule=StopOldest}
}
