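# Sound cue for the ToyFox sword swing attack sound effect, routed to the boss fight SFX sound class.
# An Islocal branch picks a default attenuation node when true or one using the ToyFox attack attenuation when false, with the unset input empty.
# Both paths share a slight downward pitch modulator over a random pick of three equally weighted swing waves.
# A concurrency override caps it at two voices, stopping the oldest.
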SoundCue /ToyFoxGameplay/Assets/Sound/SC_ToyFox_SFX_Attack_Sword_Swing_01
root=Branch_0
nodes [
 {id=Attenuation_0, children [Modulator_1]}
 {id=Attenuation_1, children [Modulator_1], properties {attenuationSettings=&/ATT_ToyFox_Attacks}}
 {id=Branch_0, children [Attenuation_0, Attenuation_1, null], properties {boolParameterName=Islocal}}
 {id=Modulator_1, children [Random_0], properties {pitchMax=1, pitchMin=0.85}}
 {id=Random_0, children [WavePlayer_0,WavePlayer_1,WavePlayer_2], properties {weights [1,1,1]}}
 {id=WavePlayer_0, properties {soundWaveAssetPtr=/ToyFoxGameplay/Assets/Sound/SoundWaves/ToyFox_SFX_Attack_Sword_Swing_FH_01}}
 {id=WavePlayer_1, properties {soundWaveAssetPtr=/ToyFoxGameplay/Assets/Sound/SoundWaves/ToyFox_SFX_Attack_Sword_Swing_FH_02}}
 {id=WavePlayer_2, properties {soundWaveAssetPtr=/ToyFoxGameplay/Assets/Sound/SoundWaves/ToyFox_SFX_Attack_Sword_Swing_FH_03}}
]
properties {
 attenuationSettings=&ATT_ToyFox_Attacks
 bOverrideConcurrency=true
 maxDistance=2097152
 soundClassObject=&Fort_BossFight_SFX
 volumeMultiplier=2
 attenuationOverrides {falloffDistance=25000, attenuationShapeExtents {x=6000}}
 concurrencyOverrides {maxCount=2, resolutionRule=StopOldest}
}
