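# Sound cue for the second ToyFox sword swing attack SFX, in the boss fight SFX class.
# A random node picks one of three swing waves and feeds a modulator that slightly varies pitch downward.
# An Islocal branch routes that shared chain through an attenuation node without assigned settings when true, ATT_ToyFox_Attacks when false, and nothing when unset.
# Overridden concurrency allows two instances, stopping the oldest.
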SoundCue /ToyFoxGameplay/Assets/Sound/SC_ToyFox_SFX_Attack_Sword_Swing_02
root=Branch_0
nodes [
 {id=Attenuation_0, children [Modulator_0]}
 {id=Attenuation_1, children [Modulator_0], properties {attenuationSettings=&/ATT_ToyFox_Attacks}}
 {id=Branch_0, children [Attenuation_0, Attenuation_1, null], properties {boolParameterName=Islocal}}
 {id=Modulator_0, children [Random_3], properties {pitchMax=1, pitchMin=0.85}}
 {id=Random_3, children [WavePlayer_9, WavePlayer_10, WavePlayer_11], properties {weights [1,1,1]}}
 {id=WavePlayer_10, properties {soundWaveAssetPtr=/ToyFoxGameplay/Assets/Sound/SoundWaves/ToyFox_SFX_Attack_Sword_Swing_BH_02}}
 {id=WavePlayer_11, properties {soundWaveAssetPtr=/ToyFoxGameplay/Assets/Sound/SoundWaves/ToyFox_SFX_Attack_Sword_Swing_BH_03}}
 {id=WavePlayer_9, properties {soundWaveAssetPtr=/ToyFoxGameplay/Assets/Sound/SoundWaves/ToyFox_SFX_Attack_Sword_Swing_BH_01}}
]
properties {
 bOverrideConcurrency=true
 maxDistance=2097152
 soundClassObject=&Fort_BossFight_SFX
 volumeMultiplier=2
 attenuationOverrides {falloffDistance=25000, attenuationShapeExtents {x=6000}}
 concurrencyOverrides {maxCount=2, resolutionRule=StopOldest}
}
