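# Sound cue for the explosion when a ToyFox weakpoint takes damage, in the boss-fight SFX sound class.
# A mixer layers a close branch and a slightly quieter distant branch, each with its own ToyFox attack attenuation.
# Each branch passes through a modulator and randomly picks one of four equally weighted wave variations.
# A concurrency override limits it to two instances, stopping the oldest.
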SoundCue /ToyFoxGameplay/Assets/Sound/SC_ToyFox_SFX_Damage_Weakpoint_Expl
root=Mixer_0
nodes [
 {id=Attenuation_0, children [Random_0], properties {attenuationSettings=&/ATT_ToyFox_Attacks}}
 {id=Attenuation_1, children [Random_3], properties {attenuationSettings=&/ATT_ToyFox_Attacks_Distant}}
 {id=Mixer_0, children [Modulator_0,Modulator_1], properties {inputVolume [1,0.7]}}
 {id=Modulator_0, children [Attenuation_0]}
 {id=Modulator_1, children [Attenuation_1]}
 {id=Random_0, children [WavePlayer_0,WavePlayer_1,WavePlayer_2,WavePlayer_6], properties {weights [1,1,1,1]}}
 {id=Random_3, children [WavePlayer_11, WavePlayer_12, WavePlayer_13, WavePlayer_14], properties {weights [1,1,1,1]}}
 {id=WavePlayer_0, properties {soundWaveAssetPtr=/ToyFoxGameplay/Assets/Sound/SoundWaves/Toyfox_Weakpoint_Expl_Close_04}}
 {id=WavePlayer_1, properties {soundWaveAssetPtr=/ToyFoxGameplay/Assets/Sound/SoundWaves/Toyfox_Weakpoint_Expl_Close_03}}
 {id=WavePlayer_11, properties {soundWaveAssetPtr=/ToyFoxGameplay/Assets/Sound/SoundWaves/Toyfox_Weakpoint_Expl_Distant_04}}
 {id=WavePlayer_12, properties {soundWaveAssetPtr=/ToyFoxGameplay/Assets/Sound/SoundWaves/Toyfox_Weakpoint_Expl_Distant_03}}
 {id=WavePlayer_13, properties {soundWaveAssetPtr=/ToyFoxGameplay/Assets/Sound/SoundWaves/Toyfox_Weakpoint_Expl_Distant_02}}
 {id=WavePlayer_14, properties {soundWaveAssetPtr=/ToyFoxGameplay/Assets/Sound/SoundWaves/Toyfox_Weakpoint_Expl_Distant_01}}
 {id=WavePlayer_2, properties {soundWaveAssetPtr=/ToyFoxGameplay/Assets/Sound/SoundWaves/Toyfox_Weakpoint_Expl_Close_02}}
 {id=WavePlayer_6, properties {soundWaveAssetPtr=/ToyFoxGameplay/Assets/Sound/SoundWaves/Toyfox_Weakpoint_Expl_Close_01}}
]
properties {
 bOverrideConcurrency=true
 maxDistance=100000
 soundClassObject=&Fort_BossFight_SFX
 volumeMultiplier=2
 attenuationOverrides {falloffDistance=25000, attenuationShapeExtents {x=6000}}
 concurrencyOverrides {maxCount=2, resolutionRule=StopOldest}
}
