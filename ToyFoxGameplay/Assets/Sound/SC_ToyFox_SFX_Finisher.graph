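# Sound cue for the ToyFox finisher sound effect, playing a single wave through a modulator that varies pitch slightly downward on each play.
# It is spatialized with the shared ToyFox attacks attenuation, carries long-range falloff values, and uses the scripted 3D events sound class.
# A bus send feeds the FortCreatures volume sidechain bus, and a concurrency override caps simultaneous instances, stopping the oldest.
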SoundCue /ToyFoxGameplay/Assets/Sound/SC_ToyFox_SFX_Finisher
root=Modulator_0
nodes [
 {id=Modulator_0, children [WavePlayer_3], properties {pitchMax=1, pitchMin=0.85}}
 {id=WavePlayer_3, properties {soundWaveAssetPtr=/ToyFoxGameplay/Assets/Sound/SoundWaves/ToyFox_SFX_Finisher_01}}
]
properties {
 attenuationSettings=&/ATT_ToyFox_Attacks
 bOverrideConcurrency=true
 maxDistance=90000
 soundClassObject=/Game/Sounds/FortSoundClassesAndModes/Fort_Events_Scripted_3D
 volumeMultiplier=0.8
 attenuationOverrides {falloffDistance=25000, attenuationShapeExtents {x=6000}}
 busSends [{audioBus=&AB_Sidechain_FortCreatures_Vol, minSendLevel=0, sendLevel=0.7, soundSourceBus=null}]
 concurrencyOverrides {maxCount=2, resolutionRule=StopOldest}
}
