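# Sound cue for the ToyFox finisher attack's impact, playing as a scripted 3D event sound.
# A mixer layers a close impact wave with a quieter, delayed distant impact wave, each under its own attenuation asset.
# It sends to creature and music sidechain volume buses, more lightly to music, and overrides concurrency to a single instance that stops the oldest.
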SoundCue /ToyFoxGameplay/Assets/Sound/SC_ToyFox_SFX_Finisher_Impact
root=Mixer_0
nodes [
 {id=Attenuation_0, children [Delay_0], properties {attenuationSettings=&/ATT_ToyFox_Attacks_Distant}}
 {id=Attenuation_1, children [WavePlayer_1], properties {attenuationSettings=&/ATT_ToyFox_Attacks}}
 {id=Delay_0, children [WavePlayer_0], properties {delayMax=0.58, delayMin=0.58}}
 {id=Mixer_0, children [Attenuation_0, Attenuation_1], properties {inputVolume [0.4,1]}}
 {id=WavePlayer_0, properties {soundWaveAssetPtr=/ToyFoxGameplay/Assets/Sound/SoundWaves/ToyFox_SFX_Finisher_Impact_Distant_01}}
 {id=WavePlayer_1, properties {soundWaveAssetPtr=/ToyFoxGameplay/Assets/Sound/SoundWaves/ToyFox_SFX_Finisher_Impact_Close_01}}
]
properties {
 bOverrideConcurrency=true
 maxDistance=100000
 soundClassObject=/Game/Sounds/FortSoundClassesAndModes/Fort_Events_Scripted_3D
 volumeMultiplier=1
 attenuationOverrides {falloffDistance=25000, attenuationShapeExtents {x=6000}}
 busSends [
  {audioBus=&AB_Sidechain_FortCreatures_Vol, minSendLevel=0, sendLevel=1, soundSourceBus=null}
  {audioBus=&AB_Sidechain_FortMusic_Vol, minSendLevel=0, sendLevel=0.2, soundSourceBus=null}
 ]
 concurrencyOverrides {maxCount=1, resolutionRule=StopOldest}
}
