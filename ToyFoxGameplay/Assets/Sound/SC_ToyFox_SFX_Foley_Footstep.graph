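# Sound cue for ToyFox foley footsteps, capped at two concurrent instances by stopping the oldest.
# An Islocal branch routes true through an unconfigured attenuation node and false through a volume boost and ATT_ToyFox_Foley attenuation; unset has no input.
# Both paths share a pitch-varying modulator over a quality switch.
# The first quality level randomizes eight step waves, while the other three share a pool of the first four.
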SoundCue /ToyFoxGameplay/Assets/Sound/SC_ToyFox_SFX_Foley_Footstep
root=Branch_0
nodes [
 {id=Attenuation_0, children [Modulator_0]}
 {id=Attenuation_1, children [Modulator_0], properties {attenuationSettings=&/ATT_ToyFox_Foley}}
 {id=Branch_0, children [Attenuation_0, Modulator_1, null], properties {boolParameterName=Islocal}}
 {id=Modulator_0, children [QualityLevel_2], properties {pitchMax=1, pitchMin=0.85}}
 {id=Modulator_1, children [Attenuation_1], properties {pitchMax=1, pitchMin=1, volumeMax=1.5, volumeMin=1.5}}
 {id=QualityLevel_2, children [Random_1,Random_3,Random_3,Random_3]}
 {
  id=Random_1
  children [WavePlayer_4, WavePlayer_5, WavePlayer_18, WavePlayer_19, WavePlayer_20, WavePlayer_21, WavePlayer_22, WavePlayer_23]
  properties {weights [1,1,1,1,1,1,1,1]}
 }
 {id=Random_3, children [WavePlayer_24, WavePlayer_25, WavePlayer_26, WavePlayer_27], properties {weights [1,1,1,1]}}
 {id=WavePlayer_18, properties {soundWaveAssetPtr=/ToyFoxGameplay/Assets/Sound/SoundWaves/ToyFox_SFX_Foley_Step_03}}
 {id=WavePlayer_19, properties {soundWaveAssetPtr=/ToyFoxGameplay/Assets/Sound/SoundWaves/ToyFox_SFX_Foley_Step_04}}
 {id=WavePlayer_20, properties {soundWaveAssetPtr=/ToyFoxGameplay/Assets/Sound/SoundWaves/ToyFox_SFX_Foley_Step_05}}
 {id=WavePlayer_21, properties {soundWaveAssetPtr=/ToyFoxGameplay/Assets/Sound/SoundWaves/ToyFox_SFX_Foley_Step_06}}
 {id=WavePlayer_22, properties {soundWaveAssetPtr=/ToyFoxGameplay/Assets/Sound/SoundWaves/ToyFox_SFX_Foley_Step_07}}
 {id=WavePlayer_23, properties {soundWaveAssetPtr=/ToyFoxGameplay/Assets/Sound/SoundWaves/ToyFox_SFX_Foley_Step_08}}
 {id=WavePlayer_24, properties {soundWaveAssetPtr=/ToyFoxGameplay/Assets/Sound/SoundWaves/ToyFox_SFX_Foley_Step_01}}
 {id=WavePlayer_25, properties {soundWaveAssetPtr=/ToyFoxGameplay/Assets/Sound/SoundWaves/ToyFox_SFX_Foley_Step_02}}
 {id=WavePlayer_26, properties {soundWaveAssetPtr=&ToyFox_SFX_Foley_Step_03}}
 {id=WavePlayer_27, properties {soundWaveAssetPtr=&ToyFox_SFX_Foley_Step_04}}
 {id=WavePlayer_4, properties {soundWaveAssetPtr=&ToyFox_SFX_Foley_Step_01}}
 {id=WavePlayer_5, properties {soundWaveAssetPtr=&ToyFox_SFX_Foley_Step_02}}
]
properties {
 bOverrideConcurrency=true
 maxDistance=2097152
 soundClassObject=&Fort_BossFight_SFX
 volumeMultiplier=1.5
 attenuationOverrides {falloffDistance=25000, attenuationShapeExtents {x=6000}}
 concurrencyOverrides {maxCount=2, resolutionRule=StopOldest}
}
