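# Sound cue that plays the ToyFox's movement foley, classed as boss-fight SFX.
# An Islocal branch picks a default attenuation node when true or ATT_ToyFox_Foley when false, with nothing wired for an unset parameter.
# Both feed one pitch-lowering modulator over a quality switch: eight random waves at the first level, four at the others.
# Overridden concurrency caps simultaneous instances at two, stopping the oldest.
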SoundCue /ToyFoxGameplay/Assets/Sound/SC_ToyFox_SFX_Foley_Move
root=Branch_0
nodes [
 {id=Attenuation_0, children [Modulator_0]}
 {id=Attenuation_1, children [Modulator_0], properties {attenuationSettings=&/ATT_ToyFox_Foley}}
 {id=Branch_0, children [Attenuation_0, Attenuation_1, null], properties {boolParameterName=Islocal}}
 {id=Modulator_0, children [QualityLevel_2], properties {pitchMax=1, pitchMin=0.85}}
 {id=QualityLevel_2, children [Random_0,Random_2,Random_2,Random_2]}
 {
  id=Random_0
  children [WavePlayer_0,WavePlayer_1,WavePlayer_2,WavePlayer_3,WavePlayer_6,WavePlayer_7,WavePlayer_8,WavePlayer_9]
  properties {weights [1,1,1,1,1,1,1,1]}
 }
 {id=Random_2, children [WavePlayer_10, WavePlayer_11, WavePlayer_12, WavePlayer_13], properties {weights [1,1,1,1]}}
 {id=WavePlayer_0, properties {soundWaveAssetPtr=/ToyFoxGameplay/Assets/Sound/SoundWaves/ToyFox_SFX_Foley_Movement_01}}
 {id=WavePlayer_1, properties {soundWaveAssetPtr=/ToyFoxGameplay/Assets/Sound/SoundWaves/ToyFox_SFX_Foley_Movement_02}}
 {id=WavePlayer_10, properties {soundWaveAssetPtr=&ToyFox_SFX_Foley_Movement_01}}
 {id=WavePlayer_11, properties {soundWaveAssetPtr=&ToyFox_SFX_Foley_Movement_02}}
 {id=WavePlayer_12, properties {soundWaveAssetPtr=/ToyFoxGameplay/Assets/Sound/SoundWaves/ToyFox_SFX_Foley_Movement_03}}
 {id=WavePlayer_13, properties {soundWaveAssetPtr=/ToyFoxGameplay/Assets/Sound/SoundWaves/ToyFox_SFX_Foley_Movement_04}}
 {id=WavePlayer_2, properties {soundWaveAssetPtr=&ToyFox_SFX_Foley_Movement_03}}
 {id=WavePlayer_3, properties {soundWaveAssetPtr=&ToyFox_SFX_Foley_Movement_04}}
 {id=WavePlayer_6, properties {soundWaveAssetPtr=/ToyFoxGameplay/Assets/Sound/SoundWaves/ToyFox_SFX_Foley_Movement_05}}
 {id=WavePlayer_7, properties {soundWaveAssetPtr=/ToyFoxGameplay/Assets/Sound/SoundWaves/ToyFox_SFX_Foley_Movement_06}}
 {id=WavePlayer_8, properties {soundWaveAssetPtr=/ToyFoxGameplay/Assets/Sound/SoundWaves/ToyFox_SFX_Foley_Movement_07}}
 {id=WavePlayer_9, properties {soundWaveAssetPtr=/ToyFoxGameplay/Assets/Sound/SoundWaves/ToyFox_SFX_Foley_Movement_08}}
]
properties {
 bOverrideConcurrency=true
 maxDistance=2097152
 soundClassObject=&Fort_BossFight_SFX
 volumeMultiplier=0.6
 attenuationOverrides {falloffDistance=25000, attenuationShapeExtents {x=6000}}
 concurrencyOverrides {maxCount=2, resolutionRule=StopOldest}
}
