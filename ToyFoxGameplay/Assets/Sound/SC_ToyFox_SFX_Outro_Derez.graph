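# Sound cue for the ToyFox outro derez sound effect.
# A single wave player feeds the output at reduced volume, under a scripted 3D events sound class with ToyFox attack attenuation.
# Overridden concurrency allows one instance at a time, stopping the oldest.
# A bus send feeds the creature volume sidechain audio bus.
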SoundCue /ToyFoxGameplay/Assets/Sound/SC_ToyFox_SFX_Outro_Derez
root=WavePlayer_2
nodes [{id=WavePlayer_2, properties {soundWaveAssetPtr=/ToyFoxGameplay/Assets/Sound/SoundWaves/ToyFox_SFX_Outro_Derez_01}}]
properties {
 attenuationSettings=&/ATT_ToyFox_Attacks
 bOverrideConcurrency=true
 maxDistance=90000
 soundClassObject=/Game/Sounds/FortSoundClassesAndModes/Fort_Events_Scripted_3D
 volumeMultiplier=0.7
 attenuationOverrides {falloffDistance=25000, attenuationShapeExtents {x=6000}}
 busSends [{audioBus=&AB_Sidechain_FortCreatures_Vol, minSendLevel=0, sendLevel=1, soundSourceBus=null}]
 concurrencyOverrides {maxCount=1, resolutionRule=StopOldest}
}
