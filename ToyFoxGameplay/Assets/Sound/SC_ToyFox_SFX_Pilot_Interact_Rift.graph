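# Sound cue for the ToyFox pilot's rift interaction SFX, in the boss fight SFX sound class.
# A Branch on the "islocal" bool plays mixed close and distant waves without attenuation nodes when true.
# When false, the same waves mix through separate close and distant attack attenuation assets; the unset input is empty.
# An overridden concurrency allows two instances, stopping the oldest.
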SoundCue /ToyFoxGameplay/Assets/Sound/SC_ToyFox_SFX_Pilot_Interact_Rift
root=Branch_0
nodes [
 {id=Attenuation_0, children [WavePlayer_3], properties {attenuationSettings=&/ATT_ToyFox_Attacks_Distant}}
 {id=Attenuation_2, children [WavePlayer_2], properties {attenuationSettings=&/ATT_ToyFox_Attacks}}
 {id=Branch_0, children [Mixer_1,Mixer_0,null], properties {boolParameterName=islocal}}
 {id=Mixer_0, children [Attenuation_0, Attenuation_2], properties {inputVolume [1,1]}}
 {id=Mixer_1, children [WavePlayer_5,WavePlayer_4], properties {inputVolume [1,1]}}
 {id=WavePlayer_2, properties {soundWaveAssetPtr=/ToyFoxGameplay/Assets/Sound/SoundWaves/ToyFox_Pilot_Interact_Close}}
 {id=WavePlayer_3, properties {soundWaveAssetPtr=/ToyFoxGameplay/Assets/Sound/SoundWaves/ToyFox_Pilot_Interact_Distant}}
 {id=WavePlayer_4, properties {soundWaveAssetPtr=&ToyFox_Pilot_Interact_Close}}
 {id=WavePlayer_5, properties {soundWaveAssetPtr=&ToyFox_Pilot_Interact_Distant}}
]
properties {
 bOverrideConcurrency=true
 maxDistance=100000
 soundClassObject=&Fort_BossFight_SFX
 volumeMultiplier=1
 attenuationOverrides {falloffDistance=25000, attenuationShapeExtents {x=6000}}
 concurrencyOverrides {maxCount=2, resolutionRule=StopOldest}
}
