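# Sound cue for the ToyFox pilot's cinematic interact sequence sound effect.
# A single wave player feeds the Pilot_Interact_CineSeq wave straight to the output, routed through the boss fight SFX sound class.
# It references the ToyFox attacks attenuation asset for long-range falloff and overrides concurrency so only one instance plays, stopping the oldest.
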SoundCue /ToyFoxGameplay/Assets/Sound/SC_ToyFox_SFX_Pilot_Sequence
root=WavePlayer_3
nodes [{id=WavePlayer_3, properties {soundWaveAssetPtr=/ToyFoxGameplay/Assets/Sound/SoundWaves/ToyFox_Pilot_Interact_CineSeq}}]
properties {
 attenuationSettings=&/ATT_ToyFox_Attacks
 bOverrideConcurrency=true
 maxDistance=90000
 soundClassObject=&Fort_BossFight_SFX
 volumeMultiplier=1
 attenuationOverrides {falloffDistance=25000, attenuationShapeExtents {x=6000}}
 concurrencyOverrides {maxCount=1, resolutionRule=StopOldest}
}
