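# Sound cue for the ToyFox spawn sequence that plays the left leg approach sound effect.
# A single wave player feeds one sound wave straight to the output, with no randomization or mixing, at a reduced volume.
# It uses the shared ToyFox spawn-assemble attenuation asset, a very large max distance and the scripted-events 3D sound class.
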SoundCue /ToyFoxGameplay/Assets/Sound/SC_ToyFox_SFX_Spawn_LeftLeg_Approach
root=WavePlayer_0
nodes [{id=WavePlayer_0, properties {soundWaveAssetPtr=/ToyFoxGameplay/Assets/Sound/SoundWaves/ToyFox_SFX_Spawn_LeftLeg_Approach_01}}]
properties {
 attenuationSettings=&/ATT_Toyfox_Spawn_Assemble
 maxDistance=180000
 soundClassObject=/Game/Sounds/FortSoundClassesAndModes/Fort_Events_Scripted_3D
 volumeMultiplier=0.7
 attenuationOverrides {falloffDistance=10000, stereoSpread=5000, attenuationShapeExtents {x=1000}}
}
