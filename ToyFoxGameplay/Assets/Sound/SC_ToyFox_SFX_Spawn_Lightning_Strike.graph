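# Sound cue for the ToyFox spawn lightning strike effect, routed to the Fort_BossFight_SFX sound class.
# A single wave player feeds the LightningBolt_01 wave straight to the output, with no randomization or mixing.
# Spatial falloff references a dedicated lightning strike attenuation asset.
# A local concurrency override caps simultaneous strikes at a few voices, stopping the oldest with a short release when exceeded.
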SoundCue /ToyFoxGameplay/Assets/Sound/SC_ToyFox_SFX_Spawn_Lightning_Strike
root=WavePlayer_5
nodes [{id=WavePlayer_5, properties {soundWaveAssetPtr=/ToyFoxGameplay/Assets/Sound/SoundWaves/ToyFox_SFX_Spawn_LightningBolt_01}}]
properties {
 attenuationSettings=&/ATT_Toyfox_Spawn_Lightning_Strike
 bOverrideConcurrency=true
 maxDistance=200000
 soundClassObject=&Fort_BossFight_SFX
 volumeMultiplier=1
 attenuationOverrides {falloffDistance=10000, stereoSpread=5000, attenuationShapeExtents {x=1000}}
 concurrencyOverrides {maxCount=3, resolutionRule=StopOldest, voiceStealReleaseTime=0.1}
}
