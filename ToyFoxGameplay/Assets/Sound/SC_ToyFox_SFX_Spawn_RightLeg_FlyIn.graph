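# Sound cue for the fly-in sound of the ToyFox's right leg during its spawn sequence.
# A single looping wave player plays the fly-in loop wave directly as the root, with no mixing or randomization.
# Playback runs at reduced volume through the 3D scripted-events sound class, using the shared spawn-pieces fly-in attenuation asset alongside a very large max distance.
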SoundCue /ToyFoxGameplay/Assets/Sound/SC_ToyFox_SFX_Spawn_RightLeg_FlyIn
root=WavePlayer_1
nodes [{
  id=WavePlayer_1
  properties {bLooping=true, soundWaveAssetPtr=/ToyFoxGameplay/Assets/Sound/SoundWaves/ToyFox_SFX_Spawn_RightLeg_FlyIn_Loop_01}
 }]
properties {
 attenuationSettings=&/ATT_Toyfox_Spawn_Pieces_FlyIn
 maxDistance=200000
 soundClassObject=/Game/Sounds/FortSoundClassesAndModes/Fort_Events_Scripted_3D
 volumeMultiplier=0.5
 attenuationOverrides {falloffDistance=10000, stereoSpread=5000, attenuationShapeExtents {x=1000}}
}
