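# Sound cue for the ToyFox low-health warning, playing a single warning wave at boosted volume in the critical in-game UI sound class.
# It references the ToyFox attacks attenuation asset despite being named as a UI sound.
# A concurrency override allows one instance at a time and blocks new ones.
# Partial bus sends feed the music volume sidechain and the boss VO envelope-follower bus.
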SoundCue /ToyFoxGameplay/Assets/Sound/SC_ToyFox_SFX_UI_LowHealth
root=WavePlayer_1
nodes [{id=WavePlayer_1, properties {soundWaveAssetPtr=/ToyFoxGameplay/Assets/Sound/SoundWaves/ToyFox_LowHealth_Warning_01}}]
properties {
 attenuationSettings=&/ATT_ToyFox_Attacks
 bOverrideConcurrency=true
 maxDistance=90000
 soundClassObject=&Fort_UI_InGame_Critical
 volumeMultiplier=1.5
 attenuationOverrides {falloffDistance=10000, stereoSpread=5000, attenuationShapeExtents {x=1000}}
 busSends [
  {audioBus=&AB_Sidechain_FortMusic_Vol, minSendLevel=0, sendLevel=0.3, soundSourceBus=null}
  {audioBus=/Game/Sounds/Modulation/EnvelopeFollowers/BR_Boss/AB_BR_Boss_VO_EnvFollow, minSendLevel=0, sendLevel=0.15, soundSourceBus=null}
 ]
 concurrencyOverrides {maxCount=1, resolutionRule=PreventNew}
}
